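# Control Rig for the PirouetteWeld female medium player body that applies pose-driven corrections to arm, shoulder and cloth bones.
# Spherical pose readers on upper arms, thighs, calf and spine weight transform offsets on twist, armor, dyn_Cloth_volume and dyn shoulder bones.
# Cloth volume bones also blend between thigh and pelvis through weighted parent constraints.
# Sequenced branches include shoulder fixes scaled by Skydive, ParachuteAfterSkydive, FroneEnd, Pistol and Rifle variables through remapped AlphaInterp.
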ControlRigBlueprint /Game/Characters/Player/Female/Medium/Bodies/F_MED_PirouetteWeld/Meshes/F_MED_PirouetteWeld_CtrlRig
graphs [
 ContainedGraph {nodes [
   n1 = RigVMFunctionEntryNode()
   SequenceAggregate = &RigVMFunction_Sequence(ExecuteContext=@n1.ExecuteContext)
   SequenceAggregate_1 = &RigVMFunction_Sequence(ExecuteContext=@SequenceAggregate.B)
   SequenceAggregate_2 = &RigVMFunction_Sequence(ExecuteContext=@SequenceAggregate_1.B)
   SequenceAggregate_3 = &RigVMFunction_Sequence(ExecuteContext=@SequenceAggregate_2.B)
   SequenceAggregate_4 = &RigVMFunction_Sequence(ExecuteContext=@SequenceAggregate_3.B)
   SequenceAggregate_5 = &RigVMFunction_Sequence(ExecuteContext=@SequenceAggregate_4.B)
   RigVMFunctionReturnNode(
    A=@SequenceAggregate.A
    B=@SequenceAggregate_1.A
    C=@SequenceAggregate_2.A
    D=@SequenceAggregate_3.A
    E=@SequenceAggregate_4.A
    F=@SequenceAggregate_5.A
    G=@SequenceAggregate_5.B
   )
  ]}
 RigVMFunctionLibrary {links [], nodes []}
 RigVMModel {nodes [
   RigUnit_BeginExecution = &RigUnit_BeginExecution()
   SequenceAggregate = RigVMAggregateNode(ExecuteContext=@RigUnit_BeginExecution.ExecutePin)
   n1 = &RigUnit_SphericalPoseReader(
    ActiveRegionScaleFactors=(PositiveWidth=2.5,NegativeWidth=3,PositiveHeight=3,NegativeHeight=3)
    ActiveRegionSize=0.1
    Debug=(bDrawDebug=false,DebugScale=25,DebugSegments=20,DebugThickness=0.25)
    DriverAxis=(X=1,Y=0,Z=0)
    DriverItem=(Type=Bone,Name="upperarm_twist_01_l")
    FalloffRegionScaleFactors=(PositiveWidth=1,NegativeWidth=1,PositiveHeight=1,NegativeHeight=1)
    FalloffSize=0.7
    OptionalParentItem=(Type=Bone)
    RotationOffset=(X=0,Y=10,Z=0)
    ExecutePin=@SequenceAggregate.A
   )
   > &RigUnit_ModifyTransforms(
    ItemToModify="((Item=(Type=Bone,Name=\"upperarm_l\"),Transform=(Rotation=(X=0,Y=0,Z=0,W=1),Translation=(X=0,Y=0,Z=0),Scale3D=(X=1,Y=1,Z=1))),(Item=(Type=Bone,Name=\"upperarm_twist_01_l\"),Transform=(Rotation=(X=0,Y=0,Z=0,W=1),Translation=(X=0,Y=0,Z=0),Scale3D=(X=1,Y=1,Z=1)) ...(truncated)"
    ItemToModify.1.Transform.Rotation=&RigVMFunction_MathQuaternionFromEuler(Euler=(X=20,Y=0,Z=0)).Result
    Weight=@n1.OutputParam
   )
   > &RigUnit_SphericalPoseReader(
    ActiveRegionScaleFactors=(PositiveWidth=2,NegativeWidth=3,PositiveHeight=1,NegativeHeight=1)
    ActiveRegionSize=0.1
    Debug=(bDrawDebug=false,DebugScale=25,DebugSegments=20,DebugThickness=0.25)
    DriverAxis=(X=1,Y=0,Z=0)
    DriverItem=(Type=Bone,Name="upperarm_l")
    FalloffRegionScaleFactors=(PositiveWidth=1,NegativeWidth=1,PositiveHeight=2,NegativeHeight=2)
    FalloffSize=0.7
    OptionalParentItem=(Type=Bone)
    RotationOffset=(X=0,Y=10,Z=0)
   )
   > n2 = &RigUnit_SphericalPoseReader(
    ActiveRegionScaleFactors=(PositiveWidth=2.5,NegativeWidth=3,PositiveHeight=3,NegativeHeight=3)
    ActiveRegionSize=0.1
    Debug=(bDrawDebug=false,DebugScale=25,DebugSegments=20,DebugThickness=0.25)
    DriverAxis=(X=-1,Y=0,Z=0)
    DriverItem=(Type=Bone,Name="upperarm_r")
    FalloffRegionScaleFactors=(PositiveWidth=1,NegativeWidth=1,PositiveHeight=1,NegativeHeight=1)
    FalloffSize=0.5
    FlipHeightScaling=true
    OptionalParentItem=(Type=Bone)
    RotationOffset=(X=180,Y=0,Z=180)
   )
   > &RigUnit_ModifyTransforms(
    ItemToModify=((Item=(Type=Bone,Name="upperarm_twist_01_r"),Transform=(Rotation=(X=0,Y=0,Z=0,W=1),Translation=(X=0,Y=0,Z=0),Scale3D=(X=1,Y=1,Z=1))))
    ItemToModify.0.Transform.Rotation=&RigVMFunction_MathQuaternionFromEuler(Euler=(X=-20,Y=0,Z=0)).Result
    Weight=@n2.OutputParam
   )
   > n3 = &RigUnit_SphericalPoseReader(
    ActiveRegionScaleFactors=(PositiveWidth=2,NegativeWidth=3,PositiveHeight=1,NegativeHeight=1)
    ActiveRegionSize=0.1
    Debug=(bDrawDebug=false,DebugScale=25,DebugSegments=20,DebugThickness=0.25)
    DriverAxis=(X=-1,Y=0,Z=0)
    DriverItem=(Type=Bone,Name="upperarm_r")
    FalloffRegionScaleFactors=(PositiveWidth=1,NegativeWidth=1,PositiveHeight=0.9,NegativeHeight=0.9)
    FalloffSize=0.7
    FlipWidthScaling=false
    OptionalParentItem=(Type=Bone)
    RotationOffset=(X=180,Y=-10,Z=180)
   )
   > &RigUnit_ModifyTransforms(
    ItemToModify="((Item=(Type=Bone,Name=\"upperarm_armor_01_r\"),Transform=(Rotation=(X=0,Y=0,Z=0,W=1),Translation=(X=-1.5,Y=2,Z=-1.5),Scale3D=(X=0.6,Y=1,Z=1))),(Item=(Type=Bone,Name=\"None\"),Transform=(Rotation=(X=0,Y=0,Z=0,W=1),Translation=(X=0,Y=0,Z=0),Scale3D=(X=0,Y=0,Z=0 ...(truncated)"
    ItemToModify.0.Transform.Rotation=&RigVMFunction_MathQuaternionFromEuler(Euler=(X=100,Y=0,Z=20)).Result
    Weight=@n3.OutputParam
   )
   n4 = &RigUnit_SphericalPoseReader(
    ActiveRegionScaleFactors=(PositiveWidth=3,NegativeWidth=0.5,PositiveHeight=3,NegativeHeight=0.1)
    ActiveRegionSize=0.1
    Debug=(bDrawDebug=false,DebugScale=25,DebugSegments=20,DebugThickness=0.25)
    DriverAxis=(X=-1,Y=0,Z=0)
    DriverItem=(Type=Bone,Name="thigh_twist_01_l")
    FalloffRegionScaleFactors=(PositiveWidth=1,NegativeWidth=2,PositiveHeight=5,NegativeHeight=1)
    FalloffSize=0.38
    OptionalParentItem=(Type=Bone,Name="pelvis")
    RotationOffset=(X=90,Y=0,Z=-35)
    ExecutePin=@SequenceAggregate.B
   )
   > &RigUnit_ParentConstraint(
    AdvancedSettings=(RotationOrderForFilter=XZY)
    Child=(Type=Bone,Name="dyn_Cloth_volume_l_A")
    Filter=(TranslationFilter=(bX=false,bY=false,bZ=false),RotationFilter=(bX=True,bY=True,bZ=True),ScaleFilter=(bX=True,bY=True,bZ=True))
    Parents=((Item=(Type=Bone,Name="thigh_twist_01_l"),Weight=0.5),(Item=(Type=Bone,Name="pelvis"),Weight=0.5))
    Weight=@n4.OutputParam
   )
   > &RigUnit_ParentConstraint(
    AdvancedSettings=(RotationOrderForFilter=XZY)
    Child=(Type=Bone,Name="dyn_Cloth_volume_l_B")
    Parents=((Item=(Type=Bone,Name="thigh_twist_01_l"),Weight=0.25),(Item=(Type=Bone,Name="pelvis"),Weight=0.75))
    Weight=@n4.OutputParam
   )
   > &RigUnit_ModifyTransforms(
    ItemToModify="((Item=(Type=Bone,Name=\"dyn_Cloth_volume_l_A\"),Transform=(Rotation=(X=0,Y=0,Z=0,W=1),Translation=(X=0,Y=-1,Z=-12),Scale3D=(X=1,Y=1,Z=1))),(Item=(Type=Bone,Name=\"dyn_Cloth_volume_l_B\"),Transform=(Rotation=(X=0,Y=0,Z=0,W=1),Translation=(X=2,Y=-2,Z=0),Scale3D ...(truncated)"
    ItemToModify.0.Transform.Rotation=&RigVMFunction_MathQuaternionFromEuler(Euler=(X=-12.5,Y=-30,Z=65)).Result
    ItemToModify.1.Transform.Rotation=&RigVMFunction_MathQuaternionFromEuler(Euler=(X=-100,Y=-40,Z=30)).Result
    ItemToModify.2.Transform.Rotation=&RigVMFunction_MathQuaternionFromEuler(Euler=(X=0,Y=0,Z=-50)).Result
    Weight=@n4.OutputParam
   )
   > n5 = &RigUnit_SphericalPoseReader(
    ActiveRegionScaleFactors=(PositiveWidth=5.5,NegativeWidth=4.5,PositiveHeight=2,NegativeHeight=2)
    ActiveRegionSize=0.1
    Debug=(bDrawDebug=false,DebugScale=25,DebugSegments=20,DebugThickness=0.25)
    DriverAxis=(X=-1,Y=0,Z=0)
    DriverItem=(Type=Bone,Name="thigh_l")
    FalloffRegionScaleFactors=(PositiveWidth=1,NegativeWidth=1.5,PositiveHeight=1,NegativeHeight=1)
    FalloffSize=0.7
    FlipWidthScaling=false
    OptionalParentItem=(Type=Bone)
    RotationOffset=(X=180,Y=50,Z=0)
   )
   > &RigUnit_ParentConstraint(
    AdvancedSettings=(RotationOrderForFilter=XZY)
    Child=(Type=Bone,Name="dyn_Cloth_volume_l_B")
    Filter=(TranslationFilter=(bX=false,bY=false,bZ=false),RotationFilter=(bX=True,bY=True,bZ=True),ScaleFilter=(bX=True,bY=True,bZ=True))
    Parents=((Item=(Type=Bone,Name="thigh_twist_01_l"),Weight=0.8),(Item=(Type=Bone,Name="pelvis"),Weight=0.2))
    Weight=@n5.OutputParam
   )
   > &RigUnit_ParentConstraint(
    AdvancedSettings=(RotationOrderForFilter=XZY)
    Child=(Type=Bone,Name="dyn_Cloth_volume_l_A")
    Filter=(TranslationFilter=(bX=false,bY=false,bZ=false),RotationFilter=(bX=True,bY=True,bZ=True),ScaleFilter=(bX=True,bY=True,bZ=True))
    Parents=((Item=(Type=Bone,Name="thigh_twist_01_l"),Weight=0.8),(Item=(Type=Bone,Name="pelvis"),Weight=0.2))
    Weight=@n5.OutputParam
   )
   > &RigUnit_ModifyTransforms(
    ItemToModify="((Item=(Type=Bone,Name=\"dyn_Cloth_volume_l_B\"),Transform=(Rotation=(X=0,Y=0,Z=0,W=1),Translation=(X=0,Y=0,Z=0),Scale3D=(X=1,Y=1,Z=1))),(Item=(Type=Bone,Name=\"dyn_Cloth_volume_l_A\"),Transform=(Rotation=(X=0,Y=0,Z=0,W=1),Translation=(X=0,Y=0,Z=0),Scale3D=(X= ...(truncated)"
    ItemToModify.2.Transform.Rotation=&RigVMFunction_MathQuaternionFromEuler(Euler=(X=0,Y=20,Z=0)).Result
    ItemToModify.1.Transform.Rotation=&RigVMFunction_MathQuaternionFromEuler(Euler=(X=0,Y=5,Z=5)).Result
    ItemToModify.0.Transform.Rotation=&RigVMFunction_MathQuaternionFromEuler(Euler=(X=0,Y=5,Z=15)).Result
    Weight=@n5.OutputParam
   )
   > n6 = &RigUnit_SphericalPoseReader(
    ActiveRegionScaleFactors=(PositiveWidth=3.5,NegativeWidth=0.1,PositiveHeight=0.5,NegativeHeight=2.3)
    ActiveRegionSize=0.1
    Debug=(bDrawDebug=false,DebugScale=25,DebugSegments=20,DebugThickness=0.25)
    DriverAxis=(X=-1,Y=0,Z=0)
    DriverItem=(Type=Bone,Name="thigh_l")
    FalloffRegionScaleFactors=(PositiveWidth=1,NegativeWidth=1.6,PositiveHeight=1,NegativeHeight=3)
    FalloffSize=0.3
    OptionalParentItem=(Type=Bone)
    RotationOffset=(X=-90,Y=0,Z=0)
   )
   > &RigUnit_ParentConstraint(
    AdvancedSettings=(RotationOrderForFilter=XZY)
    Child=(Type=Bone,Name="dyn_Cloth_volume_l_C")
    Filter=(TranslationFilter=(bX=false,bY=false,bZ=false),RotationFilter=(bX=True,bY=True,bZ=True),ScaleFilter=(bX=True,bY=True,bZ=True))
    Parents=((Item=(Type=Bone,Name="thigh_twist_01_l"),Weight=0.5),(Item=(Type=Bone,Name="pelvis"),Weight=0.5))
    Weight=@n6.OutputParam
   )
   > &RigUnit_ModifyTransforms(
    ItemToModify="((Item=(Type=Bone,Name=\"dyn_Cloth_volume_l_C\"),Transform=(Rotation=(X=0,Y=0,Z=0,W=1),Translation=(X=0,Y=0,Z=0),Scale3D=(X=1,Y=1,Z=1))),(Item=(Type=Bone,Name=\"dyn_Cloth_volume_l_B\"),Transform=(Rotation=(X=0,Y=0,Z=0,W=1),Translation=(X=0,Y=0,Z=0),Scale3D=(X= ...(truncated)"
    ItemToModify.2.Transform.Rotation=&RigVMFunction_MathQuaternionFromEuler(Euler=(X=0,Y=20,Z=60)).Result
    ItemToModify.1.Transform.Rotation=&RigVMFunction_MathQuaternionFromEuler(Euler=(X=0,Y=0,Z=0)).Result
    ItemToModify.0.Transform.Rotation=&RigVMFunction_MathQuaternionFromEuler(Euler=(X=20,Y=0,Z=50)).Result
    Weight=@n6.OutputParam
   )
   > n7 = &RigUnit_SphericalPoseReader(
    ActiveRegionScaleFactors=(PositiveWidth=3.5,NegativeWidth=0.1,PositiveHeight=0.5,NegativeHeight=2.3)
    ActiveRegionSize=0.1
    Debug=(bDrawDebug=false,DebugScale=25,DebugSegments=20,DebugThickness=0.25)
    DriverAxis=(X=-1,Y=0,Z=0)
    DriverItem=(Type=Bone,Name="thigh_l")
    FalloffRegionScaleFactors=(PositiveWidth=1,NegativeWidth=1.5,PositiveHeight=1,NegativeHeight=3)
    FalloffSize=0.3
    OptionalParentItem=(Type=Bone)
    RotationOffset=(X=90,Y=0,Z=0)
   )
   > &RigUnit_ParentConstraint(
    AdvancedSettings=(RotationOrderForFilter=XZY)
    Child=(Type=Bone,Name="dyn_Cloth_volume_l_B")
    Filter=(TranslationFilter=(bX=false,bY=false,bZ=false),RotationFilter=(bX=True,bY=True,bZ=True),ScaleFilter=(bX=True,bY=True,bZ=True))
    Parents=((Item=(Type=Bone,Name="thigh_twist_01_l"),Weight=0.5),(Item=(Type=Bone,Name="pelvis"),Weight=0.5))
    Weight=@n7.OutputParam
   )
   > n8 = &RigUnit_SphericalPoseReader(
    ActiveRegionScaleFactors=(PositiveWidth=3,NegativeWidth=0.5,PositiveHeight=0.3,NegativeHeight=0.3)
    ActiveRegionSize=0.1
    Debug=(bDrawDebug=false,DebugScale=25,DebugSegments=20,DebugThickness=0.25)
    DriverAxis=(X=1,Y=0,Z=0)
    DriverItem=(Type=Bone,Name="spine_01")
    FalloffRegionScaleFactors=(PositiveWidth=1,NegativeWidth=1,PositiveHeight=1,NegativeHeight=1)
    FalloffSize=0.45
    FlipWidthScaling=false
    OptionalParentItem=(Type=Bone,Name="None")
    RotationOffset=(X=90,Y=0,Z=-10)
   )
   > &RigUnit_ParentConstraint(
    AdvancedSettings=(RotationOrderForFilter=XZY)
    Child=(Type=Bone,Name="dyn_Cloth_volume_l_C")
    Filter=(TranslationFilter=(bX=false,bY=false,bZ=false),RotationFilter=(bX=True,bY=True,bZ=True),ScaleFilter=(bX=True,bY=True,bZ=True))
    Parents=((Item=(Type=Bone,Name="thigh_l"),Weight=0.1),(Item=(Type=Bone,Name="pelvis"),Weight=0.9))
    Weight=@n8.OutputParam
   )
   > &RigUnit_ParentConstraint(
    AdvancedSettings=(RotationOrderForFilter=XZY)
    Child=(Type=Bone,Name="dyn_Cloth_volume_r_C")
    Filter=(TranslationFilter=(bX=false,bY=false,bZ=false),RotationFilter=(bX=True,bY=True,bZ=True),ScaleFilter=(bX=True,bY=True,bZ=True))
    Parents=((Item=(Type=Bone,Name="thigh_r"),Weight=0.1),(Item=(Type=Bone,Name="pelvis"),Weight=0.9))
    Weight=@n8.OutputParam
   )
   > &RigUnit_ParentConstraint(
    AdvancedSettings=(RotationOrderForFilter=XZY)
    Child=(Type=Bone,Name="dyn_Cloth_volume_l_C01")
    Filter=(TranslationFilter=(bX=false,bY=false,bZ=false),RotationFilter=(bX=True,bY=True,bZ=True),ScaleFilter=(bX=True,bY=True,bZ=True))
    Parents=((Item=(Type=Bone,Name="root"),Weight=0),(Item=(Type=Bone,Name="pelvis"),Weight=0.1))
    bMaintainOffset=true
    Weight=@n8.OutputParam
   )
   > &RigUnit_ParentConstraint(
    AdvancedSettings=(RotationOrderForFilter=XZY)
    Child=(Type=Bone,Name="dyn_Cloth_volume_r_C01")
    Filter=(TranslationFilter=(bX=false,bY=false,bZ=false),RotationFilter=(bX=True,bY=True,bZ=True),ScaleFilter=(bX=True,bY=True,bZ=True))
    Parents=((Item=(Type=Bone,Name="root"),Weight=0),(Item=(Type=Bone,Name="pelvis"),Weight=0.1))
    Weight=@n8.OutputParam
   )
   n9 = &RigVMFunction_MathQuaternionFromEuler(Euler=(X=0,Y=0,Z=-10))
   n10 = &RigVMFunction_MathQuaternionFromEuler(Euler=(X=0,Y=0,Z=-25))
   n11 = &RigVMFunction_MathQuaternionFromEuler(Euler=(X=0,Y=0,Z=-10))
   > &RigUnit_ModifyTransforms(
    ItemToModify="((Item=(Type=Bone,Name=\"dyn_Cloth_volume_l_C\"),Transform=(Rotation=(X=0,Y=0,Z=0,W=1),Translation=(X=0,Y=0,Z=0),Scale3D=(X=1,Y=1,Z=1))),(Item=(Type=Bone,Name=\"dyn_Cloth_volume_r_C\"),Transform=(Rotation=(X=0,Y=0,Z=0,W=1),Translation=(X=0,Y=-1.2,Z=0),Scale3D= ...(truncated)"
    ItemToModify.2.Transform.Rotation=@n9.Result
    ItemToModify.3.Transform.Rotation=@n9.Result
    ItemToModify.4.Transform.Rotation=@n10.Result
    ItemToModify.5.Transform.Rotation=@n10.Result
    ItemToModify.0.Transform.Rotation=@n11.Result
    ItemToModify.1.Transform.Rotation=@n11.Result
    Weight=@n8.OutputParam
   )
   n12 = &RigUnit_SphericalPoseReader(
    ActiveRegionScaleFactors=(PositiveWidth=3,NegativeWidth=2,PositiveHeight=3,NegativeHeight=0.1)
    ActiveRegionSize=0.1
    Debug=(bDrawDebug=true,DebugScale=25,DebugSegments=20,DebugThickness=0.25)
    DriverAxis=(X=1,Y=0,Z=0)
    DriverItem=(Type=Bone,Name="thigh_twist_01_r")
    FalloffRegionScaleFactors=(PositiveWidth=1,NegativeWidth=2,PositiveHeight=5,NegativeHeight=1)
    FalloffSize=0.38
    FlipHeightScaling=false
    FlipWidthScaling=true
    OptionalParentItem=(Type=Bone,Name="pelvis")
    RotationOffset=(X=-90,Y=0,Z=-35)
    ExecutePin=@SequenceAggregate.C
   )
   > &RigUnit_ParentConstraint(
    AdvancedSettings=(RotationOrderForFilter=XZY)
    Child=(Type=Bone,Name="dyn_Cloth_volume_r_A")
    Filter=(TranslationFilter=(bX=false,bY=false,bZ=false),RotationFilter=(bX=True,bY=True,bZ=True),ScaleFilter=(bX=True,bY=True,bZ=True))
    Parents=((Item=(Type=Bone,Name="thigh_twist_01_r"),Weight=0.6),(Item=(Type=Bone,Name="pelvis"),Weight=0.4))
    Weight=@n12.OutputParam
   )
   > &RigUnit_ParentConstraint(
    AdvancedSettings=(RotationOrderForFilter=XZY)
    Child=(Type=Bone,Name="dyn_Cloth_volume_r_B")
    Parents=((Item=(Type=Bone,Name="thigh_twist_01_r"),Weight=0.25),(Item=(Type=Bone,Name="pelvis"),Weight=0.75))
    Weight=@n12.OutputParam
   )
   > &RigUnit_ModifyTransforms(
    ItemToModify="((Item=(Type=Bone,Name=\"dyn_Cloth_volume_r_A\"),Transform=(Rotation=(X=0,Y=0,Z=0,W=1),Translation=(X=4,Y=-2,Z=-2),Scale3D=(X=1,Y=1,Z=1))),(Item=(Type=Bone,Name=\"dyn_Cloth_volume_r_B\"),Transform=(Rotation=(X=0,Y=0,Z=0,W=1),Translation=(X=0,Y=0,Z=0),Scale3D=( ...(truncated)"
    ItemToModify.1.Transform.Rotation=&RigVMFunction_MathQuaternionFromEuler(Euler=(X=-20,Y=20,Z=20)).Result
    ItemToModify.2.Transform.Rotation=&RigVMFunction_MathQuaternionFromEuler(Euler=(X=0,Y=0,Z=50)).Result
    ItemToModify.0.Transform.Rotation=&RigVMFunction_MathQuaternionFromEuler(Euler=(X=12,Y=0,Z=40)).Result
    Weight=@n12.OutputParam
   )
   > n13 = &RigUnit_SphericalPoseReader(
    ActiveRegionScaleFactors=(PositiveWidth=5.5,NegativeWidth=4.5,PositiveHeight=2,NegativeHeight=2)
    ActiveRegionSize=0.1
    Debug=(bDrawDebug=false,bDraw2D=false,DebugScale=25,DebugSegments=20,DebugThickness=0.25)
    DriverAxis=(X=1,Y=0,Z=0)
    DriverItem=(Type=Bone,Name="thigh_r")
    FalloffRegionScaleFactors=(PositiveWidth=1,NegativeWidth=1,PositiveHeight=1,NegativeHeight=1)
    FalloffSize=0.7
    FlipHeightScaling=false
    FlipWidthScaling=true
    OptionalParentItem=(Type=Bone,Name="None")
    RotationOffset=(X=0,Y=50,Z=0)
   )
   > &RigUnit_ParentConstraint(
    AdvancedSettings=(RotationOrderForFilter=XZY)
    Child=(Type=Bone,Name="dyn_Cloth_volume_r_B")
    Filter=(TranslationFilter=(bX=false,bY=false,bZ=false),RotationFilter=(bX=True,bY=True,bZ=True),ScaleFilter=(bX=True,bY=True,bZ=True))
    Parents=((Item=(Type=Bone,Name="thigh_twist_01_r"),Weight=0.8),(Item=(Type=Bone,Name="pelvis"),Weight=0.2))
    Weight=@n13.OutputParam
   )
   > &RigUnit_ParentConstraint(
    AdvancedSettings=(RotationOrderForFilter=XZY)
    Child=(Type=Bone,Name="dyn_Cloth_volume_r_A")
    Filter=(TranslationFilter=(bX=false,bY=false,bZ=false),RotationFilter=(bX=True,bY=True,bZ=True),ScaleFilter=(bX=True,bY=True,bZ=True))
    Parents=((Item=(Type=Bone,Name="thigh_twist_01_r"),Weight=0.8),(Item=(Type=Bone,Name="pelvis"),Weight=0.2))
    Weight=@n13.OutputParam
   )
   > &RigUnit_ModifyTransforms(
    ItemToModify="((Item=(Type=Bone,Name=\"dyn_Cloth_volume_r_B\"),Transform=(Rotation=(X=0,Y=0,Z=0,W=1),Translation=(X=5,Y=-5,Z=0),Scale3D=(X=1,Y=1,Z=1))),(Item=(Type=Bone,Name=\"dyn_Cloth_volume_r_A\"),Transform=(Rotation=(X=0,Y=0,Z=0,W=1),Translation=(X=0,Y=0,Z=0),Scale3D=(X ...(truncated)"
    ItemToModify.2.Transform.Rotation=&RigVMFunction_MathQuaternionFromEuler(Euler=(X=0,Y=20,Z=10)).Result
    ItemToModify.1.Transform.Rotation=&RigVMFunction_MathQuaternionFromEuler(Euler=(X=10,Y=-10,Z=15)).Result
    ItemToModify.0.Transform.Rotation=&RigVMFunction_MathQuaternionFromEuler(Euler=(X=0,Y=-10,Z=30)).Result
    Weight=@n13.OutputParam
   )
   > n14 = &RigUnit_SphericalPoseReader(
    ActiveRegionScaleFactors=(PositiveWidth=3.5,NegativeWidth=0.1,PositiveHeight=0.5,NegativeHeight=2.3)
    ActiveRegionSize=0.1
    Debug=(bDrawDebug=false,DebugScale=25,DebugSegments=20,DebugThickness=0.25)
    DriverAxis=(X=1,Y=0,Z=0)
    DriverItem=(Type=Bone,Name="thigh_r")
    FalloffRegionScaleFactors=(PositiveWidth=1,NegativeWidth=1.6,PositiveHeight=1,NegativeHeight=3)
    FalloffSize=0.3
    FlipHeightScaling=false
    FlipWidthScaling=true
    OptionalParentItem=(Type=Bone,Name="pelvis")
    RotationOffset=(X=90,Y=0,Z=-5)
   )
   > &RigUnit_ParentConstraint(
    AdvancedSettings=(RotationOrderForFilter=XZY)
    Child=(Type=Bone,Name="dyn_Cloth_volume_r_C")
    Filter=(TranslationFilter=(bX=false,bY=false,bZ=false),RotationFilter=(bX=True,bY=True,bZ=True),ScaleFilter=(bX=True,bY=True,bZ=True))
    Parents=((Item=(Type=Bone,Name="thigh_twist_01_r"),Weight=0.5),(Item=(Type=Bone,Name="pelvis"),Weight=0.5))
    Weight=@n14.OutputParam
   )
   > &RigUnit_ModifyTransforms(
    ItemToModify="((Item=(Type=Bone,Name=\"dyn_Cloth_volume_r_C\"),Transform=(Rotation=(X=0,Y=0,Z=0,W=1),Translation=(X=0,Y=0,Z=0),Scale3D=(X=1,Y=1,Z=1))),(Item=(Type=Bone,Name=\"dyn_Cloth_volume_r_B\"),Transform=(Rotation=(X=0,Y=0,Z=0,W=1),Translation=(X=0,Y=0,Z=0),Scale3D=(X= ...(truncated)"
    ItemToModify.1.Transform.Rotation=&RigVMFunction_MathQuaternionFromEuler(Euler=(X=30,Y=20,Z=0)).Result
    ItemToModify.2.Transform.Rotation=&RigVMFunction_MathQuaternionFromEuler(Euler=(X=0,Y=-20,Z=60)).Result
    ItemToModify.0.Transform.Rotation=&RigVMFunction_MathQuaternionFromEuler(Euler=(X=-20,Y=0,Z=50)).Result
    Weight=@n14.OutputParam
   )
   > n15 = &RigUnit_SphericalPoseReader(
    ActiveRegionScaleFactors=(PositiveWidth=3.5,NegativeWidth=0.1,PositiveHeight=0.4,NegativeHeight=0.4)
    ActiveRegionSize=0.1
    Debug=(bDrawDebug=false,DebugScale=25,DebugSegments=20,DebugThickness=0.25)
    DriverAxis=(X=1,Y=0,Z=0)
    DriverItem=(Type=Bone,Name="thigh_r")
    FalloffRegionScaleFactors=(PositiveWidth=1,NegativeWidth=1,PositiveHeight=3,NegativeHeight=1)
    FalloffSize=0.65
    FlipHeightScaling=false
    FlipWidthScaling=true
    OptionalParentItem=(Type=Bone,Name="pelvis")
    RotationOffset=(X=-150,Y=0,Z=-10)
   )
   > &RigUnit_ParentConstraint(
    AdvancedSettings=(RotationOrderForFilter=XZY)
    Child=(Type=Bone,Name="dyn_Cloth_volume_r_B")
    Filter=(TranslationFilter=(bX=false,bY=false,bZ=false),RotationFilter=(bX=True,bY=True,bZ=True),ScaleFilter=(bX=True,bY=True,bZ=True))
    Parents=((Item=(Type=Bone,Name="thigh_twist_01_r"),Weight=0.5),(Item=(Type=Bone,Name="pelvis"),Weight=0.5))
    Weight=@n15.OutputParam
   )
   > &RigUnit_ParentConstraint(
    AdvancedSettings=(RotationOrderForFilter=XZY)
    Child=(Type=Bone,Name="dyn_Cloth_volume_r_A")
    Filter=(TranslationFilter=(bX=false,bY=false,bZ=false),RotationFilter=(bX=True,bY=True,bZ=True),ScaleFilter=(bX=True,bY=True,bZ=True))
    Parents=((Item=(Type=Bone,Name="thigh_twist_01_r"),Weight=0.8),(Item=(Type=Bone,Name="pelvis"),Weight=0.2))
    Weight=0
   )
   > n16 = &RigUnit_SphericalPoseReader(
    ActiveRegionScaleFactors=(PositiveWidth=1,NegativeWidth=1,PositiveHeight=1,NegativeHeight=1)
    ActiveRegionSize=0.1
    Debug=(bDrawDebug=false,bDraw2D=false,DebugScale=25,DebugSegments=20,DebugThickness=0.25)
    DriverAxis=(X=1,Y=0,Z=0)
    DriverItem=(Type=Bone,Name="thigh_r")
    FalloffRegionScaleFactors=(PositiveWidth=1,NegativeWidth=1,PositiveHeight=1,NegativeHeight=1)
    FalloffSize=0.28
    FlipHeightScaling=false
    FlipWidthScaling=true
    OptionalParentItem=(Type=Bone,Name="None")
    RotationOffset=(X=-90,Y=0,Z=-80)
   )
   > &RigUnit_ParentConstraint(
    AdvancedSettings=(RotationOrderForFilter=XZY)
    Child=(Type=Bone,Name="dyn_Cloth_volume_r_B")
    Filter=(TranslationFilter=(bX=true,bY=true,bZ=true),RotationFilter=(bX=True,bY=True,bZ=True),ScaleFilter=(bX=True,bY=True,bZ=True))
    Parents=((Item=(Type=Bone,Name="thigh_twist_01_r"),Weight=0.5),(Item=(Type=Bone,Name="pelvis"),Weight=0.5))
    Weight=@n16.OutputParam
   )
   > &RigUnit_ParentConstraint(
    AdvancedSettings=(RotationOrderForFilter=XZY)
    Child=(Type=Bone,Name="dyn_Cloth_volume_r_A")
    Filter=(TranslationFilter=(bX=true,bY=true,bZ=true),RotationFilter=(bX=True,bY=True,bZ=True),ScaleFilter=(bX=True,bY=True,bZ=True))
    Parents=((Item=(Type=Bone,Name="thigh_twist_01_r"),Weight=0.8),(Item=(Type=Bone,Name="pelvis"),Weight=0.2))
    Weight=@n16.OutputParam
   )
   n17 = &RigVMFunction_MathQuaternionFromEuler(Euler=(X=0,Y=0,Z=50))
   > &RigUnit_ModifyTransforms(
    ItemToModify="((Item=(Type=Bone,Name=\"dyn_Cloth_volume_r_B\"),Transform=(Rotation=(X=0,Y=0,Z=0,W=1),Translation=(X=0,Y=0,Z=0),Scale3D=(X=1,Y=1,Z=1))),(Item=(Type=Bone,Name=\"dyn_Cloth_volume_r_A\"),Transform=(Rotation=(X=0,Y=0,Z=0,W=1),Translation=(X=0,Y=0,Z=0),Scale3D=(X= ...(truncated)"
    ItemToModify.0.Transform.Rotation=@n17.Result
    ItemToModify.1.Transform.Rotation=@n17.Result
    Weight=@n16.OutputParam
   )
   > n18 = &RigUnit_SphericalPoseReader(
    ActiveRegionScaleFactors=(PositiveWidth=3,NegativeWidth=0.1,PositiveHeight=0.3,NegativeHeight=0.3)
    ActiveRegionSize=0.1
    Debug=(bDrawDebug=false,DebugScale=25,DebugSegments=20,DebugThickness=0.25)
    DriverAxis=(X=1,Y=0,Z=0)
    DriverItem=(Type=Bone,Name="pelvis")
    FalloffRegionScaleFactors=(PositiveWidth=1,NegativeWidth=1,PositiveHeight=1,NegativeHeight=1)
    FalloffSize=0.45
    FlipWidthScaling=false
    OptionalParentItem=(Type=Bone,Name="root")
    RotationOffset=(X=-90,Y=0,Z=-10)
   )
   > n19 = &RigUnit_SphericalPoseReader(
    ActiveRegionScaleFactors=(PositiveWidth=3.5,NegativeWidth=0.1,PositiveHeight=3,NegativeHeight=3)
    ActiveRegionSize=0.1
    Debug=(bDrawDebug=false,DebugScale=25,DebugSegments=20,DebugThickness=0.25)
    DriverAxis=(X=1,Y=0.05,Z=0)
    DriverItem=(Type=Bone,Name="calf_r")
    FalloffRegionScaleFactors=(PositiveWidth=1,NegativeWidth=1,PositiveHeight=2,NegativeHeight=2)
    FalloffSize=0.35
    FlipHeightScaling=false
    FlipWidthScaling=true
    OptionalParentItem=(Type=Bone,Name="thigh_r")
    RotationOffset=(X=90,Y=0,Z=70)
   )
   n20 = &RigVMFunction_MathQuaternionFromEuler(Euler=(X=0,Y=10,Z=30))
   > &RigUnit_ModifyTransforms(
    ItemToModify="((Item=(Type=Bone,Name=\"dyn_Cloth_volume_r_C\"),Transform=(Rotation=(X=0,Y=0,Z=0,W=1),Translation=(X=1,Y=-1,Z=0),Scale3D=(X=1,Y=1,Z=1))),(Item=(Type=Bone,Name=\"dyn_Cloth_volume_l_C\"),Transform=(Rotation=(X=0,Y=0,Z=0,W=1),Translation=(X=1,Y=-1,Z=0),Scale3D=( ...(truncated)"
    ItemToModify.0.Transform.Rotation=@n20.Result
    ItemToModify.1.Transform.Rotation=@n20.Result
    Weight=@n19.OutputParam
   )
   n21 = &RigUnit_SphericalPoseReader(
    ActiveRegionScaleFactors=(PositiveWidth=1,NegativeWidth=1,PositiveHeight=1,NegativeHeight=1)
    ActiveRegionSize=0.5
    Debug=(bDrawDebug=false,bDraw2D=false,bDrawLocalAxes=false,DebugScale=25,DebugSegments=20,DebugThickness=0.25)
    DriverAxis=(X=-1,Y=0,Z=0)
    DriverItem=(Type=Bone,Name="upperarm_l")
    FalloffRegionScaleFactors=(PositiveWidth=1,NegativeWidth=1,PositiveHeight=1,NegativeHeight=1)
    FalloffSize=0.5
    FlipHeightScaling=false
    FlipWidthScaling=true
    OptionalParentItem=(Type=Bone,Name="spine_04")
    RotationOffset=(X=-150,Y=0,Z=-90)
    ExecutePin=@SequenceAggregate.G
   )
   n22 = &RigVMFunction_AlphaInterp(
    InterpSpeedDecreasing=10
    InterpSpeedIncreasing=10
    Scale=1
    bClampResult=true
    bInterpResult=false
    bMapRange=true
    Value=@n21.OutputParam
    OutRange.Max=RigVMVariableNode() {variable=Skydive}.Value
   )
   > &RigUnit_ModifyTransforms(
    ItemToModify=((Item=(Type=Bone,Name="dyn_l_shoulder"),Transform=(Rotation=(X=0,Y=0,Z=0,W=1),Translation=(X=0.7,Y=0,Z=0),Scale3D=(X=1,Y=1,Z=1))))
    Weight=@n22.Result
    ItemToModify.0.Transform.Rotation=&RigVMFunction_MathQuaternionFromEuler(Euler=(X=-35,Y=22,Z=0)).Result
   )
   > n23 = &RigUnit_SphericalPoseReader(
    ActiveRegionScaleFactors=(PositiveWidth=1,NegativeWidth=1,PositiveHeight=1,NegativeHeight=1)
    ActiveRegionSize=0.5
    Debug=(bDrawDebug=false,bDraw2D=false,bDrawLocalAxes=false,DebugScale=25,DebugSegments=20,DebugThickness=0.25)
    DriverAxis=(X=-1,Y=0,Z=0)
    DriverItem=(Type=Bone,Name="upperarm_r")
    FalloffRegionScaleFactors=(PositiveWidth=1,NegativeWidth=1,PositiveHeight=1,NegativeHeight=1)
    FalloffSize=0.5
    FlipHeightScaling=false
    FlipWidthScaling=true
    OptionalParentItem=(Type=Bone,Name="spine_04")
    RotationOffset=(X=-150,Y=0,Z=-90)
   )
   n24 = &RigVMFunction_AlphaInterp(
    InterpSpeedDecreasing=10
    InterpSpeedIncreasing=10
    Scale=1
    bClampResult=true
    bInterpResult=false
    bMapRange=true
    Value=@n23.OutputParam
    OutRange.Max=RigVMVariableNode() {variable=Skydive}.Value
   )
   > &RigUnit_ModifyTransforms(
    ItemToModify=((Item=(Type=Bone,Name="dyn_r_shoulder"),Transform=(Rotation=(X=0,Y=0,Z=0,W=1),Translation=(X=-0.7,Y=0,Z=-0.5),Scale3D=(X=1,Y=1,Z=1))))
    Weight=@n24.Result
    ItemToModify.0.Transform.Rotation=&RigVMFunction_MathQuaternionFromEuler(Euler=(X=-10,Y=18,Z=0)).Result
   )
   > n25 = &RigUnit_SphericalPoseReader(
    ActiveRegionScaleFactors=(PositiveWidth=1,NegativeWidth=1,PositiveHeight=1,NegativeHeight=1)
    ActiveRegionSize=0.5
    Debug=(bDrawDebug=false,bDraw2D=false,bDrawLocalAxes=false,DebugScale=25,DebugSegments=20,DebugThickness=0.25)
    DriverAxis=(X=-1,Y=0,Z=0)
    DriverItem=(Type=Bone,Name="upperarm_l")
    FalloffRegionScaleFactors=(PositiveWidth=1,NegativeWidth=1,PositiveHeight=1,NegativeHeight=1)
    FalloffSize=0.5
    FlipHeightScaling=false
    FlipWidthScaling=true
    OptionalParentItem=(Type=Bone,Name="spine_04")
    RotationOffset=(X=193,Y=0,Z=-100)
   )
   n26 = &RigVMFunction_AlphaInterp(
    InterpSpeedDecreasing=10
    InterpSpeedIncreasing=10
    Scale=1
    bClampResult=true
    bInterpResult=false
    bMapRange=true
    Value=@n25.OutputParam
    OutRange.Max=RigVMVariableNode() {variable=ParachuteAfterSkydive}.Value
   )
   > &RigUnit_ModifyTransforms(
    ItemToModify=((Item=(Type=Bone,Name="dyn_l_shoulder"),Transform=(Rotation=(X=0,Y=0,Z=0,W=1),Translation=(X=0,Y=0,Z=0),Scale3D=(X=1,Y=1,Z=1))))
    Weight=@n26.Result
    ItemToModify.0.Transform.Rotation=&RigVMFunction_MathQuaternionFromEuler(Euler=(X=0,Y=35,Z=0)).Result
   )
   > n27 = &RigUnit_SphericalPoseReader(
    ActiveRegionScaleFactors=(PositiveWidth=1,NegativeWidth=1,PositiveHeight=1,NegativeHeight=1)
    ActiveRegionSize=0.5
    Debug=(bDrawDebug=false,bDraw2D=false,bDrawLocalAxes=false,DebugScale=25,DebugSegments=20,DebugThickness=0.25)
    DriverAxis=(X=-1,Y=0,Z=0)
    DriverItem=(Type=Bone,Name="upperarm_r")
    FalloffRegionScaleFactors=(PositiveWidth=1,NegativeWidth=1,PositiveHeight=1,NegativeHeight=1)
    FalloffSize=0.5
    FlipHeightScaling=false
    FlipWidthScaling=true
    OptionalParentItem=(Type=Bone,Name="spine_04")
    RotationOffset=(X=-179,Y=0,Z=-90)
   )
   n28 = &RigVMFunction_AlphaInterp(
    InterpSpeedDecreasing=10
    InterpSpeedIncreasing=10
    Scale=1
    bClampResult=true
    bInterpResult=false
    bMapRange=true
    Value=@n27.OutputParam
    OutRange.Max=RigVMVariableNode() {variable=ParachuteAfterSkydive}.Value
   )
   > &RigUnit_ModifyTransforms(
    ItemToModify=((Item=(Type=Bone,Name="dyn_r_shoulder"),Transform=(Rotation=(X=0,Y=0,Z=0,W=1),Translation=(X=0,Y=0,Z=0),Scale3D=(X=1,Y=1,Z=1))))
    Weight=@n28.Result
    ItemToModify.0.Transform.Rotation=&RigVMFunction_MathQuaternionFromEuler(Euler=(X=0,Y=35,Z=-13)).Result
   )
   > n29 = &RigUnit_SphericalPoseReader(
    ActiveRegionScaleFactors=(PositiveWidth=1,NegativeWidth=1,PositiveHeight=1,NegativeHeight=1)
    ActiveRegionSize=0.03
    Debug=(bDrawDebug=true,bDraw2D=false,bDrawLocalAxes=false,DebugScale=25,DebugSegments=20,DebugThickness=0.25)
    DriverAxis=(X=-1,Y=0,Z=0)
    DriverItem=(Type=Bone,Name="upperarm_l")
    FalloffRegionScaleFactors=(PositiveWidth=0.2,NegativeWidth=0.2,PositiveHeight=0.2,NegativeHeight=0.2)
    FalloffSize=0.2
    FlipHeightScaling=false
    FlipWidthScaling=true
    OptionalParentItem=(Type=Bone,Name="spine_04")
    RotationOffset=(X=0,Y=87,Z=-17)
   )
   n30 = RigVMVariableNode() {variable=FroneEnd}
   n31 = &RigVMFunction_AlphaInterp(
    InterpSpeedDecreasing=10
    InterpSpeedIncreasing=10
    Scale=1
    bClampResult=true
    bInterpResult=false
    bMapRange=true
    Value=@n29.OutputParam
    OutRange.Max=@n30.Value
   )
   > &RigUnit_ModifyTransforms(
    ItemToModify=((Item=(Type=Bone,Name="dyn_l_shoulder"),Transform=(Rotation=(X=0,Y=0,Z=0,W=1),Translation=(X=0,Y=0,Z=0),Scale3D=(X=1,Y=1,Z=1))))
    Weight=@n31.Result
    ItemToModify.0.Transform.Rotation=&RigVMFunction_MathQuaternionFromEuler(Euler=(X=35,Y=0,Z=0)).Result
   )
   > n32 = &RigUnit_SphericalPoseReader(
    ActiveRegionScaleFactors=(PositiveWidth=1,NegativeWidth=1,PositiveHeight=1,NegativeHeight=1)
    ActiveRegionSize=0.03
    Debug=(bDrawDebug=false,bDraw2D=false,bDrawLocalAxes=false,DebugScale=25,DebugSegments=20,DebugThickness=0.25)
    DriverAxis=(X=-1,Y=0,Z=0)
    DriverItem=(Type=Bone,Name="upperarm_r")
    FalloffRegionScaleFactors=(PositiveWidth=0.2,NegativeWidth=0.2,PositiveHeight=0.2,NegativeHeight=0.2)
    FalloffSize=0.5
    FlipHeightScaling=false
    FlipWidthScaling=true
    OptionalParentItem=(Type=Bone,Name="spine_04")
    RotationOffset=(X=0,Y=90,Z=-30)
   )
   n33 = &RigVMFunction_AlphaInterp(
    InterpSpeedDecreasing=10
    InterpSpeedIncreasing=10
    Scale=1
    bClampResult=true
    bInterpResult=false
    bMapRange=true
    Value=@n32.OutputParam
    OutRange.Max=@n30.Value
   )
   > &RigUnit_ModifyTransforms(
    ItemToModify=((Item=(Type=Bone,Name="dyn_r_shoulder"),Transform=(Rotation=(X=0,Y=0,Z=0,W=1),Translation=(X=0,Y=0,Z=0),Scale3D=(X=1,Y=1,Z=1))))
    Weight=@n33.Result
    ItemToModify.0.Transform.Rotation=&RigVMFunction_MathQuaternionFromEuler(Euler=(X=10,Y=0,Z=0)).Result
   )
   n34 = &RigUnit_SphericalPoseReader(
    ActiveRegionScaleFactors=(PositiveWidth=1,NegativeWidth=1,PositiveHeight=1,NegativeHeight=1)
    ActiveRegionSize=0.08
    Debug=(bDrawDebug=false,bDraw2D=false,bDrawLocalAxes=false,DebugScale=25,DebugSegments=20,DebugThickness=0.25)
    DriverAxis=(X=-1,Y=0,Z=0)
    DriverItem=(Type=Bone,Name="upperarm_l")
    FalloffRegionScaleFactors=(PositiveWidth=1,NegativeWidth=1,PositiveHeight=1,NegativeHeight=1)
    FalloffSize=0.5
    FlipHeightScaling=false
    FlipWidthScaling=true
    OptionalParentItem=(Type=Bone,Name="spine_04")
    RotationOffset=(X=-55,Y=56,Z=0)
    ExecutePin=@SequenceAggregate.F
   )
   n35 = &RigVMFunction_AlphaInterp(
    InterpSpeedDecreasing=10
    InterpSpeedIncreasing=10
    Scale=1
    bClampResult=true
    bInterpResult=false
    bMapRange=true
    Value=@n34.OutputParam
    OutRange.Max=RigVMVariableNode() {variable=Pistol}.Value
   )
   > &RigUnit_ModifyTransforms(
    ItemToModify=((Item=(Type=Bone,Name="dyn_l_shoulder"),Transform=(Rotation=(X=0,Y=0,Z=0,W=1),Translation=(X=0,Y=0,Z=0.5),Scale3D=(X=1,Y=1,Z=1))))
    Weight=@n35.Result
    ItemToModify.0.Transform.Rotation=&RigVMFunction_MathQuaternionFromEuler(Euler=(X=50,Y=0,Z=0)).Result
   )
   > n36 = &RigUnit_SphericalPoseReader(
    ActiveRegionScaleFactors=(PositiveWidth=1,NegativeWidth=1,PositiveHeight=1,NegativeHeight=1)
    ActiveRegionSize=0.04
    Debug=(bDrawDebug=false,bDraw2D=false,bDrawLocalAxes=false,DebugScale=25,DebugSegments=20,DebugThickness=0.25)
    DriverAxis=(X=-1,Y=0,Z=0)
    DriverItem=(Type=Bone,Name="upperarm_r")
    FalloffRegionScaleFactors=(PositiveWidth=1,NegativeWidth=1,PositiveHeight=1,NegativeHeight=1)
    FalloffSize=0.5
    FlipHeightScaling=false
    FlipWidthScaling=true
    OptionalParentItem=(Type=Bone,Name="spine_04")
    RotationOffset=(X=60,Y=-10,Z=140)
   )
   n37 = &RigVMFunction_AlphaInterp(
    InterpSpeedDecreasing=10
    InterpSpeedIncreasing=10
    Scale=1
    bClampResult=true
    bInterpResult=false
    bMapRange=true
    Value=@n36.OutputParam
    OutRange.Max=RigVMVariableNode() {variable=Pistol}.Value
   )
   > &RigUnit_ModifyTransforms(
    ItemToModify=((Item=(Type=Bone,Name="dyn_r_shoulder"),Transform=(Rotation=(X=0,Y=0,Z=0,W=1),Translation=(X=0,Y=-0.5,Z=0),Scale3D=(X=1,Y=1,Z=1))))
    Weight=@n37.Result
    ItemToModify.0.Transform.Rotation=&RigVMFunction_MathQuaternionFromEuler(Euler=(X=-60,Y=0,Z=0)).Result
   )
   > n38 = &RigUnit_SphericalPoseReader(
    ActiveRegionScaleFactors=(PositiveWidth=1,NegativeWidth=1,PositiveHeight=1,NegativeHeight=1)
    ActiveRegionSize=0.08
    Debug=(bDrawDebug=false,bDraw2D=false,bDrawLocalAxes=false,DebugScale=25,DebugSegments=20,DebugThickness=0.25)
    DriverAxis=(X=-1,Y=0,Z=0)
    DriverItem=(Type=Bone,Name="upperarm_l")
    FalloffRegionScaleFactors=(PositiveWidth=1,NegativeWidth=1,PositiveHeight=1,NegativeHeight=1)
    FalloffSize=0.5
    FlipHeightScaling=false
    FlipWidthScaling=true
    OptionalParentItem=(Type=Bone,Name="spine_04")
    RotationOffset=(X=-3,Y=65,Z=0)
   )
   n39 = &RigVMFunction_AlphaInterp(
    InterpSpeedDecreasing=10
    InterpSpeedIncreasing=10
    Scale=1
    bClampResult=true
    bInterpResult=false
    bMapRange=true
    Value=@n38.OutputParam
    OutRange.Max=RigVMVariableNode() {variable=Pistol}.Value
   )
   > &RigUnit_ModifyTransforms(
    ItemToModify=((Item=(Type=Bone,Name="dyn_l_shoulder"),Transform=(Rotation=(X=0,Y=0,Z=0,W=1),Translation=(X=0,Y=0,Z=0.5),Scale3D=(X=1,Y=1,Z=1))))
    Weight=@n39.Result
    ItemToModify.0.Transform.Rotation=&RigVMFunction_MathQuaternionFromEuler(Euler=(X=10,Y=0,Z=0)).Result
   )
   > n40 = &RigUnit_SphericalPoseReader(
    ActiveRegionScaleFactors=(PositiveWidth=1,NegativeWidth=1,PositiveHeight=1,NegativeHeight=1)
    ActiveRegionSize=0.08
    Debug=(bDrawDebug=false,bDraw2D=false,bDrawLocalAxes=false,DebugScale=25,DebugSegments=20,DebugThickness=0.25)
    DriverAxis=(X=-1,Y=0,Z=0)
    DriverItem=(Type=Bone,Name="upperarm_r")
    FalloffRegionScaleFactors=(PositiveWidth=1,NegativeWidth=1,PositiveHeight=1,NegativeHeight=1)
    FalloffSize=0.5
    FlipHeightScaling=false
    FlipWidthScaling=true
    OptionalParentItem=(Type=Bone,Name="spine_04")
    RotationOffset=(X=60,Y=-10,Z=90)
   )
   n41 = &RigVMFunction_AlphaInterp(
    InterpSpeedDecreasing=10
    InterpSpeedIncreasing=10
    Scale=1
    bClampResult=true
    bInterpResult=false
    bMapRange=true
    Value=@n40.OutputParam
    OutRange.Max=RigVMVariableNode() {variable=Pistol}.Value
   )
   > &RigUnit_ModifyTransforms(
    ItemToModify=((Item=(Type=Bone,Name="dyn_r_shoulder"),Transform=(Rotation=(X=0,Y=0,Z=0,W=1),Translation=(X=0,Y=-0.5,Z=0),Scale3D=(X=1,Y=1,Z=1))))
    Weight=@n41.Result
    ItemToModify.0.Transform.Rotation=&RigVMFunction_MathQuaternionFromEuler(Euler=(X=0,Y=0,Z=0)).Result
   )
   > n42 = &RigUnit_SphericalPoseReader(
    ActiveRegionScaleFactors=(PositiveWidth=1,NegativeWidth=1,PositiveHeight=1,NegativeHeight=1)
    ActiveRegionSize=0.08
    Debug=(bDrawDebug=false,bDraw2D=false,bDrawLocalAxes=false,DebugScale=25,DebugSegments=20,DebugThickness=0.25)
    DriverAxis=(X=-1,Y=0,Z=0)
    DriverItem=(Type=Bone,Name="upperarm_l")
    FalloffRegionScaleFactors=(PositiveWidth=1,NegativeWidth=1,PositiveHeight=1,NegativeHeight=1)
    FalloffSize=0.5
    FlipHeightScaling=false
    FlipWidthScaling=true
    OptionalParentItem=(Type=Bone,Name="spine_04")
    RotationOffset=(X=-7,Y=61,Z=0)
   )
   n43 = &RigVMFunction_AlphaInterp(
    InterpSpeedDecreasing=10
    InterpSpeedIncreasing=10
    Scale=1
    bClampResult=true
    bInterpResult=false
    bMapRange=true
    Value=@n42.OutputParam
    OutRange.Max=RigVMVariableNode() {variable=Rifle}.Value
   )
   > &RigUnit_ModifyTransforms(
    ItemToModify=((Item=(Type=Bone,Name="dyn_l_shoulder"),Transform=(Rotation=(X=0,Y=0,Z=0,W=1),Translation=(X=0,Y=0,Z=0.5),Scale3D=(X=1,Y=1,Z=1))))
    Weight=@n43.Result
    ItemToModify.0.Transform.Rotation=&RigVMFunction_MathQuaternionFromEuler(Euler=(X=-10,Y=0,Z=0)).Result
   )
   > n44 = &RigUnit_SphericalPoseReader(
    ActiveRegionScaleFactors=(PositiveWidth=1,NegativeWidth=1,PositiveHeight=1,NegativeHeight=1)
    ActiveRegionSize=0.04
    Debug=(bDrawDebug=false,bDraw2D=false,bDrawLocalAxes=false,DebugScale=25,DebugSegments=20,DebugThickness=0.25)
    DriverAxis=(X=-1,Y=0,Z=0)
    DriverItem=(Type=Bone,Name="upperarm_r")
    FalloffRegionScaleFactors=(PositiveWidth=1,NegativeWidth=1,PositiveHeight=1,NegativeHeight=1)
    FalloffSize=0.5
    FlipHeightScaling=false
    FlipWidthScaling=true
    OptionalParentItem=(Type=Bone,Name="spine_04")
    RotationOffset=(X=90,Y=-10,Z=117)
   )
   n45 = &RigVMFunction_AlphaInterp(
    InterpSpeedDecreasing=10
    InterpSpeedIncreasing=10
    Scale=1
    bClampResult=true
    bInterpResult=false
    bMapRange=true
    Value=@n44.OutputParam
    OutRange.Max=RigVMVariableNode() {variable=Rifle}.Value
   )
   > &RigUnit_ModifyTransforms(
    ItemToModify=((Item=(Type=Bone,Name="dyn_r_shoulder"),Transform=(Rotation=(X=0,Y=0,Z=0,W=1),Translation=(X=0,Y=-0.5,Z=0),Scale3D=(X=1,Y=1,Z=1))))
    Weight=@n45.Result
    ItemToModify.0.Transform.Rotation=&RigVMFunction_MathQuaternionFromEuler(Euler=(X=-5,Y=5,Z=0)).Result
   )
   n46 = &RigUnit_SphericalPoseReader(
    ActiveRegionScaleFactors=(PositiveWidth=1.5,NegativeWidth=0.6,PositiveHeight=1.5,NegativeHeight=1)
    ActiveRegionSize=0.15
    Debug=(bDrawDebug=false,bDraw2D=false,bDrawLocalAxes=false,DebugScale=25,DebugSegments=20,DebugThickness=0.25)
    DriverAxis=(X=-1,Y=0,Z=0)
    DriverItem=(Type=Bone,Name="upperarm_l")
    FalloffRegionScaleFactors=(PositiveWidth=0.6,NegativeWidth=1,PositiveHeight=1,NegativeHeight=1)
    FalloffSize=0.6
    FlipHeightScaling=false
    FlipWidthScaling=true
    OptionalParentItem=(Type=Bone,Name="spine_04")
    RotationOffset=(X=180,Y=-10,Z=0)
    ExecutePin=@SequenceAggregate.E
   )
   n47 = &RigVMFunction_AlphaInterp(
    InterpSpeedDecreasing=10
    InterpSpeedIncreasing=10
    Scale=1
    bClampResult=true
    bInterpResult=false
    bMapRange=true
    Value=@n46.OutputParam
    OutRange.Max=RigVMVariableNode() {variable=Rifle}.Value
   )
   > &RigUnit_ModifyTransforms(
    ItemToModify=((Item=(Type=Bone,Name="dyn_l_shoulder"),Transform=(Rotation=(X=0,Y=0,Z=0,W=1),Translation=(X=0,Y=0,Z=0.5),Scale3D=(X=1,Y=1,Z=1))))
    Weight=@n47.Result
    ItemToModify.0.Transform.Rotation=&RigVMFunction_MathQuaternionFromEuler(Euler=(X=0,Y=16,Z=10)).Result
   )
   > n48 = &RigUnit_SphericalPoseReader(
    ActiveRegionScaleFactors=(PositiveWidth=1.5,NegativeWidth=0.6,PositiveHeight=1.5,NegativeHeight=1)
    ActiveRegionSize=0.15
    Debug=(bDrawDebug=false,bDraw2D=false,bDrawLocalAxes=false,DebugScale=25,DebugSegments=20,DebugThickness=0.25)
    DriverAxis=(X=-1,Y=0,Z=0)
    DriverItem=(Type=Bone,Name="upperarm_r")
    FalloffRegionScaleFactors=(PositiveWidth=0.6,NegativeWidth=0.6,PositiveHeight=1,NegativeHeight=1)
    FalloffSize=0.45
    FlipHeightScaling=false
    FlipWidthScaling=true
    OptionalParentItem=(Type=Bone,Name="spine_04")
    RotationOffset=(X=-60,Y=25,Z=40)
   )
   AlphaInterp = &RigVMFunction_AlphaInterp(
    InterpSpeedDecreasing=10
    InterpSpeedIncreasing=10
    Scale=1
    bClampResult=true
    bInterpResult=false
    bMapRange=true
    Value=@n48.OutputParam
    OutRange.Max=RigVMVariableNode() {variable=Rifle}.Value
   )
   > &RigUnit_ModifyTransforms(
    ItemToModify=((Item=(Type=Bone,Name="dyn_r_shoulder"),Transform=(Rotation=(X=0,Y=0,Z=0,W=1),Translation=(X=0,Y=0,Z=0),Scale3D=(X=1,Y=1,Z=1))))
    Weight=@AlphaInterp.Result
    ItemToModify.0.Transform.Rotation=&RigVMFunction_MathQuaternionFromEuler(Euler=(X=30,Y=-10,Z=0)).Result
   )
   > n49 = &RigUnit_SphericalPoseReader(
    ActiveRegionScaleFactors=(PositiveWidth=1,NegativeWidth=1,PositiveHeight=1,NegativeHeight=1)
    ActiveRegionSize=0.08
    Debug=(bDrawDebug=false,bDraw2D=false,bDrawLocalAxes=true,DebugScale=25,DebugSegments=20,DebugThickness=0.25)
    DriverAxis=(X=1,Y=0,Z=0)
    DriverItem=(Type=Bone,Name="upperarm_l")
    FalloffRegionScaleFactors=(PositiveWidth=0.5,NegativeWidth=0.5,PositiveHeight=0.5,NegativeHeight=0.5)
    FalloffSize=0.45
    FlipWidthScaling=true
    OptionalParentItem=(Type=Bone,Name="spine_04")
    RotationOffset=(X=0,Y=-130,Z=0)
   )
   n50 = &RigVMFunction_AlphaInterp(
    InterpSpeedDecreasing=10
    InterpSpeedIncreasing=10
    Scale=1
    bClampResult=true
    bInterpResult=false
    bMapRange=true
    Value=@n49.OutputParam
    OutRange.Max=RigVMVariableNode() {variable=Rifle}.Value
   )
   > &RigUnit_ModifyTransforms(
    ItemToModify=((Item=(Type=Bone,Name="dyn_l_shoulder"),Transform=(Rotation=(X=0,Y=0,Z=0,W=1),Translation=(X=0,Y=0,Z=0),Scale3D=(X=1,Y=1,Z=1))))
    Weight=@n50.Result
    ItemToModify.0.Transform.Rotation=&RigVMFunction_MathQuaternionFromEuler(Euler=(X=10,Y=0,Z=0)).Result
   )
   > n51 = &RigUnit_SphericalPoseReader(
    ActiveRegionScaleFactors=(PositiveWidth=1,NegativeWidth=2,PositiveHeight=2,NegativeHeight=1)
    ActiveRegionSize=0.08
    Debug=(bDrawDebug=false,bDraw2D=false,bDrawLocalAxes=true,DebugScale=25,DebugSegments=20,DebugThickness=0.25)
    DriverAxis=(X=1,Y=0,Z=0)
    DriverItem=(Type=Bone,Name="upperarm_r")
    FalloffRegionScaleFactors=(PositiveWidth=1,NegativeWidth=1,PositiveHeight=1,NegativeHeight=1)
    FalloffSize=0.45
    FlipWidthScaling=true
    OptionalParentItem=(Type=Bone,Name="spine_04")
    RotationOffset=(X=0,Y=-90,Z=0)
   )
   n52 = &RigVMFunction_AlphaInterp(
    InterpSpeedDecreasing=10
    InterpSpeedIncreasing=10
    Scale=1
    bClampResult=true
    bInterpResult=false
    bMapRange=true
    Value=@n51.OutputParam
    OutRange.Max=RigVMVariableNode() {variable=Rifle}.Value
   )
   > &RigUnit_ModifyTransforms(
    ItemToModify=((Item=(Type=Bone,Name="dyn_r_shoulder"),Transform=(Rotation=(X=0,Y=0,Z=0,W=1),Translation=(X=0,Y=0,Z=0),Scale3D=(X=1,Y=1,Z=1))))
    Weight=@n52.Result
    ItemToModify.0.Transform.Rotation=&RigVMFunction_MathQuaternionFromEuler(Euler=(X=60,Y=-10,Z=0)).Result
   )
   > n53 = &RigUnit_SphericalPoseReader(
    ActiveRegionScaleFactors=(PositiveWidth=1.5,NegativeWidth=0.6,PositiveHeight=0.4,NegativeHeight=0.2)
    ActiveRegionSize=0.15
    Debug=(bDrawDebug=false,bDraw2D=false,bDrawLocalAxes=false,DebugScale=25,DebugSegments=20,DebugThickness=0.25)
    DriverAxis=(X=-1,Y=0,Z=0)
    DriverItem=(Type=Bone,Name="upperarm_l")
    FalloffRegionScaleFactors=(PositiveWidth=0.5,NegativeWidth=0.5,PositiveHeight=0.2,NegativeHeight=0.5)
    FalloffSize=0.6
    FlipHeightScaling=false
    FlipWidthScaling=true
    OptionalParentItem=(Type=Bone,Name="spine_04")
    RotationOffset=(X=-90,Y=0,Z=-50)
   )
   n54 = &RigVMFunction_AlphaInterp(
    InterpSpeedDecreasing=10
    InterpSpeedIncreasing=10
    Scale=1
    bClampResult=true
    bInterpResult=false
    bMapRange=true
    Value=@n53.OutputParam
    OutRange.Max=RigVMVariableNode() {variable=Rifle}.Value
   )
   > &RigUnit_ModifyTransforms(
    ItemToModify=((Item=(Type=Bone,Name="dyn_l_shoulder"),Transform=(Rotation=(X=0,Y=0,Z=0,W=1),Translation=(X=0,Y=0,Z=0.5),Scale3D=(X=1,Y=1,Z=1))))
    Weight=@n54.Result
    ItemToModify.0.Transform.Rotation=&RigVMFunction_MathQuaternionFromEuler(Euler=(X=-15,Y=0,Z=0)).Result
   )
   > n55 = &RigUnit_SphericalPoseReader(
    ActiveRegionScaleFactors=(PositiveWidth=1,NegativeWidth=0.2,PositiveHeight=1,NegativeHeight=0.2)
    ActiveRegionSize=0.08
    Debug=(bDrawDebug=false,bDraw2D=false,bDrawLocalAxes=false,DebugScale=25,DebugSegments=20,DebugThickness=0.25)
    DriverAxis=(X=-1,Y=0,Z=0)
    DriverItem=(Type=Bone,Name="upperarm_r")
    FalloffRegionScaleFactors=(PositiveWidth=1,NegativeWidth=0.5,PositiveHeight=1,NegativeHeight=0.5)
    FalloffSize=0.2
    FlipHeightScaling=false
    FlipWidthScaling=true
    OptionalParentItem=(Type=Bone,Name="spine_04")
    RotationOffset=(X=-55,Y=25,Z=-20)
   )
   n56 = &RigVMFunction_AlphaInterp(
    InterpSpeedDecreasing=10
    InterpSpeedIncreasing=10
    Scale=1
    bClampResult=true
    bInterpResult=false
    bMapRange=true
    Value=@n55.OutputParam
    OutRange.Max=RigVMVariableNode() {variable=Rifle}.Value
   )
   > &RigUnit_ModifyTransforms(
    ItemToModify=((Item=(Type=Bone,Name="dyn_r_shoulder"),Transform=(Rotation=(X=0,Y=0,Z=0,W=1),Translation=(X=0,Y=0,Z=0),Scale3D=(X=1,Y=1,Z=1))))
    Weight=@n56.Result
    ItemToModify.0.Transform.Rotation=&RigVMFunction_MathQuaternionFromEuler(Euler=(X=20,Y=10,Z=0)).Result
   )
   > n57 = &RigUnit_SphericalPoseReader(
    ActiveRegionScaleFactors=(PositiveWidth=0.5,NegativeWidth=1,PositiveHeight=1,NegativeHeight=0.2)
    ActiveRegionSize=0.08
    Debug=(bDrawDebug=false,bDraw2D=false,bDrawLocalAxes=true,DebugScale=25,DebugSegments=20,DebugThickness=0.25)
    DriverAxis=(X=1,Y=0,Z=0)
    DriverItem=(Type=Bone,Name="upperarm_l")
    FalloffRegionScaleFactors=(PositiveWidth=0.5,NegativeWidth=1,PositiveHeight=1,NegativeHeight=0.2)
    FalloffSize=0.45
    FlipWidthScaling=true
    OptionalParentItem=(Type=Bone,Name="spine_04")
    RotationOffset=(X=40,Y=0,Z=10)
   )
   n58 = &RigVMFunction_AlphaInterp(
    InterpSpeedDecreasing=10
    InterpSpeedIncreasing=10
    Scale=1
    bClampResult=true
    bInterpResult=false
    bMapRange=true
    Value=@n57.OutputParam
    OutRange.Max=RigVMVariableNode() {variable=Pistol}.Value
   )
   > &RigUnit_ModifyTransforms(
    ItemToModify=((Item=(Type=Bone,Name="dyn_l_shoulder"),Transform=(Rotation=(X=0,Y=0,Z=0,W=1),Translation=(X=0,Y=0,Z=0),Scale3D=(X=1,Y=1,Z=1))))
    Weight=@n58.Result
    ItemToModify.0.Transform.Rotation=&RigVMFunction_MathQuaternionFromEuler(Euler=(X=0,Y=5,Z=0)).Result
   )
   > n59 = &RigUnit_SphericalPoseReader(
    ActiveRegionScaleFactors=(PositiveWidth=0.15,NegativeWidth=0.15,PositiveHeight=0.15,NegativeHeight=0.15)
    ActiveRegionSize=0.15
    Debug=(bDrawDebug=false,bDraw2D=false,bDrawLocalAxes=false,DebugScale=25,DebugSegments=20,DebugThickness=0.25)
    DriverAxis=(X=-1,Y=0,Z=0)
    DriverItem=(Type=Bone,Name="upperarm_r")
    FalloffRegionScaleFactors=(PositiveWidth=0.75,NegativeWidth=0.75,PositiveHeight=0.75,NegativeHeight=0.75)
    FalloffSize=1
    FlipHeightScaling=false
    FlipWidthScaling=true
    OptionalParentItem=(Type=Bone,Name="spine_04")
    RotationOffset=(X=-100,Y=0,Z=40)
   )
   n60 = &RigVMFunction_AlphaInterp(
    InterpSpeedDecreasing=10
    InterpSpeedIncreasing=10
    Scale=1
    bClampResult=true
    bInterpResult=false
    bMapRange=true
    Value=@n59.OutputParam
    OutRange.Max=RigVMVariableNode() {variable=Pistol}.Value
   )
   > &RigUnit_ModifyTransforms(
    ItemToModify=((Item=(Type=Bone,Name="dyn_r_shoulder"),Transform=(Rotation=(X=0,Y=0,Z=0,W=1),Translation=(X=0,Y=0,Z=0),Scale3D=(X=1,Y=1,Z=1))))
    Weight=@n60.Result
    ItemToModify.0.Transform.Rotation=&RigVMFunction_MathQuaternionFromEuler(Euler=(X=5,Y=-10,Z=0)).Result
   )
   > n61 = &RigUnit_SphericalPoseReader(
    ActiveRegionScaleFactors=(PositiveWidth=1,NegativeWidth=1,PositiveHeight=1,NegativeHeight=1)
    ActiveRegionSize=0.08
    Debug=(bDrawDebug=false,bDraw2D=false,bDrawLocalAxes=true,DebugScale=25,DebugSegments=20,DebugThickness=0.25)
    DriverAxis=(X=1,Y=0,Z=0)
    DriverItem=(Type=Bone,Name="upperarm_l")
    FalloffRegionScaleFactors=(PositiveWidth=0.5,NegativeWidth=0.5,PositiveHeight=0.5,NegativeHeight=0.5)
    FalloffSize=0.45
    FlipWidthScaling=true
    OptionalParentItem=(Type=Bone,Name="spine_04")
    RotationOffset=(X=0,Y=-120,Z=0)
   )
   n62 = &RigVMFunction_AlphaInterp(
    InterpSpeedDecreasing=10
    InterpSpeedIncreasing=10
    Scale=1
    bClampResult=true
    bInterpResult=false
    bMapRange=true
    Value=@n61.OutputParam
    OutRange.Max=RigVMVariableNode() {variable=Pistol}.Value
   )
   > &RigUnit_ModifyTransforms(
    ItemToModify=((Item=(Type=Bone,Name="dyn_l_shoulder"),Transform=(Rotation=(X=0,Y=0,Z=0,W=1),Translation=(X=0,Y=0,Z=0),Scale3D=(X=1,Y=1,Z=1))))
    Weight=@n62.Result
    ItemToModify.0.Transform.Rotation=&RigVMFunction_MathQuaternionFromEuler(Euler=(X=10,Y=0,Z=0)).Result
   )
   > n63 = &RigUnit_SphericalPoseReader(
    ActiveRegionScaleFactors=(PositiveWidth=1,NegativeWidth=1,PositiveHeight=1,NegativeHeight=1)
    ActiveRegionSize=0.08
    Debug=(bDrawDebug=false,bDraw2D=false,bDrawLocalAxes=true,DebugScale=25,DebugSegments=20,DebugThickness=0.25)
    DriverAxis=(X=1,Y=0,Z=0)
    DriverItem=(Type=Bone,Name="upperarm_r")
    FalloffRegionScaleFactors=(PositiveWidth=1,NegativeWidth=1,PositiveHeight=1,NegativeHeight=1)
    FalloffSize=0.45
    FlipWidthScaling=true
    OptionalParentItem=(Type=Bone,Name="spine_04")
    RotationOffset=(X=15,Y=-150,Z=0)
   )
   n64 = &RigVMFunction_AlphaInterp(
    InterpSpeedDecreasing=10
    InterpSpeedIncreasing=10
    Scale=1
    bClampResult=true
    bInterpResult=false
    bMapRange=true
    Value=@n63.OutputParam
    OutRange.Max=RigVMVariableNode() {variable=Pistol}.Value
   )
   > &RigUnit_ModifyTransforms(
    ItemToModify=((Item=(Type=Bone,Name="dyn_r_shoulder"),Transform=(Rotation=(X=0,Y=0,Z=0,W=1),Translation=(X=0,Y=0,Z=0),Scale3D=(X=1,Y=1,Z=1))))
    Weight=@n64.Result
    ItemToModify.0.Transform.Rotation=&RigVMFunction_MathQuaternionFromEuler(Euler=(X=130,Y=-5,Z=0)).Result
   )
   > n65 = &RigUnit_SphericalPoseReader(
    ActiveRegionScaleFactors=(PositiveWidth=1,NegativeWidth=1,PositiveHeight=1,NegativeHeight=1)
    ActiveRegionSize=0.1
    Debug=(bDrawDebug=false,bDraw2D=false,bDrawLocalAxes=true,DebugScale=25,DebugSegments=20,DebugThickness=0.25)
    DriverAxis=(X=1,Y=0,Z=0)
    DriverItem=(Type=Bone,Name="upperarm_l")
    FalloffRegionScaleFactors=(PositiveWidth=1,NegativeWidth=1,PositiveHeight=1,NegativeHeight=1)
    FalloffSize=0.5
    FlipWidthScaling=true
    OptionalParentItem=(Type=Bone,Name="spine_04")
    RotationOffset=(X=100,Y=0,Z=-14)
   )
   n66 = &RigVMFunction_AlphaInterp(
    InterpSpeedDecreasing=10
    InterpSpeedIncreasing=10
    Scale=1
    bClampResult=true
    bInterpResult=false
    bMapRange=true
    Value=@n65.OutputParam
    OutRange.Max=RigVMVariableNode() {variable=Pistol}.Value
   )
   > &RigUnit_ModifyTransforms(
    ItemToModify=((Item=(Type=Bone,Name="dyn_l_shoulder"),Transform=(Rotation=(X=0,Y=0,Z=0,W=1),Translation=(X=0,Y=0,Z=0),Scale3D=(X=1,Y=1,Z=1))))
    Weight=@n66.Result
    ItemToModify.0.Transform.Rotation=&RigVMFunction_MathQuaternionFromEuler(Euler=(X=-60,Y=4,Z=0)).Result
   )
   > n67 = &RigUnit_SphericalPoseReader(
    ActiveRegionScaleFactors=(PositiveWidth=1,NegativeWidth=1,PositiveHeight=1,NegativeHeight=1)
    ActiveRegionSize=0.1
    Debug=(bDrawDebug=false,bDraw2D=false,bDrawLocalAxes=true,DebugScale=25,DebugSegments=20,DebugThickness=0.25)
    DriverAxis=(X=1,Y=0,Z=0)
    DriverItem=(Type=Bone,Name="upperarm_r")
    FalloffRegionScaleFactors=(PositiveWidth=0.5,NegativeWidth=0.5,PositiveHeight=0.5,NegativeHeight=0.2)
    FalloffSize=0.5
    FlipWidthScaling=true
    OptionalParentItem=(Type=Bone,Name="spine_04")
    RotationOffset=(X=100,Y=0,Z=3)
   )
   n68 = &RigVMFunction_AlphaInterp(
    InterpSpeedDecreasing=10
    InterpSpeedIncreasing=10
    Scale=1
    bClampResult=true
    bInterpResult=false
    bMapRange=true
    Value=@n67.OutputParam
    OutRange.Max=RigVMVariableNode() {variable=Pistol}.Value
   )
   > &RigUnit_ModifyTransforms(
    ItemToModify=((Item=(Type=Bone,Name="dyn_r_shoulder"),Transform=(Rotation=(X=0,Y=0,Z=0,W=1),Translation=(X=0,Y=0,Z=-1),Scale3D=(X=1,Y=1,Z=1))))
    Weight=@n68.Result
    ItemToModify.0.Transform.Rotation=&RigVMFunction_MathQuaternionFromEuler(Euler=(X=30,Y=4,Z=0)).Result
   )
   &RigUnit_ParentConstraint(
    AdvancedSettings=(RotationOrderForFilter=XZY)
    Child=(Type=Bone,Name="dyn_Cloth_volume_fn")
    Parents=((Item=(Type=Bone,Name="dyn_Cloth_volume_l_A"),Weight=0.5),(Item=(Type=Bone,Name="dyn_Cloth_volume_r_A"),Weight=0.5))
    Weight=1
    ExecutePin=@SequenceAggregate.D
   )
   > &RigUnit_ParentConstraint(
    AdvancedSettings=(RotationOrderForFilter=XZY)
    Child=(Type=Bone,Name="dyn_Cloth_volume_fn01")
    Parents=((Item=(Type=Bone,Name="dyn_Cloth_volume_l_A01"),Weight=0.5),(Item=(Type=Bone,Name="dyn_Cloth_volume_r_A01"),Weight=0.5))
    Weight=1
   )
   > &RigUnit_ParentConstraint(
    AdvancedSettings=(RotationOrderForFilter=XZY)
    Child=(Type=Bone,Name="dyn_Cloth_volume_fn02")
    Parents=((Item=(Type=Bone,Name="dyn_Cloth_volume_l_A02"),Weight=0.5),(Item=(Type=Bone,Name="dyn_Cloth_volume_r_A02"),Weight=0.5))
    Weight=1
   )
   > &RigUnit_ParentConstraint(
    AdvancedSettings=(RotationOrderForFilter=XZY)
    Child=(Type=Bone,Name="dyn_Cloth_volume_bk")
    Parents=((Item=(Type=Bone,Name="dyn_Cloth_volume_l_C"),Weight=0.5),(Item=(Type=Bone,Name="dyn_Cloth_volume_r_C"),Weight=0.5))
    Weight=1
   )
   > &RigUnit_ParentConstraint(
    AdvancedSettings=(RotationOrderForFilter=XZY)
    Child=(Type=Bone,Name="dyn_Cloth_volume_bk01")
    Parents=((Item=(Type=Bone,Name="dyn_Cloth_volume_l_C01"),Weight=0.5),(Item=(Type=Bone,Name="dyn_Cloth_volume_r_C01"),Weight=0.5))
    Weight=1
   )
   > &RigUnit_ParentConstraint(
    AdvancedSettings=(RotationOrderForFilter=XZY)
    Child=(Type=Bone,Name="dyn_Cloth_volume_bk02")
    Parents=((Item=(Type=Bone,Name="dyn_Cloth_volume_l_C02"),Weight=0.5),(Item=(Type=Bone,Name="dyn_Cloth_volume_r_C02"),Weight=0.5))
    Weight=1
   )
   > n69 = &RigUnit_SphericalPoseReader(
    ActiveRegionScaleFactors=(PositiveWidth=3.5,NegativeWidth=1,PositiveHeight=3,NegativeHeight=3)
    ActiveRegionSize=0.1
    Debug=(bDrawDebug=false,DebugScale=25,DebugSegments=20,DebugThickness=0.25)
    DriverAxis=(X=-1,Y=0,Z=0)
    DriverItem=(Type=Bone,Name="thigh_l")
    FalloffRegionScaleFactors=(PositiveWidth=1,NegativeWidth=2,PositiveHeight=3,NegativeHeight=1)
    FalloffSize=0.35
    OptionalParentItem=(Type=Bone)
    RotationOffset=(X=90,Y=0,Z=-35)
   )
   > &RigUnit_ParentConstraint(
    AdvancedSettings=(RotationOrderForFilter=XZY)
    Child=(Type=Bone,Name="dyn_Cloth_volume_fn")
    Filter=(TranslationFilter=(bX=false,bY=false,bZ=false),RotationFilter=(bX=True,bY=True,bZ=True),ScaleFilter=(bX=True,bY=True,bZ=True))
    Parents=((Item=(Type=Bone,Name="thigh_twist_01_l"),Weight=0.5),(Item=(Type=Bone,Name="pelvis"),Weight=0.5))
    Weight=@n69.OutputParam
   )
   > &RigUnit_ModifyTransforms(
    ItemToModify="((Item=(Type=Bone,Name=\"dyn_Cloth_volume_fn01\"),Transform=(Rotation=(X=0,Y=0,Z=0,W=1),Translation=(X=0,Y=0,Z=0),Scale3D=(X=1,Y=1,Z=1))),(Item=(Type=Bone,Name=\"dyn_Cloth_volume_fn_float\"),Transform=(Rotation=(X=0,Y=0,Z=0,W=1),Translation=(X=0,Y=0,Z=0),Scale ...(truncated)"
    ItemToModify.0.Transform.Rotation=&RigVMFunction_MathQuaternionFromEuler(Euler=(X=20,Y=10,Z=5)).Result
    ItemToModify.1.Transform.Rotation=&RigVMFunction_MathQuaternionFromEuler(Euler=(X=20,Y=0,Z=-10)).Result
    Weight=@n69.OutputParam
   )
   > n70 = &RigUnit_SphericalPoseReader(
    ActiveRegionScaleFactors=(PositiveWidth=3.5,NegativeWidth=0.1,PositiveHeight=3,NegativeHeight=3)
    ActiveRegionSize=0.1
    Debug=(bDrawDebug=false,DebugScale=25,DebugSegments=20,DebugThickness=0.25)
    DriverAxis=(X=1,Y=0,Z=0)
    DriverItem=(Type=Bone,Name="thigh_twist_01_r")
    FalloffRegionScaleFactors=(PositiveWidth=1,NegativeWidth=2.3,PositiveHeight=2,NegativeHeight=2)
    FalloffSize=0.35
    FlipHeightScaling=false
    FlipWidthScaling=true
    OptionalParentItem=(Type=Bone,Name="pelvis")
    RotationOffset=(X=-90,Y=0,Z=-25)
   )
   > &RigUnit_ParentConstraint(
    AdvancedSettings=(RotationOrderForFilter=XZY)
    Child=(Type=Bone,Name="dyn_Cloth_volume_fn")
    Filter=(TranslationFilter=(bX=false,bY=false,bZ=false),RotationFilter=(bX=True,bY=True,bZ=True),ScaleFilter=(bX=True,bY=True,bZ=True))
    Parents=((Item=(Type=Bone,Name="thigh_twist_01_r"),Weight=0.5),(Item=(Type=Bone,Name="pelvis"),Weight=0.5))
    Weight=@n70.OutputParam
   )
   > &RigUnit_ModifyTransforms(
    ItemToModify="((Item=(Type=Bone,Name=\"dyn_Cloth_volume_fn\"),Transform=(Rotation=(X=0,Y=0,Z=0,W=1),Translation=(X=0,Y=0,Z=0),Scale3D=(X=1,Y=1,Z=1))),(Item=(Type=Bone,Name=\"dyn_Cloth_volume_fn_float\"),Transform=(Rotation=(X=0,Y=0,Z=0,W=1),Translation=(X=0,Y=0,Z=0),Scale3D ...(truncated)"
    ItemToModify.0.Transform.Rotation=&RigVMFunction_MathQuaternionFromEuler(Euler=(X=-20,Y=-10,Z=10)).Result
    ItemToModify.1.Transform.Rotation=&RigVMFunction_MathQuaternionFromEuler(Euler=(X=-20,Y=0,Z=-10)).Result
    ItemToModify.2.Transform.Rotation=&RigVMFunction_MathQuaternionFromEuler(Euler=(X=0,Y=20,Z=-10)).Result
    Weight=@n70.OutputParam
   )
   > n71 = &RigUnit_SphericalPoseReader(
    ActiveRegionScaleFactors=(PositiveWidth=3.5,NegativeWidth=0.1,PositiveHeight=3,NegativeHeight=1)
    ActiveRegionSize=0.2
    Debug=(bDrawDebug=false,DebugScale=25,DebugSegments=20,DebugThickness=0.25)
    DriverAxis=(X=-1,Y=0,Z=0)
    DriverItem=(Type=Bone,Name="thigh_l")
    FalloffRegionScaleFactors=(PositiveWidth=1,NegativeWidth=1,PositiveHeight=1,NegativeHeight=5)
    FalloffSize=0.25
    FlipWidthScaling=false
    OptionalParentItem=(Type=Bone,Name="None")
    RotationOffset=(X=150,Y=100,Z=0)
   )
   > &RigUnit_ParentConstraint(
    AdvancedSettings=(RotationOrderForFilter=XZY)
    Child=(Type=Bone,Name="dyn_Cloth_volume_fn")
    Filter=(TranslationFilter=(bX=false,bY=false,bZ=false),RotationFilter=(bX=True,bY=True,bZ=True),ScaleFilter=(bX=True,bY=True,bZ=True))
    Parents=((Item=(Type=Bone,Name="thigh_twist_01_l"),Weight=0.6),(Item=(Type=Bone,Name="pelvis"),Weight=0.4))
    Weight=@n71.OutputParam
   )
   > &RigUnit_ModifyTransforms(
    ItemToModify="((Item=(Type=Bone,Name=\"dyn_Cloth_volume_fn01\"),Transform=(Rotation=(X=0,Y=0,Z=0,W=1),Translation=(X=0,Y=0,Z=0),Scale3D=(X=1,Y=1,Z=1))),(Item=(Type=Bone,Name=\"None\"),Transform=(Rotation=(X=0,Y=0,Z=0,W=1),Translation=(X=2,Y=-2,Z=0),Scale3D=(X=1,Y=1,Z=1))),( ...(truncated)"
    ItemToModify.0.Transform.Rotation=&RigVMFunction_MathQuaternionFromEuler(Euler=(X=20,Y=20,Z=20)).Result
    Weight=@n71.OutputParam
   )
   > n72 = &RigUnit_SphericalPoseReader(
    ActiveRegionScaleFactors=(PositiveWidth=3.5,NegativeWidth=0.1,PositiveHeight=3,NegativeHeight=1)
    ActiveRegionSize=0.2
    Debug=(bDrawDebug=false,DebugScale=25,DebugSegments=20,DebugThickness=0.25)
    DriverAxis=(X=1,Y=0,Z=0)
    DriverItem=(Type=Bone,Name="thigh_r")
    FalloffRegionScaleFactors=(PositiveWidth=1,NegativeWidth=1,PositiveHeight=1,NegativeHeight=5)
    FalloffSize=0.25
    FlipHeightScaling=false
    FlipWidthScaling=true
    OptionalParentItem=(Type=Bone,Name="None")
    RotationOffset=(X=-150,Y=85,Z=-120)
   )
   > &RigUnit_ParentConstraint(
    AdvancedSettings=(RotationOrderForFilter=XZY)
    Child=(Type=Bone,Name="dyn_Cloth_volume_fn")
    Filter=(TranslationFilter=(bX=false,bY=false,bZ=false),RotationFilter=(bX=True,bY=True,bZ=True),ScaleFilter=(bX=True,bY=True,bZ=True))
    Parents=((Item=(Type=Bone,Name="thigh_twist_01_r"),Weight=0.6),(Item=(Type=Bone,Name="pelvis"),Weight=0.4))
    Weight=@n72.OutputParam
   )
   > &RigUnit_ModifyTransforms(
    ItemToModify="((Item=(Type=Bone,Name=\"dyn_Cloth_volume_fn\"),Transform=(Rotation=(X=0,Y=0,Z=0,W=1),Translation=(X=0,Y=0,Z=0),Scale3D=(X=1,Y=1,Z=1))),(Item=(Type=Bone,Name=\"None\"),Transform=(Rotation=(X=0,Y=0,Z=0,W=1),Translation=(X=0,Y=0,Z=0),Scale3D=(X=1,Y=1,Z=1))),(Ite ...(truncated)"
    ItemToModify.0.Transform.Rotation=&RigVMFunction_MathQuaternionFromEuler(Euler=(X=-30,Y=-30,Z=30)).Result
    Weight=@n72.OutputParam
   )
   > &RigUnit_ParentConstraint(
    AdvancedSettings=(RotationOrderForFilter=XZY)
    Child=(Type=Bone,Name="dyn_Cloth_volume_bk01")
    Parents=((Item=(Type=Bone,Name="dyn_Cloth_volume_l_C01"),Weight=0.5),(Item=(Type=Bone,Name="dyn_Cloth_volume_r_C01"),Weight=0.5))
    Weight=1
   )
   > &RigUnit_ParentConstraint(
    AdvancedSettings=(RotationOrderForFilter=XZY)
    Child=(Type=Bone,Name="dyn_Cloth_volume_bk02")
    Parents=((Item=(Type=Bone,Name="dyn_Cloth_volume_l_C02"),Weight=0.5),(Item=(Type=Bone,Name="dyn_Cloth_volume_r_C02"),Weight=0.5))
    Weight=1
   )
   &RigVMFunction_MathQuaternionFromEuler(Euler=(X=20,Y=10,Z=0))
   &RigVMFunction_MathQuaternionFromEuler(Euler=(X=20,Y=10,Z=15))
   n73 = &RigVMFunction_MathQuaternionFromEuler(Euler=(X=0,Y=0,Z=90))
   &RigUnit_ModifyTransforms(
    ExecutePin=()
    ItemToModify="((Item=(Type=Bone,Name=\"dyn_Cloth_volume_r_C\"),Transform=(Rotation=(X=0,Y=0,Z=0,W=1),Translation=(X=0,Y=0,Z=0),Scale3D=(X=1,Y=1,Z=1))),(Item=(Type=Bone,Name=\"dyn_Cloth_volume_l_C\"),Transform=(Rotation=(X=0,Y=0,Z=0,W=1),Translation=(X=0,Y=0,Z=0),Scale3D=(X= ...(truncated)"
    ItemToModify.0.Transform.Rotation=@n73.Result
    ItemToModify.1.Transform.Rotation=@n73.Result
    Weight=@n18.OutputParam
   )
   RigVMRerouteNode(Value=())
  ]}
]
hierarchy [
 root {loc=<0 0 0>, rot=<0 0 0 1>}
 pelvis {parent=root, loc=<0 -0.5301 98.0571>, rot=<0.0127 0.707 -0.0127 -0.707>}
 spine_01 {parent=pelvis, loc=<6.1208 -0.9328 0>, rot=<0 0 0.034 -0.9994>}
 spine_02 {parent=spine_01, loc=<6.5996 -0.2095 0>, rot=<0 0 0.0238 -0.9997>}
 spine_03 {parent=spine_02, loc=<6.4914 -0.1237 0>, rot=<0 0 0.0149 -0.9999>}
 spine_04 {parent=spine_03, loc=<6.9598 0.0794 0>, rot=<0 0 -0.0132 -0.9999>}
 spine_05 {parent=spine_04, loc=<6.9987 0 0>, rot=<0 0 -0.0015 -1>}
 attach_cape {parent=spine_05, loc=<9.017 -11.714 0>, rot=<0.0616 -0.7044 -0.0616 -0.7044>}
 neck_01 {parent=spine_05, loc=<11.3138 0.1529 0>, rot=<0 0 -0.2003 -0.9797>}
 neck_02 {parent=neck_01, loc=<4.5627 0 0>, rot=<0 0 0.0504 -0.9987>}
 head {parent=neck_02, loc=<5.3914 0 0>, rot=<0 0 0.115 -0.9934>}
 FACIAL_C_FacialRoot {parent=head, loc=<2.1524 -0.746 -0.0002>, rot=<0.4879 -0.5118 -0.4879 -0.5118>}
 FACIAL_L_CheekOuter {parent=FACIAL_C_FacialRoot, loc=<3.7404 1.6506 6.2414>, rot=<0.0524 -0.3772 0.0992 -0.9193>}
 FACIAL_R_CheekOuter {parent=FACIAL_C_FacialRoot, loc=<-3.7404 1.6564 6.2409>, rot=<0.0524 0.3772 -0.0992 -0.9193>}
 FACIAL_L_NasolabialBulge {parent=FACIAL_C_FacialRoot, loc=<2.9397 4.8043 6.0424>, rot=<0.0551 -0.3774 0.0227 -0.9241>}
 FACIAL_R_NasolabialBulge {parent=FACIAL_C_FacialRoot, loc=<-2.9426 4.8019 6.0388>, rot=<0.0551 0.3774 -0.0227 -0.9241>}
 FACIAL_L_NasolabialFurrow {parent=FACIAL_C_FacialRoot, loc=<2.6623 3.7224 8.4941>, rot=<-0.036 -0.1247 -0.364 -0.9223>}
 FACIAL_R_NasolabialFurrow {parent=FACIAL_C_FacialRoot, loc=<-2.6647 3.7215 8.4928>, rot=<-0.036 0.1247 0.364 -0.9223>}
 FACIAL_L_CheekLower {parent=FACIAL_C_FacialRoot, loc=<3.3853 4.38 5.3889>, rot=<0.0319 -0.427 0.0565 -0.9019>}
 FACIAL_R_CheekLower {parent=FACIAL_C_FacialRoot, loc=<-3.388 4.3763 5.3857>, rot=<0.0319 0.427 -0.0564 -0.9019>}
 FACIAL_C_MouthUpper {parent=FACIAL_C_FacialRoot, loc=<-0.0013 4.872 7.1048>, rot=<-0.003 0 0 -1>}
 FACIAL_C_LipUpper {parent=FACIAL_C_MouthUpper, loc=<-0.002 0.0793 2.0751>, rot=<-0.0872 0 0 -0.9962>}
 FACIAL_L_LipUpper {parent=FACIAL_C_MouthUpper, loc=<0.8361 0.1614 1.9726>, rot=<-0.0778 -0.1601 -0.0379 -0.9833>}
 FACIAL_R_LipUpper {parent=FACIAL_C_MouthUpper, loc=<-0.8394 0.1608 1.9716>, rot=<-0.0778 0.1601 0.0379 -0.9833>}
 FACIAL_L_LipUpperOuter {parent=FACIAL_C_MouthUpper, loc=<1.4547 0.2502 1.5512>, rot=<-0.0984 -0.2293 -0.0983 -0.9634>}
 FACIAL_R_LipUpperOuter {parent=FACIAL_C_MouthUpper, loc=<-1.4607 0.2485 1.5478>, rot=<-0.0984 0.2293 0.0983 -0.9634>}
 FACIAL_L_LipCorner {parent=FACIAL_C_MouthUpper, loc=<1.9607 0.6077 0.8532>, rot=<0 -0.2078 0 -0.9782>}
 FACIAL_R_LipCorner {parent=FACIAL_C_MouthUpper, loc=<-1.9635 0.6084 0.8488>, rot=<0 0.2078 0 -0.9782>}
 FACIAL_L_Masseter {parent=FACIAL_C_FacialRoot, loc=<5.585 2.034 3.9538>, rot=<0.0664 0.1129 0.0082 -0.9914>}
 FACIAL_R_Masseter {parent=FACIAL_C_FacialRoot, loc=<-5.586 2.0338 3.9519>, rot=<0.0664 -0.1129 -0.0082 -0.9914>}
 FACIAL_C_UnderChin {parent=FACIAL_C_FacialRoot, loc=<-0.0007 8.241 3.1986>, rot=<0.4973 0 0 -0.8676>}
 FACIAL_L_UnderChin {parent=FACIAL_C_FacialRoot, loc=<2.2067 7.3452 2.4569>, rot=<0.4 -0.3551 0.0813 -0.841>}
 FACIAL_R_UnderChin {parent=FACIAL_C_FacialRoot, loc=<-2.2069 7.3457 2.457>, rot=<0.4 0.3551 -0.0812 -0.841>}
 FACIAL_C_TeethUpper {parent=FACIAL_C_FacialRoot, loc=<-0.001 5.3239 8.1125>, rot=<-0.003 0 0 -1>}
 FACIAL_C_LowerLipRotation {parent=FACIAL_C_FacialRoot, loc=<-0.0019 2.2688 2.2024>, rot=<0.3392 0 0 -0.9407>}
 FACIAL_C_MouthLower {parent=FACIAL_C_LowerLipRotation, loc=<0.0005 -1.1249 5.4355>, rot=<-0.342 0 0 -0.9397>}
 FACIAL_C_LipLower {parent=FACIAL_C_MouthLower, loc=<0.0002 1.155 1.7726>, rot=<0.0872 0 0 -0.9962>}
 FACIAL_L_LipLower {parent=FACIAL_C_MouthLower, loc=<0.727 1.106 1.6362>, rot=<0.0927 -0.1294 0.0272 -0.9869>}
 FACIAL_R_LipLower {parent=FACIAL_C_MouthLower, loc=<-0.7263 1.106 1.6348>, rot=<0.0927 0.1294 -0.0272 -0.9869>}
 FACIAL_L_LipLowerOuter {parent=FACIAL_C_MouthLower, loc=<1.3471 0.9642 1.3227>, rot=<0.0913 -0.2321 0.0318 -0.9679>}
 FACIAL_R_LipLowerOuter {parent=FACIAL_C_MouthLower, loc=<-1.3461 0.9621 1.3176>, rot=<0.0913 0.2321 -0.0318 -0.9679>}
 FACIAL_C_TeethLower {parent=FACIAL_C_LowerLipRotation, loc=<0.0007 -1.3299 6.462>, rot=<-0.342 0 0 -0.9397>}
 FACIAL_C_Tongue1 {parent=FACIAL_C_TeethLower, loc=<0.0011 0.1869 -4.7606>, rot=<-0.2588 0 0 -0.9659>}
 FACIAL_C_Tongue2 {parent=FACIAL_C_Tongue1, loc=<-0.0009 0.3718 1.393>, rot=<0.2588 0 0 -0.9659>}
 FACIAL_C_Tongue3 {parent=FACIAL_C_Tongue2, loc=<-0.0003 0.1895 1.2252>, rot=<0.1305 0 0 -0.9914>}
 FACIAL_C_Tongue4 {parent=FACIAL_C_Tongue3, loc=<0.0008 -0.1477 1.0709>, rot=<0.0436 0 0 -0.999>}
 FACIAL_C_TongueUpper3 {parent=FACIAL_C_Tongue3, loc=<0.0003 -0.3008 -0.1976>, rot=<0 0 0 1>}
 FACIAL_C_TongueLower3 {parent=FACIAL_C_Tongue3, loc=<-0.0005 0.7843 0.0061>, rot=<0 0 0 1>}
 FACIAL_L_TongueSide3 {parent=FACIAL_C_Tongue3, loc=<1.449 -0.0085 0.0879>, rot=<0 0 0 1>}
 FACIAL_R_TongueSide3 {parent=FACIAL_C_Tongue3, loc=<-1.4494 -0.0083 0.0853>, rot=<0 0 0 1>}
 FACIAL_C_TongueUpper2 {parent=FACIAL_C_Tongue2, loc=<0.0002 -0.5648 -0.2816>, rot=<0 0 0 1>}
 FACIAL_L_TongueSide2 {parent=FACIAL_C_Tongue2, loc=<1.7322 -0.0583 0.0691>, rot=<0 0 0 1>}
 FACIAL_R_TongueSide2 {parent=FACIAL_C_Tongue2, loc=<-1.7341 -0.0597 0.0673>, rot=<0 0 0 1>}
 FACIAL_C_Jaw {parent=FACIAL_C_FacialRoot, loc=<-0.0019 2.2688 2.2024>, rot=<0.3392 0 0 -0.9407>}
 FACIAL_C_Jawline {parent=FACIAL_C_Jaw, loc=<0.0014 2.4414 6.4469>, rot=<0.2586 0 0 -0.966>}
 FACIAL_L_ChinSide {parent=FACIAL_C_Jaw, loc=<2.3283 0.6117 6.8052>, rot=<-0.2208 -0.3738 0.0499 -0.8995>}
 FACIAL_R_ChinSide {parent=FACIAL_C_Jaw, loc=<-2.3224 0.6113 6.805>, rot=<-0.2208 0.3738 -0.0499 -0.8995>}
 FACIAL_C_Chin {parent=FACIAL_C_Jaw, loc=<0.0024 1.2677 7.043>, rot=<-0.2591 0 0 -0.9659>}
 FACIAL_L_Ear {parent=FACIAL_C_FacialRoot, loc=<6.1615 0.2004 0.7449>, rot=<-0.0293 -0.0006 -0.0261 -0.9992>}
 FACIAL_R_Ear {parent=FACIAL_C_FacialRoot, loc=<-6.1619 0.2002 0.7442>, rot=<-0.0293 0.0006 0.0262 -0.9992>}
 FACIAL_L_Jawline {parent=FACIAL_C_FacialRoot, loc=<3.6882 5.5696 4.341>, rot=<0.2064 0.1579 -0.0542 -0.9641>}
 FACIAL_R_Jawline {parent=FACIAL_C_FacialRoot, loc=<-3.6937 5.5713 4.3378>, rot=<0.2064 -0.1579 0.0542 -0.9641>}
 FACIAL_C_Forehead {parent=FACIAL_C_FacialRoot, loc=<-0.0005 -2.3671 9.182>, rot=<-0.0464 0 0 -0.9989>}
 FACIAL_L_ForeheadIn {parent=FACIAL_C_FacialRoot, loc=<1.5213 -2.5092 8.9536>, rot=<-0.0637 -0.0696 0.0041 -0.9955>}
 FACIAL_R_ForeheadIn {parent=FACIAL_C_FacialRoot, loc=<-1.5216 -2.5079 8.9538>, rot=<-0.0637 0.0696 -0.004 -0.9955>}
 FACIAL_L_ForeheadMid {parent=FACIAL_C_FacialRoot, loc=<2.9518 -2.937 8.2688>, rot=<-0.0739 -0.1756 -0.0211 -0.9815>}
 FACIAL_R_ForeheadMid {parent=FACIAL_C_FacialRoot, loc=<-2.9491 -2.9355 8.265>, rot=<-0.0739 0.1756 0.0211 -0.9815>}
 FACIAL_L_ForeheadOut {parent=FACIAL_C_FacialRoot, loc=<4.2888 -2.894 7.3168>, rot=<-0.1242 -0.2554 -0.0799 -0.9555>}
 FACIAL_R_ForeheadOut {parent=FACIAL_C_FacialRoot, loc=<-4.2868 -2.8925 7.3172>, rot=<-0.1242 0.2554 0.0799 -0.9555>}
 FACIAL_L_EyesackUpper {parent=FACIAL_C_FacialRoot, loc=<3.04 -1.9131 9.0035>, rot=<0.0458 -0.084 -0.0647 -0.9933>}
 FACIAL_R_EyesackUpper {parent=FACIAL_C_FacialRoot, loc=<-3.0373 -1.9096 8.9998>, rot=<0.0458 0.084 0.0648 -0.9933>}
 FACIAL_L_EyelidUpperFurrow {parent=FACIAL_C_FacialRoot, loc=<2.9152 -1.641 8.873>, rot=<-0.033 -0.1506 -0.0801 -0.9848>}
 FACIAL_R_EyelidUpperFurrow {parent=FACIAL_C_FacialRoot, loc=<-2.9106 -1.6408 8.8696>, rot=<-0.033 0.1506 0.0801 -0.9848>}
 FACIAL_L_EyelidUpperA {parent=FACIAL_C_FacialRoot, loc=<2.6673 -0.5603 6.9224>, rot=<-0.1508 0 0 -0.9886>}
 FACIAL_R_EyelidUpperA {parent=FACIAL_C_FacialRoot, loc=<-2.667 -0.5612 6.9271>, rot=<-0.1508 0 0 -0.9886>}
 FACIAL_L_Eye {parent=FACIAL_C_FacialRoot, loc=<2.6508 -0.5711 6.8965>, rot=<-0.003 0 0 -1>}
 FACIAL_R_Eye {parent=FACIAL_C_FacialRoot, loc=<-2.6518 -0.5707 6.9003>, rot=<-0.003 0 0 -1>}
 FACIAL_L_EyelidLowerA {parent=FACIAL_C_FacialRoot, loc=<2.6673 -0.5603 6.9224>, rot=<0.2811 0 0 -0.9597>}
 FACIAL_R_EyelidLowerA {parent=FACIAL_C_FacialRoot, loc=<-2.667 -0.5611 6.9271>, rot=<0.2811 0 0 -0.9597>}
 FACIAL_L_EyeCornerInner {parent=FACIAL_C_FacialRoot, loc=<1.6501 0.1708 7.7858>, rot=<-0.003 0 0 -1>}
 FACIAL_R_EyeCornerInner {parent=FACIAL_C_FacialRoot, loc=<-1.6449 0.1565 7.7899>, rot=<-0.003 0 0 -1>}
 FACIAL_L_EyeCornerOuter {parent=FACIAL_C_FacialRoot, loc=<4.4518 -0.6122 7.4636>, rot=<-0.0028 -0.3826 -0.0011 -0.9239>}
 FACIAL_R_EyeCornerOuter {parent=FACIAL_C_FacialRoot, loc=<-4.4517 -0.6068 7.475>, rot=<-0.0028 0.3826 0.0012 -0.9239>}
 FACIAL_L_EyesackLower {parent=FACIAL_C_FacialRoot, loc=<3.3542 1.2416 7.2501>, rot=<-0.0252 -0.1396 -0.0227 -0.9896>}
 FACIAL_R_EyesackLower {parent=FACIAL_C_FacialRoot, loc=<-3.3509 1.2392 7.2542>, rot=<-0.0253 0.1396 0.0227 -0.9896>}
 FACIAL_L_CheekInner {parent=FACIAL_C_FacialRoot, loc=<2.0973 1.6656 7.9744>, rot=<-0.1261 -0.0874 -0.1974 -0.9682>}
 FACIAL_R_CheekInner {parent=FACIAL_C_FacialRoot, loc=<-2.0971 1.6659 7.9701>, rot=<-0.1261 0.0874 0.1974 -0.9682>}
 FACIAL_C_Nose {parent=FACIAL_C_FacialRoot, loc=<-0.0012 1.0507 9.7749>, rot=<-0.1333 0 0 -0.9911>}
 FACIAL_C_NoseLower {parent=FACIAL_C_Nose, loc=<0.0003 2.6077 -0.6358>, rot=<0.1305 0 0 -0.9914>}
 FACIAL_L_Nostril {parent=FACIAL_C_Nose, loc=<0.7198 1.796 -0.4976>, rot=<0.0173 0.1305 0.0023 -0.9913>}
 FACIAL_R_Nostril {parent=FACIAL_C_Nose, loc=<-0.7192 1.7955 -0.4993>, rot=<0.0173 -0.1305 -0.0023 -0.9913>}
 FACIAL_C_NoseTip {parent=FACIAL_C_Nose, loc=<0.0006 2.1229 0.188>, rot=<0.0872 0 0 -0.9962>}
 FACIAL_C_NoseBridge {parent=FACIAL_C_FacialRoot, loc=<-0.0008 -0.8842 9.9032>, rot=<0.0408 0 0 -0.9992>}
 FACIAL_L_EyelidUpperB {parent=FACIAL_C_FacialRoot, loc=<2.6673 -0.5603 6.9224>, rot=<-0.2279 0 0 -0.9737>}
 FACIAL_L_EyelidLowerB {parent=FACIAL_C_FacialRoot, loc=<2.6673 -0.5603 6.9224>, rot=<0.3145 0 0 -0.9493>}
 FACIAL_R_EyelidLowerB {parent=FACIAL_C_FacialRoot, loc=<-2.667 -0.5612 6.9271>, rot=<0.3227 0 0 -0.9465>}
 FACIAL_R_EyelidUpperB {parent=FACIAL_C_FacialRoot, loc=<-2.667 -0.5612 6.9271>, rot=<-0.2279 0 0 -0.9737>}
 dyn_hair_side_r_mstr {parent=head, loc=<9.7903 5.3031 5.6103>, rot=<0.1138 0.1364 0.9654 -0.1908>}
 dyn_hair_side_r_01 {parent=dyn_hair_side_r_mstr, loc=<0 0 0>, rot=<0 0 0 1>}
 dyn_hair_side_r_02 {parent=dyn_hair_side_r_01, loc=<4.2565 0 0>, rot=<-0.0282 -0.1357 0.1786 -0.9741>}
 dyn_hair_side_r_03 {parent=dyn_hair_side_r_02, loc=<3.402 0 0>, rot=<0.06 -0.0021 0.0181 -0.998>}
 dyn_hair_side_r_04 {parent=dyn_hair_side_r_03, loc=<3.5834 0 0>, rot=<0.0012 -0.0306 0.1012 -0.9944>}
 dyn_hair_side_r_05 {parent=dyn_hair_side_r_04, loc=<3.4563 0 0>, rot=<0.0311 0 0 -0.9995>}
 dyn_hair_side_r_07 {parent=dyn_hair_side_r_03, loc=<0.9984 -1.7134 0.0106>, rot=<-0.0832 -0.0435 0.1316 -0.9868>}
 dyn_hair_side_r_06 {parent=dyn_hair_side_r_02, loc=<0.075 2.3308 -0.0464>, rot=<-0.0268 0.0746 -0.0982 -0.992>}
 dyn_flyoff_02 {parent=head, loc=<10.3056 2.7278 6.302>, rot=<0.4687 0.1662 0.7221 -0.481>}
 dyn_strand_fr_r_a_mstr {parent=head, loc=<8.5004 9.0216 3.3881>, rot=<0.034 0.5449 0.8261 -0.1394>}
 dyn_strand_fr_r_a_01 {parent=dyn_strand_fr_r_a_mstr, loc=<0 0 0>, rot=<0 0 0 1>}
 dyn_strand_fr_r_a_02 {parent=dyn_strand_fr_r_a_01, loc=<2.8093 0.1265 0.0289>, rot=<0.0194 -0.0278 -0.1886 -0.9815>}
 dyn_strand_fr_r_mstr {parent=head, loc=<10.3468 9.3966 -0.0612>, rot=<0.181 0.6835 0.707 -0.0123>}
 dyn_strand_fr_r_01 {parent=dyn_strand_fr_r_mstr, loc=<0 0 0>, rot=<0 0 0 1>}
 dyn_strand_fr_r_02 {parent=dyn_strand_fr_r_01, loc=<3.2631 0 0>, rot=<0.0046 -0.2041 0.0148 -0.9788>}
 dyn_strand_fr_r_03 {parent=dyn_strand_fr_r_02, loc=<2.8093 0.1265 0.0289>, rot=<0.0296 -0.1922 -0.0689 -0.9785>}
 dyn_strand_fr_l_mstr {parent=head, loc=<10.6232 8.7005 -2.7381>, rot=<-0.0945 -0.8411 -0.5169 -0.1281>}
 dyn_strand_fr_l_01 {parent=dyn_strand_fr_l_mstr, loc=<0 0 0>, rot=<0 0 0 1>}
 dyn_strand_fr_l_04 {parent=dyn_strand_fr_l_01, loc=<1.8164 0.7577 0.0712>, rot=<0.0145 -0.0622 -0.0623 -0.996>}
 dyn_strand_fr_l_05 {parent=dyn_strand_fr_l_04, loc=<2.2706 -0.116 -0.2045>, rot=<0.0423 -0.1167 0.048 -0.9911>}
 dyn_strand_fr_l_02 {parent=dyn_strand_fr_l_01, loc=<2.5243 -0.7097 -0.1242>, rot=<-0.0382 -0.18 0.203 -0.9617>}
 dyn_strand_fr_l_03 {parent=dyn_strand_fr_l_02, loc=<2.853 -0.032 -0.1031>, rot=<-0.0152 0.0158 0.0934 -0.9954>}
 dyn_hair_side_l_mstr {parent=head, loc=<9.8705 5.3552 -5.4283>, rot=<0.1256 -0.9751 -0.115 -0.1419>}
 dyn_hair_side_l_01 {parent=dyn_hair_side_l_mstr, loc=<0 0 0>, rot=<0 0 0 1>}
 dyn_hair_side_l_02 {parent=dyn_hair_side_l_01, loc=<2.5944 0 0>, rot=<-0.0067 -0.1035 -0.0293 -0.9942>}
 dyn_hair_side_l_03 {parent=dyn_hair_side_l_02, loc=<3.9509 0 0>, rot=<0.0081 -0.1119 -0.0784 -0.9906>}
 dyn_hair_side_l_04 {parent=dyn_hair_side_l_03, loc=<3.3001 0 0>, rot=<0.0193 0.0461 -0.0705 -0.9963>}
 dyn_hair_side_l_05 {parent=dyn_hair_side_l_04, loc=<3.8494 0 0>, rot=<0.0202 -0.0033 -0.0006 -0.9998>}
 dyn_hair_side_l_06 {parent=dyn_hair_side_l_03, loc=<2.9607 1.7491 -0.0846>, rot=<0.0703 -0.0876 -0.1709 -0.9789>}
 dyn_hair_side_l_07 {parent=dyn_hair_side_l_02, loc=<1.3083 -2.7948 0.224>, rot=<-0.0319 -0.0405 0.017 -0.9985>}
 dyn_hair_side_l_08 {parent=dyn_hair_side_l_07, loc=<2.7199 0 0>, rot=<0 0 0.0688 -0.9976>}
 dyn_flyoff_01 {parent=head, loc=<10.6279 2.4112 -5.7487>, rot=<0.733 -0.6007 0.2461 -0.2032>}
 dyn_hair_bun {parent=head, loc=<10.4 -5.3436 0.2906>, rot=<0.8841 -0.4666 -0.0161 0.0187>}
 dyn_bun_top_l {parent=dyn_hair_bun, loc=<3.6594 -2.3382 2.6422>, rot=<0.5882 -0.8037 0.0255 0.0858>}
 dyn_bun_top_r_01 {parent=dyn_hair_bun, loc=<6.0659 -2.3744 -4.0671>, rot=<0.8745 -0.2284 -0.3119 0.2928>}
 dyn_bun_top_r_02 {parent=dyn_bun_top_r_01, loc=<1.4706 -0.009 0.0552>, rot=<-0.1024 -0.3269 -0.0174 -0.9393>}
 dyn_bun_l_01 {parent=dyn_hair_bun, loc=<3.232 3.4746 3.7057>, rot=<0.4421 0.1007 -0.1021 -0.8854>}
 dyn_bun_l_02 {parent=dyn_bun_l_01, loc=<2.4939 -0.0154 0.2002>, rot=<-0.0278 -0.332 -0.0157 -0.9427>}
 dyn_bun_l_03 {parent=dyn_hair_bun, loc=<1.8613 4.5447 -2.4881>, rot=<0.5453 -0.061 -0.2501 -0.7977>}
 dyn_bun_l_04 {parent=dyn_bun_l_03, loc=<2.4939 -0.0154 0.2002>, rot=<-0.0278 -0.332 -0.0157 -0.9427>}
 attach_backpack {parent=spine_05, loc=<-0.123 -13.473 0>, rot=<-0.0093 -0.707 0.0093 -0.707>}
 clavicle_l {parent=spine_05, loc=<5.4831 -0.1053 -3.2946>, rot=<0.0465 -0.7289 -0.009 -0.6829>}
 upperarm_l {parent=clavicle_l, loc=<9.8658 -1.2136 -0.4105>, rot=<0.0164 -0.3245 -0.0269 -0.9454>}
 lowerarm_l {parent=upperarm_l, loc=<26.465 -0.8345 0.259>, rot=<0.0036 0.0161 -0.2298 -0.9731>}
 hand_l {parent=lowerarm_l, loc=<22.7355 -0.3203 -0.6105>, rot=<-0.6784 -0.0223 0.0735 -0.7307>}
 weapon_l {parent=hand_l, loc=<-0.1974 -3.4111 0.3811>, rot=<0 0 0 1>}
 hand_attach_l {parent=hand_l, loc=<8.5 -3 0>, rot=<0.7071 0.7071 0 0>}
 thumb_01_l {parent=hand_l, loc=<1.3118 -0.3191 -2.459>, rot=<0.6546 -0.4042 0.0664 -0.6354>}
 thumb_02_l {parent=thumb_01_l, loc=<5.9336 -0.0448 0.0401>, rot=<-0.0037 0.0213 0.298 -0.9543>}
 thumb_03_l {parent=thumb_02_l, loc=<2.2223 0.3682 -0.0907>, rot=<0.0118 -0.013 0.0066 -0.9998>}
 index_metacarpal_l {parent=hand_l, loc=<1.6535 0.0183 -2.96>, rot=<0.0158 -0.0124 -0.0173 -0.9996>}
 index_01_l {parent=index_metacarpal_l, loc=<8.3008 0.0357 -0.0524>, rot=<-0.0044 0.0485 0.0924 -0.9945>}
 index_02_l {parent=index_01_l, loc=<3.3587 -0.3004 0.0452>, rot=<-0.002 0.0027 0.3225 -0.9466>}
 index_03_l {parent=index_02_l, loc=<2.5578 -0.123 -0.0252>, rot=<-0.002 0.0067 0.2558 -0.9667>}
 middle_metacarpal_l {parent=hand_l, loc=<2.1856 -0.1725 -0.5984>, rot=<0.006 0.0168 -0.0329 -0.9993>}
 middle_01_l {parent=middle_metacarpal_l, loc=<7.3275 -0.2447 0.0129>, rot=<-0.0208 0.0042 0.0858 -0.9961>}
 middle_02_l {parent=middle_01_l, loc=<4.0123 0.1401 0.1175>, rot=<-0.0218 0.0258 0.3512 -0.9357>}
 middle_03_l {parent=middle_02_l, loc=<2.6077 0.0534 0.1373>, rot=<0.0082 0.0188 0.1708 -0.9851>}
 ring_metacarpal_l {parent=hand_l, loc=<2.2074 -0.3625 1.3638>, rot=<-0.0284 0.0722 -0.0458 -0.9959>}
 ring_01_l {parent=ring_metacarpal_l, loc=<7.0295 -0.6197 -0.1446>, rot=<-0.0489 -0.0252 0.1892 -0.9804>}
 ring_02_l {parent=ring_01_l, loc=<3.6174 0.4647 -0.1027>, rot=<0 0 0.2806 -0.9598>}
 ring_03_l {parent=ring_02_l, loc=<2.4152 0.114 0.0199>, rot=<0 0 0.1189 -0.9929>}
 pinky_metacarpal_l {parent=hand_l, loc=<1.7616 -0.9753 3.3814>, rot=<-0.0397 0.0814 -0.069 -0.9935>}
 pinky_01_l {parent=pinky_metacarpal_l, loc=<7.1709 -0.4831 -0.1094>, rot=<-0.0654 -0.0035 0.2242 -0.9723>}
 pinky_02_l {parent=pinky_01_l, loc=<2.8274 0.0893 -0.1043>, rot=<0.0037 0.0011 0.2854 -0.9584>}
 pinky_03_l {parent=pinky_02_l, loc=<2.2046 0.0307 -0.0078>, rot=<-0.0002 0.0098 0.1445 -0.9895>}
 lowerarm_twist_02_l {parent=lowerarm_l, loc=<10.6707 0 0>, rot=<0 0 0 1>}
 lowerarm_twist_01_l {parent=lowerarm_l, loc=<18.316 0 0>, rot=<0 0 0 1>}
 upperarm_twist_01_l {parent=upperarm_l, loc=<7.2135 -0.1134 0>, rot=<0 0 0 1>}
 dyn_lowerarm_base_l {parent=upperarm_l, loc=<19.7998 -0.1134 0>, rot=<0 0 0 1>}
 upperarm_twist_02_l {parent=upperarm_l, loc=<19.7998 -0.1134 0>, rot=<0 0 0 1>}
 clavicle_r {parent=spine_05, loc=<5.4831 -0.1053 3.2946>, rot=<0.7289 0.0465 -0.6829 0.009>}
 upperarm_r {parent=clavicle_r, loc=<-9.8658 1.2136 0.4105>, rot=<0.0164 -0.3245 -0.0269 -0.9454>}
 lowerarm_r {parent=upperarm_r, loc=<-26.465 0.8345 -0.259>, rot=<0.0036 0.0161 -0.2298 -0.9731>}
 hand_r {parent=lowerarm_r, loc=<-22.7355 0.3203 0.6106>, rot=<-0.6784 -0.0223 0.0735 -0.7307>}
 weapon_r {parent=hand_r, loc=<1.7269 4.0421 -5.202>, rot=<-0.1366 -0.0241 0.9888 -0.056>}
 thumb_01_r {parent=hand_r, loc=<-1.3118 0.3191 2.459>, rot=<0.6546 -0.4042 0.0664 -0.6354>}
 thumb_02_r {parent=thumb_01_r, loc=<-5.9336 0.0448 -0.0401>, rot=<-0.0037 0.0213 0.298 -0.9543>}
 thumb_03_r {parent=thumb_02_r, loc=<-2.2223 -0.3682 0.0907>, rot=<0.0118 -0.013 0.0066 -0.9998>}
 index_metacarpal_r {parent=hand_r, loc=<-1.6535 -0.0183 2.96>, rot=<0.0158 -0.0124 -0.0173 -0.9996>}
 index_01_r {parent=index_metacarpal_r, loc=<-8.3008 -0.0357 0.0524>, rot=<-0.0044 0.0485 0.0924 -0.9945>}
 index_02_r {parent=index_01_r, loc=<-3.3587 0.3004 -0.0452>, rot=<-0.002 0.0027 0.3225 -0.9466>}
 index_03_r {parent=index_02_r, loc=<-2.5578 0.123 0.0252>, rot=<-0.002 0.0067 0.2558 -0.9667>}
 middle_metacarpal_r {parent=hand_r, loc=<-2.1856 0.1725 0.5984>, rot=<0.006 0.0168 -0.0329 -0.9993>}
 middle_01_r {parent=middle_metacarpal_r, loc=<-7.3275 0.2447 -0.0129>, rot=<-0.0208 0.0042 0.0858 -0.9961>}
 middle_02_r {parent=middle_01_r, loc=<-4.0123 -0.1401 -0.1175>, rot=<-0.0218 0.0258 0.3512 -0.9357>}
 middle_03_r {parent=middle_02_r, loc=<-2.6077 -0.0534 -0.1373>, rot=<0.0082 0.0188 0.1708 -0.9851>}
 ring_metacarpal_r {parent=hand_r, loc=<-2.2074 0.3625 -1.3638>, rot=<-0.0284 0.0722 -0.0458 -0.9959>}
 ring_01_r {parent=ring_metacarpal_r, loc=<-7.0295 0.6197 0.1446>, rot=<-0.0489 -0.0252 0.1892 -0.9804>}
 ring_02_r {parent=ring_01_r, loc=<-3.6174 -0.4647 0.1027>, rot=<0 0 0.2806 -0.9598>}
 ring_03_r {parent=ring_02_r, loc=<-2.4152 -0.114 -0.0199>, rot=<0 0 0.1189 -0.9929>}
 pinky_metacarpal_r {parent=hand_r, loc=<-1.7616 0.9753 -3.3814>, rot=<-0.0397 0.0814 -0.069 -0.9935>}
 pinky_01_r {parent=pinky_metacarpal_r, loc=<-7.1709 0.4831 0.1094>, rot=<-0.0654 -0.0035 0.2242 -0.9723>}
 pinky_02_r {parent=pinky_01_r, loc=<-2.8274 -0.0893 0.1043>, rot=<0.0037 0.0011 0.2854 -0.9584>}
 pinky_03_r {parent=pinky_02_r, loc=<-2.2046 -0.0307 0.0078>, rot=<-0.0002 0.0098 0.1445 -0.9895>}
 hand_attach_r {parent=hand_r, loc=<-8.5 3 0>, rot=<0 0 -0.7071 -0.7071>}
 lowerarm_twist_02_r {parent=lowerarm_r, loc=<-10.6707 0 0>, rot=<0 0 0 -1>}
 lowerarm_twist_01_r {parent=lowerarm_r, loc=<-18.316 0 0>, rot=<0 0 0 -1>}
 upperarm_twist_01_r {parent=upperarm_r, loc=<-7.2135 0.1134 0>, rot=<0 0 0 -1>}
 upperarm_twist_02_r {parent=upperarm_r, loc=<-19.7998 0.1134 0>, rot=<0 0 0 -1>}
 dyn_ribbon_MSTR {parent=spine_05, loc=<8.4172 6.27 0>, rot=<0 0 -0.036 -0.9994>}
 dyn_ribbon_l_01 {parent=dyn_ribbon_MSTR, loc=<0.7859 0.5223 -1.006>, rot=<0.2568 -0.6406 -0.289 -0.6635>}
 dyn_ribbon_r_01 {parent=dyn_ribbon_MSTR, loc=<0.7861 0.5223 1.006>, rot=<0.6635 0.289 0.6406 -0.2568>}
 dyn_ribbon_l_02 {parent=dyn_ribbon_MSTR, loc=<-0.0786 0.3114 -1.1844>, rot=<-0.2185 -0.6547 -0.6495 -0.3192>}
 dyn_ribbon_r_02 {parent=dyn_ribbon_MSTR, loc=<-0.0781 0.3114 1.1844>, rot=<0.3192 0.6495 0.6547 0.2185>}
 dyn_l_ShoulderBlade {parent=dyn_ribbon_MSTR, loc=<-4.122 -14.7201 -9.7676>, rot=<0 0 0 1>}
 dyn_r_ShoulderBlade {parent=dyn_ribbon_MSTR, loc=<-4.1221 -14.7202 9.7676>, rot=<0 0 0 1>}
 dyn_arm_r {parent=spine_05, loc=<4.4663 -1.5473 13.084>, rot=<0.9118 0.0385 -0.4081 0.0238>}
 dyn_arm_aim_r {parent=dyn_arm_r, loc=<0 0 0>, rot=<0 0 0 1>}
 dyn_r_shoulder {parent=dyn_arm_aim_r, loc=<-7.1391 0.464 -0.0065>, rot=<0 0 0 -1>}
 dyn_lowerarm_aim_r {parent=dyn_arm_aim_r, loc=<-26.3906 1.1852 -0.2655>, rot=<0.0036 0.0161 -0.2298 -0.9731>}
 dyn_lowerarm_base_r {parent=dyn_lowerarm_aim_r, loc=<0 0 0>, rot=<0 0 0 1>}
 dyn_arm_l {parent=spine_05, loc=<4.4663 -1.5473 -13.084>, rot=<0.0385 -0.9118 -0.0238 -0.4081>}
 dyn_arm_aim_l {parent=dyn_arm_l, loc=<0 0 0>, rot=<0 0 0 1>}
 dyn_l_shoulder {parent=dyn_arm_aim_l, loc=<7.1391 -0.464 0.0065>, rot=<0 0 0 1>}
 dyn_lowerarm_base_l_02 {parent=dyn_arm_aim_l, loc=<26.3906 -1.1852 0.2655>, rot=<0.0036 0.0161 -0.2298 -0.9731>}
 thigh_l {parent=pelvis, loc=<-1.0855 -0.2109 -8.7053>, rot=<0.0009 0.0355 0.0155 -0.9993>}
 calf_l {parent=thigh_l, loc=<-42.5284 -0.0019 -0.7297>, rot=<0.0006 -0.0161 -0.0391 -0.9991>}
 foot_l {parent=calf_l, loc=<-46.7401 -2.5014 -0.592>, rot=<0.0427 -0.0238 0.0353 -0.9982>}
 ball_l {parent=foot_l, loc=<-3.5658 10.8066 -0.1544>, rot=<0.0001 -0.0001 -0.7063 -0.7079>}
 calf_twist_02_l {parent=calf_l, loc=<-19.1546 -1.7943 -0.2638>, rot=<0.0237 0.0095 -0.0035 -0.9997>}
 calf_twist_01_l {parent=calf_l, loc=<-35.8483 -1.9197 -0.5794>, rot=<0.0237 0.0095 -0.0035 -0.9997>}
 dyn_l_knee {parent=calf_l, loc=<0 0 0>, rot=<0 0 0 1>}
 thigh_twist_01_l {parent=thigh_l, loc=<-27.952 0 0>, rot=<0.0013 0 0 -1>}
 thigh_r {parent=pelvis, loc=<-1.0855 -0.2109 8.7053>, rot=<0.0355 -0.0009 0.9993 0.0155>}
 calf_r {parent=thigh_r, loc=<42.5284 0.0019 0.7297>, rot=<0.0006 -0.0161 -0.0391 -0.9991>}
 foot_r {parent=calf_r, loc=<46.7401 2.5014 0.592>, rot=<0.0427 -0.0238 0.0353 -0.9982>}
 ball_r {parent=foot_r, loc=<3.5658 -10.8066 0.1544>, rot=<0.0001 -0.0001 -0.7063 -0.7079>}
 calf_twist_02_r {parent=calf_r, loc=<19.1546 1.7943 0.2638>, rot=<0.0237 0.0095 -0.0035 -0.9997>}
 calf_twist_01_r {parent=calf_r, loc=<35.8483 1.9197 0.5794>, rot=<0.0237 0.0095 -0.0035 -0.9997>}
 dyn_r_knee {parent=calf_r, loc=<0 0 0>, rot=<0 0 0 1>}
 thigh_twist_01_r {parent=thigh_r, loc=<27.952 0 0>, rot=<0.0013 0 0 -1>}
 dyn_Cloth_volume_fn {parent=pelvis, loc=<7.6158 8.0085 0.1611>, rot=<0.9842 -0.1772 0 0>}
 dyn_Cloth_volume_fn01 {parent=dyn_Cloth_volume_fn, loc=<-5.4605 0 0>, rot=<0.0282 -0.0061 0.1067 -0.9939>}
 dyn_Cloth_volume_fn02 {parent=dyn_Cloth_volume_fn01, loc=<-14.9443 0 0>, rot=<0 0 0 1>}
 dyn_Cloth_volume_bk {parent=pelvis, loc=<5.7039 -10.3194 0.2721>, rot=<0 0 -0.1508 -0.9886>}
 dyn_Cloth_volume_bk01 {parent=dyn_Cloth_volume_bk, loc=<-4.8233 0 0>, rot=<-0.0406 0.0112 0.0695 -0.9967>}
 dyn_Cloth_volume_bk02 {parent=dyn_Cloth_volume_bk01, loc=<-12.1566 0 0>, rot=<0 0 0 1>}
 dyn_Cloth_volume_l_A {parent=pelvis, loc=<8.194 6.1203 -9.4731>, rot=<-0.8438 0.1554 -0.0749 -0.5082>}
 dyn_Cloth_volume_l_A01 {parent=dyn_Cloth_volume_l_A, loc=<-4.7573 0 0>, rot=<0.0177 -0.0055 0.0294 -0.9994>}
 dyn_Cloth_volume_l_A02 {parent=dyn_Cloth_volume_l_A01, loc=<-14.7041 0 0>, rot=<0 0 0 1>}
 dyn_Cloth_volume_l_B {parent=pelvis, loc=<7.8743 -1.909 -12.2536>, rot=<-0.6825 0.197 -0.1708 -0.6828>}
 dyn_Cloth_volume_l_B01 {parent=dyn_Cloth_volume_l_B, loc=<-5.1213 0 0>, rot=<0 0 0.0769 -0.997>}
 dyn_Cloth_volume_l_B02 {parent=dyn_Cloth_volume_l_B01, loc=<-12.7543 0 0>, rot=<0 0 0 1>}
 dyn_Cloth_volume_l_C {parent=pelvis, loc=<6.7 -8.0799 -8.976>, rot=<-0.2746 0.0751 -0.2195 -0.9331>}
 dyn_Cloth_volume_l_C01 {parent=dyn_Cloth_volume_l_C, loc=<-4.8373 0 0>, rot=<-0.01 0.0036 0.1469 -0.9891>}
 dyn_Cloth_volume_l_C02 {parent=dyn_Cloth_volume_l_C01, loc=<-14.0264 0 0>, rot=<0 0 0 1>}
 dyn_Cloth_volume_r_A {parent=pelvis, loc=<8.1896 5.5998 8.6377>, rot=<0.9132 -0.2092 -0.0658 -0.3434>}
 dyn_Cloth_volume_r_A01 {parent=dyn_Cloth_volume_r_A, loc=<-4.832 0 0>, rot=<-0.0155 0.0059 0.0503 -0.9986>}
 dyn_Cloth_volume_r_A02 {parent=dyn_Cloth_volume_r_A01, loc=<-14.7456 0 0>, rot=<0 0 0 1>}
 dyn_Cloth_volume_r_B {parent=pelvis, loc=<7.949 -1.9071 12.7534>, rot=<0.6815 -0.2005 -0.1743 -0.6819>}
 dyn_Cloth_volume_r_B01 {parent=dyn_Cloth_volume_r_B, loc=<-5.0511 0 0>, rot=<0 0 0.0882 -0.9961>}
 dyn_Cloth_volume_r_B02 {parent=dyn_Cloth_volume_r_B01, loc=<-12.7352 0 0>, rot=<0 0 0 1>}
 dyn_Cloth_volume_r_C {parent=pelvis, loc=<6.8422 -8.301 8.9665>, rot=<0.3346 -0.0888 -0.208 -0.9148>}
 dyn_Cloth_volume_r_C01 {parent=dyn_Cloth_volume_r_C, loc=<-4.7481 0 0>, rot=<0.0404 -0.0153 0.1207 -0.9917>}
 dyn_Cloth_volume_r_C02 {parent=dyn_Cloth_volume_r_C01, loc=<-13.2071 0 0>, rot=<0 0 0 1>}
 attach {parent=root, loc=<0 0 0>, rot=<0 0 0 1>}
 ik_hand_root {parent=root, loc=<0 0 0>, rot=<0 0 0 1>}
 ik_hand_gun {parent=ik_hand_root, loc=<-48.1743 3.3545 104.4381>, rot=<0.6276 0.1309 0.4183 -0.6434>}
 ik_hand_r {parent=ik_hand_gun, loc=<0.0112 -0.0274 0.0522>, rot=<0 0 0 1>}
 ik_hand_l {parent=ik_hand_gun, loc=<9.5353 13.8545 -46.5669>, rot=<0.3518 -0.9349 0.0463 -0.0071>}
 ik_foot_root {parent=root, loc=<0 0 0>, rot=<0 0 0 1>}
 ik_foot_l {parent=ik_foot_root, loc=<14.8262 -7.3545 8.296>, rot=<0.0357 0.7045 0.0284 -0.7082>}
 ik_foot_r {parent=ik_foot_root, loc=<-14.8262 -7.3545 8.296>, rot=<0.7082 0.0284 -0.7045 0.0357>}
 attach_fp {parent=root, loc=<0 0 -0.0267>, rot=<0 0 0 1>}
 camera_root {parent=root, loc=<0 0 0>, rot=<0 0 0 1>}
 camera {parent=camera_root, loc=<0 0 0>, rot=<0 0 0 1>}
 warp_point_01 {parent=root, loc=<0 0 0>, rot=<0 0 0 1>}
 warp_point_02 {parent=root, loc=<0 0 0>, rot=<0 0 0 1>}
 warp_point_03 {parent=root, loc=<0 0 0>, rot=<0 0 0 1>}
]
variables [
 FroneEnd {type=double, default=0}
 ParachuteAfterSkydive {type=double, default=0}
 Pickaxe {type=double, default=0}
 Pistol {type=double, default=0}
 Rifle {type=double, default=0}
 Skydive {type=double, default=0}
]
