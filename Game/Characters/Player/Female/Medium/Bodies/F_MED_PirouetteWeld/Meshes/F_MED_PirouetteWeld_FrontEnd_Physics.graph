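# Physics asset for the front-end PirouetteWeld medium female player body, covering its dynamic cloth-volume bones around the hips.
# Eight three-bone capsule chains (front, back, three per side) hang from the pelvis on angularly free root constraints.
# Links down each chain have limited swing and twist, tighter swing on the front and right A chains.
# Pelvis and thigh-twist capsules represent the body, and the left A tip carries three capsules.
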PhysicsAsset /Game/Characters/Player/Female/Medium/Bodies/F_MED_PirouetteWeld/Meshes/F_MED_PirouetteWeld_FrontEnd_Physics
bodies {
 dyn_Cloth_volume_bk [{
   shapeType=Capsule
   length=0.505
   radius=0.505
   shapeName=dyn_Cloth_volume_bk_capsule
   center=<-4.8702 0.3265 -0.2842>
   rotation=<rot -0.4767 -90 0.5473>
  }]
 dyn_Cloth_volume_bk01 [{shapeType=Capsule, length=9.0204, radius=7.1161, shapeName=dyn_Cloth_volume_bk01_capsule, center=<-7.4118 -3 0>, rotation=<rot 90 0 0>}]
 dyn_Cloth_volume_bk02 [{shapeType=Capsule, length=9.0204, radius=7.1161, shapeName=dyn_Cloth_volume_bk02_capsule, center=<-7.4118 -3 0>, rotation=<rot 90 0 0>}]
 dyn_Cloth_volume_fn [{
   shapeType=Capsule
   length=0.505
   radius=0.505
   shapeName=dyn_Cloth_volume_fn_capsule
   center=<-5.8692 0.6644 5.5143>
   rotation=<rot 0 -14.193 -80.002>
  }]
 dyn_Cloth_volume_fn01 [{shapeType=Capsule, length=9.0204, radius=3, shapeName=dyn_Cloth_volume_fn01_capsule, center=<-7.4118 -3 0>, rotation=<rot 90 0 0>}]
 dyn_Cloth_volume_fn02 [{shapeType=Capsule, length=9.0204, radius=3, shapeName=dyn_Cloth_volume_fn02_capsule, center=<-7.4118 -3 0>, rotation=<rot 90 0 0>}]
 dyn_Cloth_volume_l_A [{
   shapeType=Capsule
   length=3.6651
   radius=0.979
   shapeName=dyn_Cloth_volume_l_A_capsule
   center=<-3.2428 -0.4955 2.3396>
   rotation=<rot 0 -84.7489 -71.6904>
  }]
 dyn_Cloth_volume_l_A01 [{shapeType=Capsule, length=11.3163, radius=3, shapeName=dyn_Cloth_volume_l_A01_capsule, center=<-7.4118 -3 0>, rotation=<rot 90 0 0>}]
 dyn_Cloth_volume_l_A02 [
  {shapeType=Capsule, length=10.5041, radius=3, shapeName=dyn_Cloth_volume_l_A02_capsule, center=<-7.4118 -3 0>, rotation=<rot 90 0 0>}
  {
   shapeType=Capsule
   length=7.083
   radius=5.1344
   shapeName=dyn_Cloth_volume_l_A02_capsule
   center=<-7.6845 3.1217 -7.3066>
   rotation=<rot 90 0 0>
  }
  {
   shapeType=Capsule
   length=9.0204
   radius=7.1161
   shapeName=dyn_Cloth_volume_l_A02_capsule
   center=<-4.5963 -3 10.3317>
   rotation=<rot 0.3782 -180 166.2642>
  }
 ]
 dyn_Cloth_volume_l_B [{shapeType=Capsule, length=0.505, radius=0.505, shapeName=dyn_Cloth_volume_l_B_capsule, rotation=<rot 0 -90 0>}]
 dyn_Cloth_volume_l_B01 [{shapeType=Capsule, length=9.0204, radius=7.1161, shapeName=dyn_Cloth_volume_l_B01_capsule, center=<-7.4118 -3 0>, rotation=<rot 90 0 0>}]
 dyn_Cloth_volume_l_B02 [{shapeType=Capsule, length=9.0204, radius=7.1161, shapeName=dyn_Cloth_volume_l_B02_capsule, center=<-7.4118 -3 0>, rotation=<rot 90 0 0>}]
 dyn_Cloth_volume_l_C [{shapeType=Capsule, length=0.505, radius=0.505, shapeName=dyn_Cloth_volume_l_C_capsule, rotation=<rot 0 -90 0>}]
 dyn_Cloth_volume_l_C01 [{shapeType=Capsule, length=9.0204, radius=7.1161, shapeName=dyn_Cloth_volume_l_C01_capsule, center=<-7.4118 -3 0>, rotation=<rot 90 0 0>}]
 dyn_Cloth_volume_l_C02 [{shapeType=Capsule, length=9.0204, radius=7.1161, shapeName=dyn_Cloth_volume_l_C02_capsule, center=<-7.4118 -3 0>, rotation=<rot 90 0 0>}]
 dyn_Cloth_volume_r_A [{shapeType=Capsule, length=0.505, radius=3, shapeName=dyn_Cloth_volume_r_A_capsule, rotation=<rot 0 -90 0>}]
 dyn_Cloth_volume_r_A01 [{shapeType=Capsule, length=9.0204, radius=3, shapeName=dyn_Cloth_volume_r_A01_capsule, center=<-7.4118 -0.7843 0>, rotation=<rot 90 0 0>}]
 dyn_Cloth_volume_r_A02 [{
   shapeType=Capsule
   length=13.016
   radius=3
   shapeName=dyn_Cloth_volume_r_A02_capsule
   center=<-7.0864 1.2426 -0.3797>
   rotation=<rot -0.0001 0.0001 22.4204>
  }]
 dyn_Cloth_volume_r_B [{shapeType=Capsule, length=0.505, radius=0.505, shapeName=dyn_Cloth_volume_r_B_capsule, rotation=<rot 0 -90 0>}]
 dyn_Cloth_volume_r_B01 [{shapeType=Capsule, length=9.0204, radius=7.1161, shapeName=dyn_Cloth_volume_r_B01_capsule, center=<-7.4118 -3 0>, rotation=<rot 90 0 0>}]
 dyn_Cloth_volume_r_B02 [{shapeType=Capsule, length=9.0204, radius=7.1161, shapeName=dyn_Cloth_volume_r_B02_capsule, center=<-7.4118 -3 0>, rotation=<rot 90 0 0>}]
 dyn_Cloth_volume_r_C [{shapeType=Capsule, length=0.505, radius=0.505, shapeName=dyn_Cloth_volume_r_C_capsule, rotation=<rot 0 -90 0>}]
 dyn_Cloth_volume_r_C01 [{shapeType=Capsule, length=9.0204, radius=7.1161, shapeName=dyn_Cloth_volume_r_C01_capsule, center=<-7.4118 -3 0>, rotation=<rot 90 0 0>}]
 dyn_Cloth_volume_r_C02 [{shapeType=Capsule, length=9.0204, radius=7.1161, shapeName=dyn_Cloth_volume_r_C02_capsule, center=<-7.4118 -3 0>, rotation=<rot 90 0 0>}]
 pelvis [{shapeType=Capsule, length=5.0262, radius=9.3927, shapeName=pelvis_capsule}]
 thigh_twist_01_l [{shapeType=Capsule, length=42.32, radius=6.4576, shapeName=thigh_twist_01_l_capsule, center=<3.0271 0 0>, rotation=<rot 90 87.3911 90.6561>}]
 thigh_twist_01_r [{
   shapeType=Capsule
   length=42.1615
   radius=6.616
   shapeName=thigh_twist_01_l_capsule
   center=<-1.8476 -0.8051 0>
   rotation=<rot 90 180 -176.735>
  }]
}
constraints [
 {bone1Name=dyn_Cloth_volume_bk01, bone2Name=dyn_Cloth_volume_bk, swing1LimitDegrees=45, swing2LimitDegrees=45, twistLimitDegrees=15}
 {bone1Name=dyn_Cloth_volume_bk02, bone2Name=dyn_Cloth_volume_bk01, swing1LimitDegrees=45, swing2LimitDegrees=45, twistLimitDegrees=15}
 {bone1Name=dyn_Cloth_volume_fn01, bone2Name=dyn_Cloth_volume_fn, swing1LimitDegrees=20, swing2LimitDegrees=20, twistLimitDegrees=15}
 {bone1Name=dyn_Cloth_volume_fn02, bone2Name=dyn_Cloth_volume_fn01, swing1LimitDegrees=20, swing2LimitDegrees=20, twistLimitDegrees=15}
 {bone1Name=dyn_Cloth_volume_l_A01, bone2Name=dyn_Cloth_volume_l_A, swing1LimitDegrees=45, swing2LimitDegrees=45, twistLimitDegrees=15}
 {bone1Name=dyn_Cloth_volume_l_A02, bone2Name=dyn_Cloth_volume_l_A01, swing1LimitDegrees=45, swing2LimitDegrees=45, twistLimitDegrees=15}
 {bone1Name=dyn_Cloth_volume_l_B01, bone2Name=dyn_Cloth_volume_l_B, swing1LimitDegrees=45, swing2LimitDegrees=45, twistLimitDegrees=15}
 {bone1Name=dyn_Cloth_volume_l_B02, bone2Name=dyn_Cloth_volume_l_B01, swing1LimitDegrees=45, swing2LimitDegrees=45, twistLimitDegrees=15}
 {bone1Name=dyn_Cloth_volume_l_C01, bone2Name=dyn_Cloth_volume_l_C, swing1LimitDegrees=45, swing2LimitDegrees=45, twistLimitDegrees=15}
 {bone1Name=dyn_Cloth_volume_l_C02, bone2Name=dyn_Cloth_volume_l_C01, swing1LimitDegrees=45, swing2LimitDegrees=45, twistLimitDegrees=15}
 {bone1Name=dyn_Cloth_volume_r_A02, bone2Name=dyn_Cloth_volume_r_A01, swing1LimitDegrees=20, swing2LimitDegrees=20, twistLimitDegrees=15}
 {bone1Name=dyn_Cloth_volume_r_A01, bone2Name=dyn_Cloth_volume_r_A, swing1LimitDegrees=20, swing2LimitDegrees=20, twistLimitDegrees=15}
 {bone1Name=dyn_Cloth_volume_r_B01, bone2Name=dyn_Cloth_volume_r_B, swing1LimitDegrees=45, swing2LimitDegrees=45, twistLimitDegrees=15}
 {bone1Name=dyn_Cloth_volume_r_B02, bone2Name=dyn_Cloth_volume_r_B01, swing1LimitDegrees=45, swing2LimitDegrees=45, twistLimitDegrees=15}
 {bone1Name=dyn_Cloth_volume_r_C01, bone2Name=dyn_Cloth_volume_r_C, swing1LimitDegrees=45, swing2LimitDegrees=45, twistLimitDegrees=15}
 {bone1Name=dyn_Cloth_volume_r_C02, bone2Name=dyn_Cloth_volume_r_C01, swing1LimitDegrees=45, swing2LimitDegrees=45, twistLimitDegrees=15}
 {
  bone1Name=dyn_Cloth_volume_fn
  bone2Name=pelvis
  swing1LimitDegrees=45
  swing1Motion=Free
  swing2LimitDegrees=45
  swing2Motion=Free
  twistLimitDegrees=45
  twistMotion=Free
 }
 {
  bone1Name=dyn_Cloth_volume_bk
  bone2Name=pelvis
  swing1LimitDegrees=45
  swing1Motion=Free
  swing2LimitDegrees=45
  swing2Motion=Free
  twistLimitDegrees=45
  twistMotion=Free
 }
 {
  bone1Name=dyn_Cloth_volume_l_A
  bone2Name=pelvis
  swing1LimitDegrees=45
  swing1Motion=Free
  swing2LimitDegrees=45
  swing2Motion=Free
  twistLimitDegrees=45
  twistMotion=Free
 }
 {
  bone1Name=dyn_Cloth_volume_l_B
  bone2Name=pelvis
  swing1LimitDegrees=45
  swing1Motion=Free
  swing2LimitDegrees=45
  swing2Motion=Free
  twistLimitDegrees=45
  twistMotion=Free
 }
 {
  bone1Name=dyn_Cloth_volume_l_C
  bone2Name=pelvis
  swing1LimitDegrees=45
  swing1Motion=Free
  swing2LimitDegrees=45
  swing2Motion=Free
  twistLimitDegrees=45
  twistMotion=Free
 }
 {
  bone1Name=dyn_Cloth_volume_r_A
  bone2Name=pelvis
  swing1LimitDegrees=45
  swing1Motion=Free
  swing2LimitDegrees=45
  swing2Motion=Free
  twistLimitDegrees=45
  twistMotion=Free
 }
 {
  bone1Name=dyn_Cloth_volume_r_B
  bone2Name=pelvis
  swing1LimitDegrees=45
  swing1Motion=Free
  swing2LimitDegrees=45
  swing2Motion=Free
  twistLimitDegrees=45
  twistMotion=Free
 }
 {
  bone1Name=dyn_Cloth_volume_r_C
  bone2Name=pelvis
  swing1LimitDegrees=45
  swing1Motion=Free
  swing2LimitDegrees=45
  swing2Motion=Free
  twistLimitDegrees=45
  twistMotion=Free
 }
 {
  bone1Name=thigh_twist_01_l
  bone2Name=pelvis
  swing1LimitDegrees=45
  swing1Motion=Free
  swing2LimitDegrees=45
  swing2Motion=Free
  twistLimitDegrees=45
  twistMotion=Free
 }
 {
  bone1Name=thigh_twist_01_r
  bone2Name=pelvis
  swing1LimitDegrees=45
  swing1Motion=Free
  swing2LimitDegrees=45
  swing2Motion=Free
  twistLimitDegrees=45
  twistMotion=Free
 }
]
