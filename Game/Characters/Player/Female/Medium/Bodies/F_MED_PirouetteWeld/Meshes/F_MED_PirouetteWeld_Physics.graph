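# Physics asset for the PirouetteWeld medium female player body, defining capsule bodies for a two-sided dynamic ribbon on the character.
# A master ribbon body attaches to spine_05, which has a large zero-length capsule, through a constraint with free swing and twist.
# The left and right ribbons each have two segments, all constrained directly to the master instead of chained, with very tight twist and wider first-axis swing on the outer segments.
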PhysicsAsset /Game/Characters/Player/Female/Medium/Bodies/F_MED_PirouetteWeld/Meshes/F_MED_PirouetteWeld_Physics
bodies {
 dyn_ribbon_MSTR [{shapeType=Capsule, length=0.505, radius=0.505, shapeName=dyn_ribbon_MSTR_capsule, rotation=<rot 0 -90 0>}]
 dyn_ribbon_l_01 [{shapeType=Capsule, length=1.398, radius=0.931, shapeName=dyn_ribbon_l_01_capsule, center=<1 0 0>, rotation=<rot 90 0 0>}]
 dyn_ribbon_l_02 [{shapeType=Capsule, length=4.6503, radius=0.931, shapeName=dyn_ribbon_l_02_capsule, center=<2.7676 0 0>, rotation=<rot 90 0 0>}]
 dyn_ribbon_r_01 [{shapeType=Capsule, length=1.398, radius=0.931, shapeName=dyn_ribbon_r_01_capsule, center=<1 0 0>, rotation=<rot 90 0 0>}]
 dyn_ribbon_r_02 [{shapeType=Capsule, length=4.6503, radius=0.931, shapeName=dyn_ribbon_r_02_capsule, center=<2.7676 0 0>, rotation=<rot 90 0 0>}]
 spine_05 [{shapeType=Capsule, length=0, radius=10.5954, shapeName=spine_05_capsule, center=<0.1358 -0.94 0>}]
}
constraints [
 {
  bone1Name=dyn_ribbon_MSTR
  bone2Name=spine_05
  swing1LimitDegrees=45
  swing1Motion=Free
  swing2LimitDegrees=45
  swing2Motion=Free
  twistLimitDegrees=45
  twistMotion=Free
 }
 {bone1Name=dyn_ribbon_l_01, bone2Name=dyn_ribbon_MSTR, swing1LimitDegrees=10, swing2LimitDegrees=10, twistLimitDegrees=0.5}
 {bone1Name=dyn_ribbon_l_02, bone2Name=dyn_ribbon_MSTR, swing1LimitDegrees=30, swing2LimitDegrees=10, twistLimitDegrees=0.5}
 {bone1Name=dyn_ribbon_r_01, bone2Name=dyn_ribbon_MSTR, swing1LimitDegrees=10, swing2LimitDegrees=10, twistLimitDegrees=0.5}
 {bone1Name=dyn_ribbon_r_02, bone2Name=dyn_ribbon_MSTR, swing1LimitDegrees=30, swing2LimitDegrees=10, twistLimitDegrees=0.5}
]
