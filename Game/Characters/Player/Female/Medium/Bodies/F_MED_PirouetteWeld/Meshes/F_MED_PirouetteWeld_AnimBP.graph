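# Animation Blueprint for the F_MED_PirouetteWeld female body, adding secondary motion to its skirt and corrective motion to shoulders, arms and knees.
# Poses copied from the parent and head meshes pass through constraints, Look Ats and a Control Rig into a cached pose.
# Skirt bone chains get additive offsets and AnimDynamics springs.
# The output switches between skydive-aware gameplay branches and a front-end Control Rig plus RigidBody branch, then offsets the knees.
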AnimBlueprint /Game/Characters/Player/Female/Medium/Bodies/F_MED_PirouetteWeld/Meshes/F_MED_PirouetteWeld_AnimBP
target_skeleton=&/F_MED_PirouetteWeld_Skeleton
graphs [
 AnimGraph {nodes [
   RigidBody!(ComponentPose, Alpha=0.3, OverrideWorldGravity:Vector="0,0,-980", ExternalForce) -> (Pose) {
    bClampLinearTranslationLimitToRefPose=true
    bOverrideWorldGravity=true
    overridePhysicsAsset=&/F_MED_PirouetteWeld_Physics
    simulationSpace=BaseBoneSpace
    alphaBoolBlend {blendOption=Linear}
    baseBoneRef {boneName=root}
    componentAppliedLinearAccClamp=<3000 3000 3000>
    componentLinearAccScale=<0.5 0.3 0.3>
    componentLinearVelScale=<0.5 0.3 0.3>
    overrideWorldGravity {z=-980}
   }
   n1 = Use cached pose 'SkirtCachePose'()
   n2 = Use cached pose 'SkirtDynPose'()
   n3 = "Transform (Modify) Bone - Bone: dyn_Cloth_volume_bk"!(
    ComponentPose=Local To Component(LocalPose=Two Way Blend(A=@n2, B=@n1, Alpha=0.7) {blendNode {alphaBoolBlend {blendOption=Linear}}})
    bAlphaBoolEnabled:bool=True
    Translation="0,0,0"
    Rotation="0,-15,0"
    Scale="1,1,1"
   ) -> (Pose) {
    alphaInputType=Bool
    rotationMode=BMM_Additive
    rotationSpace=BCS_BoneSpace
    scaleMode=BMM_Additive
    scaleSpace=BCS_BoneSpace
    translationMode=BMM_Additive
    translationSpace=BCS_BoneSpace
    alphaBoolBlend {blendInTime=0.1, blendOption=Linear, blendOutTime=0.1}
    boneToModify {boneName=dyn_Cloth_volume_bk}
    rotation {yaw=-15}
   }
   n4 = "Transform (Modify) Bone - Bone: dyn_Cloth_volume_l_C"!(ComponentPose=@n3, bAlphaBoolEnabled:bool=True, Translation="0,0,0", Rotation="0,-15,0", Scale="1,1,1") -> (Pose) {
    alphaInputType=Bool
    rotationMode=BMM_Additive
    rotationSpace=BCS_BoneSpace
    scaleMode=BMM_Additive
    scaleSpace=BCS_BoneSpace
    translationMode=BMM_Additive
    translationSpace=BCS_BoneSpace
    alphaBoolBlend {blendInTime=0.1, blendOption=Linear, blendOutTime=0.1}
    boneToModify {boneName=dyn_Cloth_volume_l_C}
    rotation {yaw=-15}
   }
   n5 = "Transform (Modify) Bone - Bone: dyn_Cloth_volume_r_C"!(ComponentPose=@n4, bAlphaBoolEnabled:bool=True, Translation="0,0,0", Rotation="0,-15,0", Scale="1,1,1") -> (Pose) {
    alphaInputType=Bool
    rotationMode=BMM_Additive
    rotationSpace=BCS_BoneSpace
    scaleMode=BMM_Additive
    scaleSpace=BCS_BoneSpace
    translationMode=BMM_Additive
    translationSpace=BCS_BoneSpace
    alphaBoolBlend {blendInTime=0.1, blendOption=Linear, blendOutTime=0.1}
    boneToModify {boneName=dyn_Cloth_volume_r_C}
    rotation {yaw=-15}
   }
   n6 = "Transform (Modify) Bone - Bone: dyn_l_shoulder"!(ComponentPose=@n5, bAlphaBoolEnabled:bool=True, Translation="0,0,0", Rotation="10,0.1,-54", Scale="1,1,1") -> (Pose) {
    alphaInputType=Bool
    rotationMode=BMM_Additive
    rotationSpace=BCS_BoneSpace
    scaleMode=BMM_Additive
    scaleSpace=BCS_BoneSpace
    translationMode=BMM_Additive
    translationSpace=BCS_BoneSpace
    alphaBoolBlend {blendOption=Linear}
    boneToModify {boneName=dyn_l_shoulder}
    rotation=<rot 10 0.1 -54>
   }
   n7 = "Transform (Modify) Bone - Bone: dyn_r_shoulder"!(ComponentPose=@n6, bAlphaBoolEnabled:bool=True, Translation="0,0,0", Rotation="10,-8,-70", Scale="1,1,1") -> (Pose) {
    alphaInputType=Bool
    rotationMode=BMM_Additive
    rotationSpace=BCS_BoneSpace
    scaleMode=BMM_Additive
    scaleSpace=BCS_BoneSpace
    translationMode=BMM_Additive
    translationSpace=BCS_BoneSpace
    alphaBoolBlend {blendOption=Linear}
    boneToModify {boneName=dyn_r_shoulder}
    rotation=<rot 10 -8 -70>
   }
   n8 = "Constraint - dyn_l_shoulder (P)"(ComponentPose=@n7, Alpha=1, ConstraintWeights_0=1) {
    alphaBoolBlend {blendInTime=0.1, blendOption=Linear, blendOutTime=1}
    alphaScaleBiasClamp {bInterpResult=true, interpSpeedDecreasing=1, interpSpeedIncreasing=1}
    boneToModify {boneName=dyn_l_shoulder}
    constraintSetup [{targetBone {boneName=upperarm_l}}]
    constraintWeights [1]
   }
   n9 = "Constraint - dyn_r_shoulder (P)"(ComponentPose=@n8, Alpha=1, ConstraintWeights_0=1) {
    alphaBoolBlend {blendInTime=0.1, blendOption=Linear, blendOutTime=1}
    alphaScaleBiasClamp {bInterpResult=true, interpSpeedDecreasing=1, interpSpeedIncreasing=1}
    boneToModify {boneName=dyn_r_shoulder}
    constraintSetup [{targetBone {boneName=upperarm_r}}]
    constraintWeights [1]
   }
   n10 = "Constraint - dyn_l_shoulder (P)"!(ComponentPose=@n9, bAlphaBoolEnabled:bool=True, ConstraintWeights_0=1) -> (Pose) {
    alphaInputType=Bool
    alphaBoolBlend {blendInTime=0.1, blendOption=Linear, blendOutTime=1}
    alphaScaleBiasClamp {bInterpResult=true, interpSpeedDecreasing=1, interpSpeedIncreasing=1}
    boneToModify {boneName=dyn_l_shoulder}
    constraintSetup [{targetBone {boneName=upperarm_l}}]
    constraintWeights [1]
   }
   n11 = "Constraint - dyn_r_shoulder (P)"!(ComponentPose=@n10, bAlphaBoolEnabled:bool=True, ConstraintWeights_0=1) -> (Pose) {
    alphaInputType=Bool
    alphaBoolBlend {blendInTime=0.1, blendOption=Linear, blendOutTime=1}
    alphaScaleBiasClamp {bInterpResult=true, interpSpeedDecreasing=1, interpSpeedIncreasing=1}
    boneToModify {boneName=dyn_r_shoulder}
    constraintSetup [{targetBone {boneName=upperarm_r}}]
    constraintWeights [1]
   }
   n12 = "Transform (Modify) Bone - Bone: dyn_Cloth_volume_bk"(
    ComponentPose=Local To Component(LocalPose=Use cached pose 'SkirtDynPose'())
    bAlphaBoolEnabled=True
    Translation="0,0,0"
    Rotation="0,-15,0"
    Scale="1,1,1"
   ) {
    alphaInputType=Bool
    rotationMode=BMM_Additive
    rotationSpace=BCS_BoneSpace
    scaleMode=BMM_Additive
    scaleSpace=BCS_BoneSpace
    translationMode=BMM_Additive
    translationSpace=BCS_BoneSpace
    alphaBoolBlend {blendInTime=0.1, blendOption=Linear, blendOutTime=0.1}
    boneToModify {boneName=dyn_Cloth_volume_bk}
    rotation {yaw=-15}
   }
   n13 = "Transform (Modify) Bone - Bone: dyn_Cloth_volume_l_C"(ComponentPose=@n12, bAlphaBoolEnabled=True, Translation="0,0,0", Rotation="0,-15,0", Scale="1,1,1") {
    alphaInputType=Bool
    rotationMode=BMM_Additive
    rotationSpace=BCS_BoneSpace
    scaleMode=BMM_Additive
    scaleSpace=BCS_BoneSpace
    translationMode=BMM_Additive
    translationSpace=BCS_BoneSpace
    alphaBoolBlend {blendInTime=0.1, blendOption=Linear, blendOutTime=0.1}
    boneToModify {boneName=dyn_Cloth_volume_l_C}
    rotation {yaw=-15}
   }
   n14 = "Transform (Modify) Bone - Bone: dyn_Cloth_volume_r_C"(ComponentPose=@n13, bAlphaBoolEnabled=True, Translation="0,0,0", Rotation="0,-15,0", Scale="1,1,1") {
    alphaInputType=Bool
    rotationMode=BMM_Additive
    rotationSpace=BCS_BoneSpace
    scaleMode=BMM_Additive
    scaleSpace=BCS_BoneSpace
    translationMode=BMM_Additive
    translationSpace=BCS_BoneSpace
    alphaBoolBlend {blendInTime=0.1, blendOption=Linear, blendOutTime=0.1}
    boneToModify {boneName=dyn_Cloth_volume_r_C}
    rotation {yaw=-15}
   }
   n15 = Get bIsSkydiveDiveMode()
   n16 = Get bIsSkydiveFloating()
   n17 = Blend Poses by bool(
    BlendPose_0=Blend Poses by bool(
     BlendPose_0=Component To Local(ComponentPose=@n11)
     BlendPose_1=Component To Local(ComponentPose=@n14)
     bActiveValue=OR Boolean!(A=@n15, B=@n16, C:bool) -> (ReturnValue)
    ) {blendType=Linear}
    BlendPose_1=Use cached pose 'SkirtCachePose'()
   ) {blendType=Linear}
   n18 = Control Rig!(
    Source=Use cached pose 'SkirtDynPose'()
    Alpha
    Pickaxe:double=0
    Rifle:double=0
    Pistol:double=0
    Skydive:double=0
    ParachuteAfterSkydive:double=0
    FroneEnd:double=Get Frontend() -> (Frontend:double)
   ) -> (Pose) {
    alphaBoolBlend {blendOption=Linear}
    controlRigAssetReference {blueprintRigClass=&/F_MED_PirouetteWeld_CtrlRig.F_MED_PirouetteWeld_CtrlRig_C}
   }
   n19 = "Transform (Modify) Bone - Bone: dyn_Cloth_volume_r_A"!(ComponentPose=Local To Component(LocalPose=@n18), bAlphaBoolEnabled:bool=True, Translation="0,0,0", Rotation="20,0,20", Scale="1,1,1") -> (Pose) {
    alphaInputType=Bool
    rotationMode=BMM_Additive
    alphaBoolBlend {blendOption=Linear}
    boneToModify {boneName=dyn_Cloth_volume_r_A}
    rotation {pitch=20, roll=20}
   }
   n20 = "Transform (Modify) Bone - Bone: dyn_Cloth_volume_fn"!(ComponentPose=@n19, bAlphaBoolEnabled:bool=True, Translation="0,0,0", Rotation="10,0,20", Scale="1,1,1") -> (Pose) {
    alphaInputType=Bool
    rotationMode=BMM_Additive
    alphaBoolBlend {blendOption=Linear}
    boneToModify {boneName=dyn_Cloth_volume_fn}
    rotation {pitch=10, roll=20}
   }
   n21 = "Transform (Modify) Bone - Bone: dyn_Cloth_volume_l_A"!(ComponentPose=@n20, bAlphaBoolEnabled:bool=True, Translation="0,0,0", Rotation="-20,0,20", Scale="1,1,1") -> (Pose) {
    alphaInputType=Bool
    rotationMode=BMM_Additive
    alphaBoolBlend {blendOption=Linear}
    boneToModify {boneName=dyn_Cloth_volume_l_A}
    rotation {pitch=-20, roll=20}
   }
   n22 = "Transform (Modify) Bone - Bone: dyn_Cloth_volume_l_B"!(ComponentPose=@n21, bAlphaBoolEnabled:bool=True, Translation="0,-2,0", Rotation="-30,0,0", Scale="1,1,1") -> (Pose) {
    alphaInputType=Bool
    rotationMode=BMM_Additive
    translationMode=BMM_Additive
    translationSpace=BCS_BoneSpace
    alphaBoolBlend {blendOption=Linear}
    boneToModify {boneName=dyn_Cloth_volume_l_B}
    rotation {pitch=-30}
    translation {y=-2}
   }
   n23 = "Transform (Modify) Bone - Bone: dyn_Cloth_volume_r_B"!(ComponentPose=@n22, bAlphaBoolEnabled:bool=True, Translation="0,0,0", Rotation="20,0,0", Scale="1,1,1") -> (Pose) {
    alphaInputType=Bool
    rotationMode=BMM_Additive
    alphaBoolBlend {blendOption=Linear}
    boneToModify {boneName=dyn_Cloth_volume_r_B}
    rotation {pitch=20}
   }
   n24 = "Transform (Modify) Bone - Bone: dyn_Cloth_volume_bk"(ComponentPose=@n23, bAlphaBoolEnabled=True, Translation="0,0,0", Rotation="-10,0,-30", Scale="1,1,1") {
    alphaInputType=Bool
    rotationMode=BMM_Additive
    alphaBoolBlend {blendOption=Linear}
    boneToModify {boneName=dyn_Cloth_volume_bk}
    rotation {pitch=-10, roll=-30}
   }
   n25 = "Transform (Modify) Bone - Bone: dyn_Cloth_volume_bk"(ComponentPose=@n24, bAlphaBoolEnabled=True, Translation="0,0,0", Rotation="2.292,-6.413,9.312", Scale="1,1,1") {
    alphaInputType=Bool
    rotationMode=BMM_Additive
    alphaBoolBlend {blendOption=Linear}
    boneToModify {boneName=dyn_Cloth_volume_bk}
    rotation=<rot 2.292 -6.413 9.312>
   }
   n26 = "Transform (Modify) Bone - Bone: dyn_Cloth_volume_r_C"(ComponentPose=@n25, bAlphaBoolEnabled=True, Translation="0,0,0", Rotation="9.391,-3.450,-20.284", Scale="1,1,1") {
    alphaInputType=Bool
    rotationMode=BMM_Additive
    alphaBoolBlend {blendOption=Linear}
    boneToModify {boneName=dyn_Cloth_volume_r_C}
    rotation=<rot 9.391 -3.45 -20.284>
   }
   n27 = "Transform (Modify) Bone - Bone: dyn_Cloth_volume_l_C"(ComponentPose=@n26, bAlphaBoolEnabled=True, Translation="0,0,0", Rotation="-10.000,0.000,-0.000", Scale="1,1,1") {
    alphaInputType=Bool
    rotationMode=BMM_Additive
    alphaBoolBlend {blendOption=Linear}
    boneToModify {boneName=dyn_Cloth_volume_l_C}
    rotation {pitch=-10}
   }
   n28 = RigidBody(ComponentPose=@n27, OverrideWorldGravity="0,0,-980") {
    bClampLinearTranslationLimitToRefPose=true
    bOverrideWorldGravity=true
    overridePhysicsAsset=&/F_MED_PirouetteWeld_FrontEnd_Physics
    simulationSpace=BaseBoneSpace
    alphaBoolBlend {blendOption=Linear}
    baseBoneRef {boneName=pelvis}
    componentAppliedLinearAccClamp=<5000 5000 5000>
    componentLinearAccScale=<0.35 0.35 0.4>
    overrideWorldGravity {z=-980}
    simSpaceSettings {
     maxAngularAcceleration=1000
     maxAngularVelocity=3
     maxLinearAcceleration=700
     maxLinearVelocity=500
     velocityScaleZ=0
     worldAlpha=0.25
     externalLinearDragV=<0.2 0.2>
    }
   }
   n29 = Component To Local(ComponentPose=@n28)
   n30 = Get bIsInFrontEnd()
   n31 = "Transform (Modify) Bone - Bone: dyn_l_knee"!(
    ComponentPose=Local To Component(LocalPose=Blend Poses by bool(BlendPose_0=@n29, BlendPose_1=@n17, bActiveValue=@n30) {blendType=Linear})
    bAlphaBoolEnabled:bool=True
    Translation="0,-2,0"
    Rotation="0,0,0"
    Scale="1,1,1"
   ) -> (Pose) {
    alphaInputType=Bool
    rotationMode=BMM_Additive
    rotationSpace=BCS_BoneSpace
    scaleMode=BMM_Additive
    scaleSpace=BCS_BoneSpace
    translationMode=BMM_Additive
    translationSpace=BCS_BoneSpace
    alphaBoolBlend {blendOption=Linear}
    boneToModify {boneName=dyn_l_knee}
    translation {y=-2}
   }
   n32 = "Transform (Modify) Bone - Bone: dyn_r_knee"!(ComponentPose=@n31, bAlphaBoolEnabled:bool=True, Translation="0,2,0", Rotation="0,0,0", Scale="1,1,1") -> (Pose) {
    alphaInputType=Bool
    rotationMode=BMM_Additive
    rotationSpace=BCS_BoneSpace
    scaleMode=BMM_Additive
    scaleSpace=BCS_BoneSpace
    translationMode=BMM_Additive
    translationSpace=BCS_BoneSpace
    alphaBoolBlend {blendOption=Linear}
    boneToModify {boneName=dyn_r_knee}
    translation {y=2}
   }
   Output Pose(Result=Component To Local(ComponentPose=@n32)) {layerGroup=None}
   n33 = Layered blend per bone(
    BasePose=Copy Pose From Mesh() {bCopyCurves=true, bUseAttachedParent=true}
    BlendPoses_0=Copy Pose From Mesh(SourceMeshComponent=Get head_mesh() -> (head_mesh:Skeletal Mesh Component Object Reference)) {bCopyCurves=true, bUseAttachedParent=true}
   ) {layerSetup [{branchFilters [
       {boneName=FACIAL_C_Jaw}
       {boneName=FACIAL_L_CheekLower}
       {boneName=FACIAL_L_CheekOuter}
       {boneName=FACIAL_R_CheekLower}
       {boneName=FACIAL_R_CheekOuter}
       {boneName=FACIAL_L_NasolabialBulge}
       {boneName=FACIAL_R_NasolabialBulge}
       {boneName=FACIAL_L_LipCorner}
       {boneName=FACIAL_R_LipCorner}
       {boneName=FACIAL_L_Masseter}
       {boneName=FACIAL_R_Masseter}
       {boneName=FACIAL_C_Chin}
       {boneName=FACIAL_L_ChinSide}
       {boneName=FACIAL_R_ChinSide}
       {boneName=FACIAL_C_UnderChin}
       {boneName=FACIAL_L_UnderChin}
       {boneName=FACIAL_R_UnderChin}
       {boneName=FACIAL_C_LowerLipRotation}
       {boneName=FACIAL_C_LipLower}
       {boneName=FACIAL_L_LipLower}
       {boneName=FACIAL_L_LipLowerOuter}
       {boneName=FACIAL_R_LipLower}
       {boneName=FACIAL_R_LipLowerOuter}
       {boneName=FACIAL_L_LipCorner}
       {boneName=FACIAL_R_LipCorner}
       {boneName=FACIAL_C_Jawline}
       {boneName=FACIAL_L_Jawline}
       {boneName=FACIAL_R_Jawline}
       {boneName=FACIAL_L_Ear}
       {boneName=FACIAL_R_Ear}
      ]}]}
   n34 = "Constraint - dyn_l_shoulder (R)"(ComponentPose=Local To Component(LocalPose=@n33), Alpha=1, ConstraintWeights_0=1) {
    alphaBoolBlend {blendOption=Linear}
    boneToModify {boneName=dyn_l_shoulder}
    constraintSetup [{transformType=Rotation, targetBone {boneName=dyn_arm_l}}]
    constraintWeights [1]
   }
   n35 = "Constraint - dyn_r_shoulder (R)"(ComponentPose=@n34, Alpha=1, ConstraintWeights_0=1) {
    alphaBoolBlend {blendOption=Linear}
    boneToModify {boneName=dyn_r_shoulder}
    constraintSetup [{transformType=Rotation, targetBone {boneName=dyn_arm_r}}]
    constraintWeights [1]
   }
   n36 = Control Rig!(
    Source=Component To Local(ComponentPose=@n35)
    Alpha
    Pickaxe=Get isPickaxe() -> (isPickaxe:double)
    Rifle=Get isAimingRifle() -> (isAimingRifle:double)
    Pistol=Get isAimingPistol() -> (isAimingPistol:double)
    Skydive=Get isSkyDiveNew() -> (isSkyDiveNew:double)
    ParachuteAfterSkydive=Get ParachutingFromSkydive() -> (ParachutingFromSkydive:double)
   ) -> (Pose) {
    alphaBoolBlend {blendOption=Linear}
    controlRigAssetReference {blueprintRigClass=&F_MED_PirouetteWeld_CtrlRig_C}
   }
   n37 = "Constraint - dyn_arm_l (R)"(ComponentPose=Local To Component(LocalPose=@n36), Alpha=1, ConstraintWeights_0=1) {
    alphaBoolBlend {blendOption=Linear}
    boneToModify {boneName=dyn_arm_l}
    constraintSetup [{transformType=Rotation, targetBone {boneName=upperarm_l}}]
    constraintWeights [1]
   }
   n38 = "Constraint - dyn_arm_r (R)"(ComponentPose=@n37, Alpha=1, ConstraintWeights_0=1) {
    alphaBoolBlend {blendOption=Linear}
    boneToModify {boneName=dyn_arm_r}
    constraintSetup [{transformType=Rotation, targetBone {boneName=upperarm_r}}]
    constraintWeights [1]
   }
   n39 = "Constraint - dyn_arm_aim_l (PP)"!(ComponentPose=@n38, Alpha=1, ConstraintWeights_0=0.3, ConstraintWeights_1:float=0.7) -> (Pose) {
    alphaBoolBlend {blendOption=Linear}
    boneToModify {boneName=dyn_arm_aim_l}
    constraintSetup [{targetBone {boneName=upperarm_l}}, {targetBone {boneName=dyn_arm_l}}]
    constraintWeights [1,1]
   }
   n40 = "Constraint - dyn_arm_aim_r (PP)"!(ComponentPose=@n39, Alpha=1, ConstraintWeights_0=0.3, ConstraintWeights_1:float=0.7) -> (Pose) {
    alphaBoolBlend {blendOption=Linear}
    boneToModify {boneName=dyn_arm_aim_r}
    constraintSetup [{targetBone {boneName=upperarm_r}}, {targetBone {boneName=dyn_arm_r}}]
    constraintWeights [1,1]
   }
   n41 = "Constraint - dyn_lowerarm_base_l_02 (PP)"!(ComponentPose=@n40, Alpha=1, ConstraintWeights_0=0.7, ConstraintWeights_1:float=0.3) -> (Pose) {
    alphaBoolBlend {blendOption=Linear}
    boneToModify {boneName=dyn_lowerarm_base_l_02}
    constraintSetup [{targetBone {boneName=lowerarm_l}}, {targetBone {boneName=dyn_arm_l}}]
    constraintWeights [1,1]
   }
   n42 = "Constraint - dyn_lowerarm_base_r (PP)"!(ComponentPose=@n41, Alpha=1, ConstraintWeights_0=0.7, ConstraintWeights_1:float=0.3) -> (Pose) {
    alphaBoolBlend {blendOption=Linear}
    boneToModify {boneName=dyn_lowerarm_base_r}
    constraintSetup [{targetBone {boneName=lowerarm_r}}, {targetBone {boneName=dyn_arm_r}}]
    constraintWeights [1,1]
   }
   n43 = Look At - Bone: dyn_lowerarm_base_l_02(ComponentPose=@n42, Alpha=1) {
    alphaBoolBlend {blendOption=Linear}
    boneToModify {boneName=dyn_lowerarm_base_l_02}
    lookAtLocation {x=0}
    lookAtTarget {boneReference {boneName=hand_l}}
    lookAt_Axis {axis=<1 0>}
   }
   n44 = Look At - Bone: dyn_lowerarm_base_r(ComponentPose=@n43, Alpha=1) {
    alphaBoolBlend {blendOption=Linear}
    boneToModify {boneName=dyn_lowerarm_base_r}
    lookAtLocation {x=-2}
    lookAtTarget {boneReference {boneName=hand_r}}
    lookAt_Axis {axis=<-1 0>}
   }
   n45 = "Constraint - dyn_l_knee (PP)"!(ComponentPose=@n44, Alpha=1, ConstraintWeights_0=0.3, ConstraintWeights_1:float=0.7) -> (Pose) {
    alphaBoolBlend {blendOption=Linear}
    boneToModify {boneName=dyn_l_knee}
    constraintSetup [{targetBone {boneName=thigh_l}}, {targetBone {boneName=calf_l}}]
    constraintWeights [1,1]
   }
   n46 = "Constraint - dyn_r_knee (PP)"!(ComponentPose=@n45, Alpha=1, ConstraintWeights_0=0.3, ConstraintWeights_1:float=0.7) -> (Pose) {
    alphaBoolBlend {blendOption=Linear}
    boneToModify {boneName=dyn_r_knee}
    constraintSetup [{targetBone {boneName=thigh_r}}, {targetBone {boneName=calf_r}}]
    constraintWeights [1,1]
   }
   n47 = "Constraint - dyn_ribbon_MSTR (PPPPPP)"!(
    ComponentPose=@n46
    Alpha=1
    ConstraintWeights_0=0
    ConstraintWeights_1:float=1
    ConstraintWeights_2:float=0
    ConstraintWeights_3:float=0
    ConstraintWeights_4:float=0
    ConstraintWeights_5:float=0
   ) -> (Pose) {
    alphaBoolBlend {blendOption=Linear}
    boneToModify {boneName=dyn_ribbon_MSTR}
    constraintSetup [
     {targetBone {boneName=spine_04}}
     {targetBone {boneName=spine_05}}
     {targetBone {boneName=clavicle_l}}
     {targetBone {boneName=neck_01}}
     {targetBone {boneName=head}}
     {targetBone {boneName=neck_02}}
    ]
    constraintWeights [1,1,1,1,1,1]
   }
   n48 = "Transform (Modify) Bone - Bone: dyn_l_shoulder"!(ComponentPose=@n47, Alpha:float=1, Translation="0,0,0", Rotation="2.218,-0.798,-11.916", Scale="1,1,1") -> (Pose) {
    rotationMode=BMM_Additive
    rotationSpace=BCS_BoneSpace
    scaleMode=BMM_Additive
    scaleSpace=BCS_BoneSpace
    translationMode=BMM_Additive
    translationSpace=BCS_BoneSpace
    alphaBoolBlend {blendOption=Linear}
    boneToModify {boneName=dyn_l_shoulder}
    rotation=<rot 2.218 -0.798 -11.916>
   }
   n49 = "Transform (Modify) Bone - Bone: dyn_r_shoulder"!(ComponentPose=@n48, Alpha:float=1, Translation="0,-0.5,0", Rotation="3.434,0.217,18.906", Scale="1,1,1") -> (Pose) {
    rotationMode=BMM_Additive
    rotationSpace=BCS_BoneSpace
    scaleMode=BMM_Additive
    scaleSpace=BCS_BoneSpace
    translationMode=BMM_Additive
    translationSpace=BCS_BoneSpace
    alphaBoolBlend {blendOption=Linear}
    boneToModify {boneName=dyn_r_shoulder}
    rotation=<rot 3.434 0.217 18.906>
    translation {y=-0.5}
   }
   n50 = "Transform (Modify) Bone - Bone: dyn_l_shoulder"!(ComponentPose=@n49, bAlphaBoolEnabled:bool=True, Translation="0,0,-1.5", Rotation="0.545,-0.232,-0.799", Scale="1,1,1") -> (Pose) {
    alphaInputType=Bool
    rotationMode=BMM_Additive
    rotationSpace=BCS_BoneSpace
    scaleMode=BMM_Additive
    scaleSpace=BCS_BoneSpace
    translationMode=BMM_Additive
    translationSpace=BCS_BoneSpace
    alphaBoolBlend {blendInTime=0.1, blendOption=Linear, blendOutTime=0.1}
    boneToModify {boneName=dyn_l_shoulder}
    rotation=<rot 0.545 -0.232 -0.799>
    translation {z=-1.5}
   }
   n51 = "Transform (Modify) Bone - Bone: dyn_r_shoulder"!(ComponentPose=@n50, bAlphaBoolEnabled:bool=True, Translation="0,0,1.5", Rotation="0.001,-0.394,-28.387", Scale="1,1,1") -> (Pose) {
    alphaInputType=Bool
    rotationMode=BMM_Additive
    rotationSpace=BCS_BoneSpace
    scaleMode=BMM_Additive
    scaleSpace=BCS_BoneSpace
    translationMode=BMM_Additive
    translationSpace=BCS_BoneSpace
    alphaBoolBlend {blendInTime=0.1, blendOption=Linear, blendOutTime=0.1}
    boneToModify {boneName=dyn_r_shoulder}
    rotation=<rot 0.001 -0.394 -28.387>
    translation {z=1.5}
   }
   Save cached pose 'SkirtCachePose'(Pose=Component To Local(ComponentPose=@n51)) {cacheName=SkirtCachePose}
   n52 = "Transform (Modify) Bone - Bone: dyn_Cloth_volume_l_A"!(
    ComponentPose=Local To Component(LocalPose=Use cached pose 'SkirtCachePose'())
    Alpha:float=1
    Translation="0,0,0"
    Rotation="0,10,0"
    Scale="1,1,1"
   ) -> (Pose) {
    rotationMode=BMM_Additive
    rotationSpace=BCS_BoneSpace
    scaleMode=BMM_Additive
    scaleSpace=BCS_BoneSpace
    translationMode=BMM_Additive
    translationSpace=BCS_BoneSpace
    alphaBoolBlend {blendOption=Linear}
    boneToModify {boneName=dyn_Cloth_volume_l_A}
    rotation {yaw=10}
   }
   n53 = "Transform (Modify) Bone - Bone: dyn_Cloth_volume_l_A01"(ComponentPose=@n52, Alpha=1, Translation="0,0,0", Rotation="0,10,0", Scale="1,1,1") {
    rotationMode=BMM_Additive
    rotationSpace=BCS_BoneSpace
    scaleMode=BMM_Additive
    scaleSpace=BCS_BoneSpace
    translationMode=BMM_Additive
    translationSpace=BCS_BoneSpace
    alphaBoolBlend {blendOption=Linear}
    boneToModify {boneName=dyn_Cloth_volume_l_A01}
    rotation {yaw=10}
   }
   n54 = "Transform (Modify) Bone - Bone: dyn_Cloth_volume_l_A02"(ComponentPose=@n53, Alpha=1, Translation="0,0,0", Rotation="0,10,0", Scale="1,1,1") {
    rotationMode=BMM_Additive
    rotationSpace=BCS_BoneSpace
    scaleMode=BMM_Additive
    scaleSpace=BCS_BoneSpace
    translationMode=BMM_Additive
    translationSpace=BCS_BoneSpace
    alphaBoolBlend {blendOption=Linear}
    boneToModify {boneName=dyn_Cloth_volume_l_A02}
    rotation {yaw=10}
   }
   n55 = "Transform (Modify) Bone - Bone: dyn_Cloth_volume_l_B"!(ComponentPose=@n54, Alpha:float=1, Translation="0,0,0", Rotation="0,10,0", Scale="1,1,1") -> (Pose) {
    rotationMode=BMM_Additive
    rotationSpace=BCS_BoneSpace
    scaleMode=BMM_Additive
    scaleSpace=BCS_BoneSpace
    translationMode=BMM_Additive
    translationSpace=BCS_BoneSpace
    alphaBoolBlend {blendOption=Linear}
    boneToModify {boneName=dyn_Cloth_volume_l_B}
    rotation {yaw=10}
   }
   n56 = "Transform (Modify) Bone - Bone: dyn_Cloth_volume_l_B01"(ComponentPose=@n55, Alpha=1, Translation="0,0,0", Rotation="0,10,0", Scale="1,1,1") {
    rotationMode=BMM_Additive
    rotationSpace=BCS_BoneSpace
    scaleMode=BMM_Additive
    scaleSpace=BCS_BoneSpace
    translationMode=BMM_Additive
    translationSpace=BCS_BoneSpace
    alphaBoolBlend {blendOption=Linear}
    boneToModify {boneName=dyn_Cloth_volume_l_B01}
    rotation {yaw=10}
   }
   n57 = "Transform (Modify) Bone - Bone: dyn_Cloth_volume_l_B02"(ComponentPose=@n56, Alpha=1, Translation="0,0,0", Rotation="0,10,0", Scale="1,1,1") {
    rotationMode=BMM_Additive
    rotationSpace=BCS_BoneSpace
    scaleMode=BMM_Additive
    scaleSpace=BCS_BoneSpace
    translationMode=BMM_Additive
    translationSpace=BCS_BoneSpace
    alphaBoolBlend {blendOption=Linear}
    boneToModify {boneName=dyn_Cloth_volume_l_B02}
    rotation {yaw=10}
   }
   n58 = "Transform (Modify) Bone - Bone: dyn_Cloth_volume_l_C"!(ComponentPose=@n57, Alpha:float=1, Translation="0,0,0", Rotation="0,10,0", Scale="1,1,1") -> (Pose) {
    rotationMode=BMM_Additive
    rotationSpace=BCS_BoneSpace
    scaleMode=BMM_Additive
    scaleSpace=BCS_BoneSpace
    translationMode=BMM_Additive
    translationSpace=BCS_BoneSpace
    alphaBoolBlend {blendOption=Linear}
    boneToModify {boneName=dyn_Cloth_volume_l_C}
    rotation {yaw=10}
   }
   n59 = "Transform (Modify) Bone - Bone: dyn_Cloth_volume_l_C01"(ComponentPose=@n58, Alpha=1, Translation="0,0,0", Rotation="0,10,0", Scale="1,1,1") {
    rotationMode=BMM_Additive
    rotationSpace=BCS_BoneSpace
    scaleMode=BMM_Additive
    scaleSpace=BCS_BoneSpace
    translationMode=BMM_Additive
    translationSpace=BCS_BoneSpace
    alphaBoolBlend {blendOption=Linear}
    boneToModify {boneName=dyn_Cloth_volume_l_C01}
    rotation {yaw=10}
   }
   n60 = "Transform (Modify) Bone - Bone: dyn_Cloth_volume_l_C02"(ComponentPose=@n59, Alpha=1, Translation="0,0,0", Rotation="0,10,0", Scale="1,1,1") {
    rotationMode=BMM_Additive
    rotationSpace=BCS_BoneSpace
    scaleMode=BMM_Additive
    scaleSpace=BCS_BoneSpace
    translationMode=BMM_Additive
    translationSpace=BCS_BoneSpace
    alphaBoolBlend {blendOption=Linear}
    boneToModify {boneName=dyn_Cloth_volume_l_C02}
    rotation {yaw=10}
   }
   n61 = "Transform (Modify) Bone - Bone: dyn_Cloth_volume_fn"!(ComponentPose=@n60, Alpha:float=1, Translation="0,0,0", Rotation="0,10,0", Scale="1,1,1") -> (Pose) {
    rotationMode=BMM_Additive
    rotationSpace=BCS_BoneSpace
    scaleMode=BMM_Additive
    scaleSpace=BCS_BoneSpace
    translationMode=BMM_Additive
    translationSpace=BCS_BoneSpace
    alphaBoolBlend {blendOption=Linear}
    boneToModify {boneName=dyn_Cloth_volume_fn}
    rotation {yaw=10}
   }
   n62 = "Transform (Modify) Bone - Bone: dyn_Cloth_volume_fn01"(ComponentPose=@n61, Alpha=1, Translation="0,0,0", Rotation="0,10,0", Scale="1,1,1") {
    rotationMode=BMM_Additive
    rotationSpace=BCS_BoneSpace
    scaleMode=BMM_Additive
    scaleSpace=BCS_BoneSpace
    translationMode=BMM_Additive
    translationSpace=BCS_BoneSpace
    alphaBoolBlend {blendOption=Linear}
    boneToModify {boneName=dyn_Cloth_volume_fn01}
    rotation {yaw=10}
   }
   n63 = "Transform (Modify) Bone - Bone: dyn_Cloth_volume_fn02"(ComponentPose=@n62, Alpha=1, Translation="0,0,0", Rotation="0,10,0", Scale="1,1,1") {
    rotationMode=BMM_Additive
    rotationSpace=BCS_BoneSpace
    scaleMode=BMM_Additive
    scaleSpace=BCS_BoneSpace
    translationMode=BMM_Additive
    translationSpace=BCS_BoneSpace
    alphaBoolBlend {blendOption=Linear}
    boneToModify {boneName=dyn_Cloth_volume_fn02}
    rotation {yaw=10}
   }
   n64 = "Transform (Modify) Bone - Bone: dyn_Cloth_volume_r_A"!(ComponentPose=@n63, Alpha:float=1, Translation="0,0,0", Rotation="0,10,0", Scale="1,1,1") -> (Pose) {
    rotationMode=BMM_Additive
    rotationSpace=BCS_BoneSpace
    scaleMode=BMM_Additive
    scaleSpace=BCS_BoneSpace
    translationMode=BMM_Additive
    translationSpace=BCS_BoneSpace
    alphaBoolBlend {blendOption=Linear}
    boneToModify {boneName=dyn_Cloth_volume_r_A}
    rotation {yaw=10}
   }
   n65 = "Transform (Modify) Bone - Bone: dyn_Cloth_volume_r_A01"(ComponentPose=@n64, Alpha=1, Translation="0,0,0", Rotation="0,10,0", Scale="1,1,1") {
    rotationMode=BMM_Additive
    rotationSpace=BCS_BoneSpace
    scaleMode=BMM_Additive
    scaleSpace=BCS_BoneSpace
    translationMode=BMM_Additive
    translationSpace=BCS_BoneSpace
    alphaBoolBlend {blendOption=Linear}
    boneToModify {boneName=dyn_Cloth_volume_r_A01}
    rotation {yaw=10}
   }
   n66 = "Transform (Modify) Bone - Bone: dyn_Cloth_volume_r_A02"(ComponentPose=@n65, Alpha=1, Translation="0,0,0", Rotation="0,10,0", Scale="1,1,1") {
    rotationMode=BMM_Additive
    rotationSpace=BCS_BoneSpace
    scaleMode=BMM_Additive
    scaleSpace=BCS_BoneSpace
    translationMode=BMM_Additive
    translationSpace=BCS_BoneSpace
    alphaBoolBlend {blendOption=Linear}
    boneToModify {boneName=dyn_Cloth_volume_r_A02}
    rotation {yaw=10}
   }
   n67 = "Transform (Modify) Bone - Bone: dyn_Cloth_volume_r_B"!(ComponentPose=@n66, Alpha:float=1, Translation="0,0,0", Rotation="0,10,0", Scale="1,1,1") -> (Pose) {
    rotationMode=BMM_Additive
    rotationSpace=BCS_BoneSpace
    scaleMode=BMM_Additive
    scaleSpace=BCS_BoneSpace
    translationMode=BMM_Additive
    translationSpace=BCS_BoneSpace
    alphaBoolBlend {blendOption=Linear}
    boneToModify {boneName=dyn_Cloth_volume_r_B}
    rotation {yaw=10}
   }
   n68 = "Transform (Modify) Bone - Bone: dyn_Cloth_volume_r_B01"(ComponentPose=@n67, Alpha=1, Translation="0,0,0", Rotation="0,10,0", Scale="1,1,1") {
    rotationMode=BMM_Additive
    rotationSpace=BCS_BoneSpace
    scaleMode=BMM_Additive
    scaleSpace=BCS_BoneSpace
    translationMode=BMM_Additive
    translationSpace=BCS_BoneSpace
    alphaBoolBlend {blendOption=Linear}
    boneToModify {boneName=dyn_Cloth_volume_r_B01}
    rotation {yaw=10}
   }
   n69 = "Transform (Modify) Bone - Bone: dyn_Cloth_volume_r_B02"(ComponentPose=@n68, Alpha=1, Translation="0,0,0", Rotation="0,10,0", Scale="1,1,1") {
    rotationMode=BMM_Additive
    rotationSpace=BCS_BoneSpace
    scaleMode=BMM_Additive
    scaleSpace=BCS_BoneSpace
    translationMode=BMM_Additive
    translationSpace=BCS_BoneSpace
    alphaBoolBlend {blendOption=Linear}
    boneToModify {boneName=dyn_Cloth_volume_r_B02}
    rotation {yaw=10}
   }
   n70 = "Transform (Modify) Bone - Bone: dyn_Cloth_volume_r_C"!(ComponentPose=@n69, Alpha:float=1, Translation="0,0,0", Rotation="0,10,0", Scale="1,1,1") -> (Pose) {
    rotationMode=BMM_Additive
    rotationSpace=BCS_BoneSpace
    scaleMode=BMM_Additive
    scaleSpace=BCS_BoneSpace
    translationMode=BMM_Additive
    translationSpace=BCS_BoneSpace
    alphaBoolBlend {blendOption=Linear}
    boneToModify {boneName=dyn_Cloth_volume_r_C}
    rotation {yaw=10}
   }
   n71 = "Transform (Modify) Bone - Bone: dyn_Cloth_volume_r_C01"(ComponentPose=@n70, Alpha=1, Translation="0,0,0", Rotation="0,10,0", Scale="1,1,1") {
    rotationMode=BMM_Additive
    rotationSpace=BCS_BoneSpace
    scaleMode=BMM_Additive
    scaleSpace=BCS_BoneSpace
    translationMode=BMM_Additive
    translationSpace=BCS_BoneSpace
    alphaBoolBlend {blendOption=Linear}
    boneToModify {boneName=dyn_Cloth_volume_r_C01}
    rotation {yaw=10}
   }
   n72 = "Transform (Modify) Bone - Bone: dyn_Cloth_volume_r_C02"(ComponentPose=@n71, Alpha=1, Translation="0,0,0", Rotation="0,10,0", Scale="1,1,1") {
    rotationMode=BMM_Additive
    rotationSpace=BCS_BoneSpace
    scaleMode=BMM_Additive
    scaleSpace=BCS_BoneSpace
    translationMode=BMM_Additive
    translationSpace=BCS_BoneSpace
    alphaBoolBlend {blendOption=Linear}
    boneToModify {boneName=dyn_Cloth_volume_r_C02}
    rotation {yaw=10}
   }
   n73 = "Transform (Modify) Bone - Bone: dyn_Cloth_volume_bk"!(ComponentPose=@n72, Alpha:float=1, Translation="0,0,0", Rotation="0,10,0", Scale="1,1,1") -> (Pose) {
    rotationMode=BMM_Additive
    rotationSpace=BCS_BoneSpace
    scaleMode=BMM_Additive
    scaleSpace=BCS_BoneSpace
    translationMode=BMM_Additive
    translationSpace=BCS_BoneSpace
    alphaBoolBlend {blendOption=Linear}
    boneToModify {boneName=dyn_Cloth_volume_bk}
    rotation {yaw=10}
   }
   n74 = "Transform (Modify) Bone - Bone: dyn_Cloth_volume_bk01"(ComponentPose=@n73, Alpha=1, Translation="0,0,0", Rotation="0,10,0", Scale="1,1,1") {
    rotationMode=BMM_Additive
    rotationSpace=BCS_BoneSpace
    scaleMode=BMM_Additive
    scaleSpace=BCS_BoneSpace
    translationMode=BMM_Additive
    translationSpace=BCS_BoneSpace
    alphaBoolBlend {blendOption=Linear}
    boneToModify {boneName=dyn_Cloth_volume_bk01}
    rotation {yaw=10}
   }
   n75 = "Transform (Modify) Bone - Bone: dyn_Cloth_volume_bk02"(ComponentPose=@n74, Alpha=1, Translation="0,0,0", Rotation="0,10,0", Scale="1,1,1") {
    rotationMode=BMM_Additive
    rotationSpace=BCS_BoneSpace
    scaleMode=BMM_Additive
    scaleSpace=BCS_BoneSpace
    translationMode=BMM_Additive
    translationSpace=BCS_BoneSpace
    alphaBoolBlend {blendOption=Linear}
    boneToModify {boneName=dyn_Cloth_volume_bk02}
    rotation {yaw=10}
   }
   n76 = Anim Dynamics - Bone: dyn_Cloth_volume_l_A!(ComponentPose=@n75, Alpha=1, GravityOverride:Vector="0,0,-980", ExternalForce="0,0,0") -> (Pose) {
    angularDampingOverride=0.99
    angularSpringConstant=800
    bAngularSpring=true
    bGravityOverrideInSimSpace=true
    bOverrideAngularDamping=true
    bUseGravityOverride=true
    gravityScale=0.3
    simulationSpace=BoneRelative
    alphaBoolBlend {blendOption=Linear}
    boundBone {boneName=dyn_Cloth_volume_l_A}
    componentAppliedLinearAccClamp=<2500 2500 1100>
    componentLinearAccScale=<0.5 0.5 0.4>
    componentLinearVelScale=<0.55 0.55 0.45>
    gravityOverride {z=500}
    physicsBodyDefinitions [{
      sphereCollisionRadius=0
      boundBone {boneName=dyn_Cloth_volume_l_A}
      boxExtents=<60 60 60>
      constraintSetup {angularLimitsMax=<0 0 20>, angularLimitsMin=<0 0 -10>, angularTarget=<0 0 0>}
      localJointOffset=<-5 0.2 -0.1>
     }]
    relativeSpaceBone {boneName=root}
   }
   n77 = Anim Dynamics - Bone: dyn_Cloth_volume_l_A01!(ComponentPose=@n76, Alpha=1, GravityOverride:Vector="0,0,-980", ExternalForce="0,0,0") -> (Pose) {
    angularDampingOverride=0.99
    angularSpringConstant=800
    bAngularSpring=true
    bGravityOverrideInSimSpace=true
    bOverrideAngularDamping=true
    bUseGravityOverride=true
    gravityScale=0.3
    simulationSpace=BoneRelative
    alphaBoolBlend {blendOption=Linear}
    boundBone {boneName=dyn_Cloth_volume_l_A01}
    componentAppliedLinearAccClamp=<2500 2500 1100>
    componentLinearAccScale=<0.5 0.5 0.4>
    componentLinearVelScale=<0.55 0.55 0.45>
    gravityOverride {z=500}
    physicsBodyDefinitions [{
      sphereCollisionRadius=0
      boundBone {boneName=dyn_Cloth_volume_l_A01}
      boxExtents=<35 17.5 35>
      constraintSetup {angularLimitsMax=<0 0 20>, angularLimitsMin=<0 0 -10>, angularTarget=<0 0 0>}
      localJointOffset=<-5 0.2 -0.1>
     }]
    relativeSpaceBone {boneName=root}
   }
   n78 = Anim Dynamics - Bone: dyn_Cloth_volume_l_A02!(ComponentPose=@n77, Alpha=1, GravityOverride:Vector="0,0,-980", ExternalForce="0,0,0") -> (Pose) {
    angularDampingOverride=0.99
    angularSpringConstant=800
    bAngularSpring=true
    bGravityOverrideInSimSpace=true
    bOverrideAngularDamping=true
    bUseGravityOverride=true
    gravityScale=0.3
    simulationSpace=BoneRelative
    alphaBoolBlend {blendOption=Linear}
    boundBone {boneName=dyn_Cloth_volume_l_A02}
    componentAppliedLinearAccClamp=<2500 2500 1100>
    componentLinearAccScale=<0.5 0.5 0.4>
    componentLinearVelScale=<0.55 0.55 0.45>
    gravityOverride {z=500}
    physicsBodyDefinitions [{
      sphereCollisionRadius=0
      boundBone {boneName=dyn_Cloth_volume_l_A02}
      boxExtents=<15 10 15>
      constraintSetup {angularLimitsMax=<0 0 20>, angularLimitsMin=<0 0 -10>, angularTarget=<0 0 0>}
      localJointOffset=<-5 0.2 -0.1>
     }]
    relativeSpaceBone {boneName=root}
   }
   n79 = Anim Dynamics - Bone: dyn_Cloth_volume_l_B!(ComponentPose=@n78, Alpha=1, GravityOverride:Vector="0,0,-980", ExternalForce="0,0,0") -> (Pose) {
    angularDampingOverride=0.99
    angularSpringConstant=800
    bAngularSpring=true
    bGravityOverrideInSimSpace=true
    bOverrideAngularDamping=true
    bUseGravityOverride=true
    gravityScale=0.3
    simulationSpace=BoneRelative
    alphaBoolBlend {blendOption=Linear}
    boundBone {boneName=dyn_Cloth_volume_l_B}
    componentAppliedLinearAccClamp=<2500 2500 1100>
    componentLinearAccScale=<0.5 0.5 0.4>
    componentLinearVelScale=<0.55 0.55 0.45>
    gravityOverride {z=500}
    physicsBodyDefinitions [{
      sphereCollisionRadius=0
      boundBone {boneName=dyn_Cloth_volume_l_B}
      boxExtents=<60 60 60>
      constraintSetup {angularLimitsMax=<2 3.5 20>, angularLimitsMin=<-2 -3.5 -10>, angularTarget=<0 0 0>}
      localJointOffset=<-5 0.2 -0.1>
     }]
    relativeSpaceBone {boneName=root}
   }
   n80 = Anim Dynamics - Bone: dyn_Cloth_volume_l_B01!(ComponentPose=@n79, Alpha=1, GravityOverride:Vector="0,0,-980", ExternalForce="0,0,0") -> (Pose) {
    angularDampingOverride=0.99
    angularSpringConstant=800
    bAngularSpring=true
    bGravityOverrideInSimSpace=true
    bOverrideAngularDamping=true
    bUseGravityOverride=true
    gravityScale=0.3
    simulationSpace=BoneRelative
    alphaBoolBlend {blendOption=Linear}
    boundBone {boneName=dyn_Cloth_volume_l_B01}
    componentAppliedLinearAccClamp=<2500 2500 1100>
    componentLinearAccScale=<0.5 0.5 0.4>
    componentLinearVelScale=<0.55 0.55 0.45>
    gravityOverride {z=500}
    physicsBodyDefinitions [{
      sphereCollisionRadius=0
      boundBone {boneName=dyn_Cloth_volume_l_B01}
      boxExtents=<35 17.5 35>
      constraintSetup {angularLimitsMax=<2 3.5 20>, angularLimitsMin=<-2 -3.5 -10>, angularTarget=<0 0 0>}
      localJointOffset=<-5 0.2 -0.1>
     }]
    relativeSpaceBone {boneName=root}
    sphericalLimits [{limitRadius=22, drivingBone {boneName=lowerarm_l}, sphereLocalOffset=<25 0 17>}]
   }
   n81 = Anim Dynamics - Bone: dyn_Cloth_volume_l_B02!(ComponentPose=@n80, Alpha=1, GravityOverride:Vector="0,0,-980", ExternalForce="0,0,0") -> (Pose) {
    angularDampingOverride=0.99
    angularSpringConstant=800
    bAngularSpring=true
    bGravityOverrideInSimSpace=true
    bOverrideAngularDamping=true
    bUseGravityOverride=true
    bUseSphericalLimits=true
    gravityScale=0.3
    simulationSpace=BoneRelative
    alphaBoolBlend {blendOption=Linear}
    boundBone {boneName=dyn_Cloth_volume_l_B02}
    componentAppliedLinearAccClamp=<2500 2500 1100>
    componentLinearAccScale=<0.5 0.5 0.4>
    componentLinearVelScale=<0.55 0.55 0.45>
    gravityOverride {z=500}
    physicsBodyDefinitions [{
      sphereCollisionRadius=0
      boundBone {boneName=dyn_Cloth_volume_l_B02}
      boxExtents=<15 10 15>
      constraintSetup {angularLimitsMax=<2 3.5 20>, angularLimitsMin=<-2 -3.5 -10>, angularTarget=<0 0 0>}
      localJointOffset=<-5 0.2 -0.1>
     }]
    relativeSpaceBone {boneName=root}
    sphericalLimits [{limitRadius=22, drivingBone {boneName=lowerarm_l}, sphereLocalOffset=<25 0 17>}]
   }
   n82 = Anim Dynamics - Bone: dyn_Cloth_volume_l_C!(ComponentPose=@n81, Alpha=1, GravityOverride:Vector="0,0,-980", ExternalForce="0,0,0") -> (Pose) {
    angularDampingOverride=0.99
    angularSpringConstant=800
    bAngularSpring=true
    bGravityOverrideInSimSpace=true
    bOverrideAngularDamping=true
    bUseGravityOverride=true
    gravityScale=0.3
    simulationSpace=BoneRelative
    alphaBoolBlend {blendOption=Linear}
    boundBone {boneName=dyn_Cloth_volume_l_C}
    componentAppliedLinearAccClamp=<2500 2500 1100>
    componentLinearAccScale=<0.5 0.5 0.4>
    componentLinearVelScale=<0.55 0.55 0.45>
    gravityOverride {z=500}
    physicsBodyDefinitions [{
      sphereCollisionRadius=0
      boundBone {boneName=dyn_Cloth_volume_l_C}
      boxExtents=<60 60 60>
      constraintSetup {angularLimitsMax=<0 5 20>, angularLimitsMin=<0 -5 -10>, angularTarget=<0 0 0>}
      localJointOffset=<-5 0.2 -0.1>
     }]
    relativeSpaceBone {boneName=root}
   }
   n83 = Anim Dynamics - Bone: dyn_Cloth_volume_l_C(ComponentPose=@n82, Alpha=1, GravityOverride="0,0,-980", ExternalForce="0,0,0") {
    angularDampingOverride=0.99
    angularSpringConstant=800
    bAngularSpring=true
    bGravityOverrideInSimSpace=true
    bOverrideAngularDamping=true
    bUseGravityOverride=true
    gravityScale=0.3
    simulationSpace=BoneRelative
    alphaBoolBlend {blendOption=Linear}
    boundBone {boneName=dyn_Cloth_volume_l_C}
    componentAppliedLinearAccClamp=<2500 2500 1100>
    componentLinearAccScale=<0.5 0.5 0.4>
    componentLinearVelScale=<0.55 0.55 0.45>
    gravityOverride {z=500}
    physicsBodyDefinitions [{
      sphereCollisionRadius=0
      boundBone {boneName=dyn_Cloth_volume_l_C}
      boxExtents=<35 17.5 35>
      constraintSetup {angularLimitsMax=<0 5 20>, angularLimitsMin=<0 -5 -10>, angularTarget=<0 0 0>}
      localJointOffset=<-5 0.2 -0.1>
     }]
    relativeSpaceBone {boneName=root}
   }
   n84 = Anim Dynamics - Bone: dyn_Cloth_volume_l_C02!(ComponentPose=@n83, Alpha=1, GravityOverride:Vector="0,0,-980", ExternalForce="0,0,0") -> (Pose) {
    angularDampingOverride=0.99
    angularSpringConstant=800
    bAngularSpring=true
    bGravityOverrideInSimSpace=true
    bOverrideAngularDamping=true
    bUseGravityOverride=true
    gravityScale=0.3
    simulationSpace=BoneRelative
    alphaBoolBlend {blendOption=Linear}
    boundBone {boneName=dyn_Cloth_volume_l_C02}
    componentAppliedLinearAccClamp=<2500 2500 1100>
    componentLinearAccScale=<0.5 0.5 0.4>
    componentLinearVelScale=<0.55 0.55 0.45>
    gravityOverride {z=500}
    physicsBodyDefinitions [{
      sphereCollisionRadius=0
      boundBone {boneName=dyn_Cloth_volume_l_C02}
      boxExtents=<15 10 15>
      constraintSetup {angularLimitsMax=<0 5 20>, angularLimitsMin=<0 -5 -15>, angularTarget=<0 0 0>}
      localJointOffset=<-5 0.2 -0.1>
     }]
    relativeSpaceBone {boneName=root}
   }
   n85 = Anim Dynamics - Bone: dyn_Cloth_volume_fn!(ComponentPose=@n84, Alpha=1, GravityOverride:Vector="0,0,-980", ExternalForce="0,0,0") -> (Pose) {
    angularDampingOverride=0.99
    angularSpringConstant=800
    bAngularSpring=true
    bGravityOverrideInSimSpace=true
    bOverrideAngularDamping=true
    bUseGravityOverride=true
    gravityScale=0.3
    simulationSpace=BoneRelative
    alphaBoolBlend {blendOption=Linear}
    boundBone {boneName=dyn_Cloth_volume_fn}
    componentAppliedLinearAccClamp=<2500 2500 1100>
    componentLinearAccScale=<0.5 0.5 0.4>
    componentLinearVelScale=<0.55 0.55 0.45>
    gravityOverride {z=500}
    physicsBodyDefinitions [{
      sphereCollisionRadius=0
      boundBone {boneName=dyn_Cloth_volume_fn}
      boxExtents=<60 60 60>
      constraintSetup {angularLimitsMax=<0 3.5 20>, angularLimitsMin=<0 -3.5 -10>, angularTarget=<0 0 0>}
      localJointOffset=<-5 0.2 -0.1>
     }]
    relativeSpaceBone {boneName=root}
   }
   n86 = Anim Dynamics - Bone: dyn_Cloth_volume_fn01!(ComponentPose=@n85, Alpha=1, GravityOverride:Vector="0,0,-980", ExternalForce="0,0,0") -> (Pose) {
    angularDampingOverride=0.99
    angularSpringConstant=800
    bAngularSpring=true
    bGravityOverrideInSimSpace=true
    bOverrideAngularDamping=true
    bUseGravityOverride=true
    gravityScale=0.3
    simulationSpace=BoneRelative
    alphaBoolBlend {blendOption=Linear}
    boundBone {boneName=dyn_Cloth_volume_fn01}
    componentAppliedLinearAccClamp=<2500 2500 1100>
    componentLinearAccScale=<0.5 0.5 0.4>
    componentLinearVelScale=<0.55 0.55 0.45>
    gravityOverride {z=500}
    physicsBodyDefinitions [{
      sphereCollisionRadius=0
      boundBone {boneName=dyn_Cloth_volume_fn01}
      boxExtents=<35 17.5 35>
      constraintSetup {angularLimitsMax=<0 3.5 20>, angularLimitsMin=<0 -3.5 -10>, angularTarget=<0 0 0>}
      localJointOffset=<-5 0.2 -0.1>
     }]
    relativeSpaceBone {boneName=root}
   }
   n87 = Anim Dynamics - Bone: dyn_Cloth_volume_fn02!(ComponentPose=@n86, Alpha=1, GravityOverride:Vector="0,0,-980", ExternalForce="0,0,0") -> (Pose) {
    angularDampingOverride=0.99
    angularSpringConstant=800
    bAngularSpring=true
    bGravityOverrideInSimSpace=true
    bOverrideAngularDamping=true
    bUseGravityOverride=true
    gravityScale=0.3
    simulationSpace=BoneRelative
    alphaBoolBlend {blendOption=Linear}
    boundBone {boneName=dyn_Cloth_volume_fn02}
    componentAppliedLinearAccClamp=<2500 2500 1100>
    componentLinearAccScale=<0.5 0.5 0.4>
    componentLinearVelScale=<0.55 0.55 0.45>
    gravityOverride {z=500}
    physicsBodyDefinitions [{
      sphereCollisionRadius=0
      boundBone {boneName=dyn_Cloth_volume_fn02}
      boxExtents=<15 10 15>
      constraintSetup {angularLimitsMax=<0 3.5 20>, angularLimitsMin=<0 -3.5 -10>, angularTarget=<0 0 0>}
      localJointOffset=<-5 0.2 -0.1>
     }]
    relativeSpaceBone {boneName=root}
   }
   n88 = Anim Dynamics - Bone: dyn_Cloth_volume_r_A!(ComponentPose=@n87, Alpha=1, GravityOverride:Vector="0,0,-980", ExternalForce="0,0,0") -> (Pose) {
    angularDampingOverride=0.99
    angularSpringConstant=800
    bAngularSpring=true
    bGravityOverrideInSimSpace=true
    bOverrideAngularDamping=true
    bUseGravityOverride=true
    bUsePlanarLimit=false
    gravityScale=0.3
    simulationSpace=BoneRelative
    alphaBoolBlend {blendOption=Linear}
    boundBone {boneName=dyn_Cloth_volume_r_A}
    componentAppliedLinearAccClamp=<2500 2500 1100>
    componentLinearAccScale=<0.5 0.5 0.4>
    componentLinearVelScale=<0.55 0.55 0.45>
    gravityOverride {z=-980}
    physicsBodyDefinitions [{
      sphereCollisionRadius=0
      boundBone {boneName=dyn_Cloth_volume_r_A}
      boxExtents=<35 20 35>
      constraintSetup {angularLimitsMax=<10 10 20>, angularLimitsMin=<-10 -10 -5>, angularTarget=<0 0 0>}
      localJointOffset=<-5 0.2 -0.1>
     }]
    relativeSpaceBone {boneName=root}
   }
   n89 = Anim Dynamics - Bone: dyn_Cloth_volume_r_A01!(ComponentPose=@n88, Alpha=1, GravityOverride:Vector="0,0,-980", ExternalForce="0,0,0") -> (Pose) {
    angularDampingOverride=0.99
    angularSpringConstant=800
    bAngularSpring=true
    bGravityOverrideInSimSpace=true
    bOverrideAngularDamping=true
    bUseGravityOverride=true
    gravityScale=0.3
    simulationSpace=BoneRelative
    alphaBoolBlend {blendOption=Linear}
    boundBone {boneName=dyn_Cloth_volume_r_A01}
    componentAppliedLinearAccClamp=<2500 2500 1100>
    componentLinearAccScale=<0.5 0.5 0.4>
    componentLinearVelScale=<0.55 0.55 0.45>
    gravityOverride {z=-980}
    physicsBodyDefinitions [{
      sphereCollisionRadius=0
      boundBone {boneName=dyn_Cloth_volume_r_A01}
      boxExtents=<25 15 25>
      constraintSetup {angularLimitsMax=<10 10 20>, angularLimitsMin=<-10 -10 -5>, angularTarget=<0 0 0>}
      localJointOffset=<-5 0.2 -0.1>
     }]
    relativeSpaceBone {boneName=root}
   }
   n90 = Anim Dynamics - Bone: dyn_Cloth_volume_r_A02!(ComponentPose=@n89, Alpha=1, GravityOverride:Vector="0,0,-980", ExternalForce="0,0,0") -> (Pose) {
    angularDampingOverride=0.99
    angularSpringConstant=800
    bAngularSpring=true
    bGravityOverrideInSimSpace=true
    bOverrideAngularDamping=true
    bUseGravityOverride=true
    gravityScale=0.3
    simulationSpace=BoneRelative
    alphaBoolBlend {blendOption=Linear}
    boundBone {boneName=dyn_Cloth_volume_r_A02}
    componentAppliedLinearAccClamp=<2500 2500 1100>
    componentLinearAccScale=<0.5 0.5 0.4>
    componentLinearVelScale=<0.55 0.55 0.45>
    gravityOverride {z=-980}
    physicsBodyDefinitions [{
      sphereCollisionRadius=0
      boundBone {boneName=dyn_Cloth_volume_r_A02}
      boxExtents=<25 15 25>
      constraintSetup {angularLimitsMax=<10 10 20>, angularLimitsMin=<-10 -10 -5>, angularTarget=<0 0 0>}
      localJointOffset=<-5 0.2 -0.1>
     }]
    relativeSpaceBone {boneName=root}
   }
   n91 = Anim Dynamics - Bone: dyn_Cloth_volume_r_B!(ComponentPose=@n90, Alpha=1, GravityOverride:Vector="0,0,-980", ExternalForce="0,0,0") -> (Pose) {
    angularDampingOverride=0.99
    angularSpringConstant=800
    bAngularSpring=true
    bGravityOverrideInSimSpace=true
    bOverrideAngularDamping=true
    bUseGravityOverride=true
    gravityScale=0.3
    simulationSpace=BoneRelative
    alphaBoolBlend {blendOption=Linear}
    boundBone {boneName=dyn_Cloth_volume_r_B}
    componentAppliedLinearAccClamp=<2500 2500 1100>
    componentLinearAccScale=<0.5 0.5 0.4>
    componentLinearVelScale=<0.55 0.55 0.45>
    gravityOverride {z=-980}
    physicsBodyDefinitions [{
      sphereCollisionRadius=0
      boundBone {boneName=dyn_Cloth_volume_r_B}
      boxExtents=<35 20 35>
      constraintSetup {angularLimitsMax=<10 10 20>, angularLimitsMin=<-10 -10 -5>, angularTarget=<0 0 0>}
      localJointOffset=<-5 0.2 -0.1>
     }]
    relativeSpaceBone {boneName=root}
   }
   n92 = Anim Dynamics - Bone: dyn_Cloth_volume_r_B01!(ComponentPose=@n91, Alpha=1, GravityOverride:Vector="0,0,-980", ExternalForce="0,0,0") -> (Pose) {
    angularDampingOverride=0.99
    angularSpringConstant=800
    bAngularSpring=true
    bGravityOverrideInSimSpace=true
    bOverrideAngularDamping=true
    bUseGravityOverride=true
    gravityScale=0.3
    simulationSpace=BoneRelative
    alphaBoolBlend {blendOption=Linear}
    boundBone {boneName=dyn_Cloth_volume_r_B01}
    componentAppliedLinearAccClamp=<2500 2500 1100>
    componentLinearAccScale=<0.5 0.5 0.4>
    componentLinearVelScale=<0.55 0.55 0.45>
    gravityOverride {z=-980}
    physicsBodyDefinitions [{
      sphereCollisionRadius=0
      boundBone {boneName=dyn_Cloth_volume_r_B01}
      boxExtents=<25 15 25>
      constraintSetup {angularLimitsMax=<10 10 20>, angularLimitsMin=<-10 -10 -5>, angularTarget=<0 0 0>}
      localJointOffset=<-5 0.2 -0.1>
     }]
    relativeSpaceBone {boneName=root}
   }
   n93 = Anim Dynamics - Bone: dyn_Cloth_volume_r_B02!(ComponentPose=@n92, Alpha=1, GravityOverride:Vector="0,0,-980", ExternalForce="0,0,0") -> (Pose) {
    angularDampingOverride=0.99
    angularSpringConstant=800
    bAngularSpring=true
    bGravityOverrideInSimSpace=true
    bOverrideAngularDamping=true
    bUseGravityOverride=true
    bUseSphericalLimits=true
    gravityScale=0.3
    simulationSpace=BoneRelative
    alphaBoolBlend {blendOption=Linear}
    boundBone {boneName=dyn_Cloth_volume_r_B02}
    componentAppliedLinearAccClamp=<2500 2500 1100>
    componentLinearAccScale=<0.5 0.5 0.4>
    componentLinearVelScale=<0.55 0.55 0.45>
    gravityOverride {z=-980}
    physicsBodyDefinitions [{
      sphereCollisionRadius=0
      boundBone {boneName=dyn_Cloth_volume_r_B02}
      boxExtents=<25 15 25>
      constraintSetup {angularLimitsMax=<10 10 20>, angularLimitsMin=<-10 -10 -5>, angularTarget=<0 0 0>}
      localJointOffset=<-5 0.2 -0.1>
     }]
    relativeSpaceBone {boneName=root}
    sphericalLimits [
     {limitRadius=22, drivingBone {boneName=lowerarm_r}, sphereLocalOffset=<-25 0 -17>}
     {limitRadius=0, drivingBone {boneName=None}, sphereLocalOffset=<0 0 0>}
    ]
   }
   n94 = Anim Dynamics - Bone: dyn_Cloth_volume_r_C!(ComponentPose=@n93, Alpha=1, GravityOverride:Vector="0,0,-980", ExternalForce="0,0,0") -> (Pose) {
    angularDampingOverride=0.99
    angularSpringConstant=800
    bAngularSpring=true
    bGravityOverrideInSimSpace=true
    bOverrideAngularDamping=true
    bUseGravityOverride=true
    gravityScale=0.3
    simulationSpace=BoneRelative
    alphaBoolBlend {blendOption=Linear}
    boundBone {boneName=dyn_Cloth_volume_r_C}
    componentAppliedLinearAccClamp=<2500 2500 1100>
    componentLinearAccScale=<0.5 0.5 0.4>
    componentLinearVelScale=<0.55 0.55 0.45>
    gravityOverride {z=-980}
    physicsBodyDefinitions [{
      sphereCollisionRadius=0
      boundBone {boneName=dyn_Cloth_volume_r_C}
      boxExtents=<35 20 35>
      constraintSetup {angularLimitsMax=<0 5 20>, angularLimitsMin=<0 -5 -5>, angularTarget=<0 0 0>}
      localJointOffset=<-5 0.2 -0.1>
     }]
    relativeSpaceBone {boneName=root}
   }
   n95 = Anim Dynamics - Bone: dyn_Cloth_volume_r_C01!(ComponentPose=@n94, Alpha=1, GravityOverride:Vector="0,0,-980", ExternalForce="0,0,0") -> (Pose) {
    angularDampingOverride=0.99
    angularSpringConstant=800
    bAngularSpring=true
    bGravityOverrideInSimSpace=true
    bOverrideAngularDamping=true
    bUseGravityOverride=true
    gravityScale=0.3
    simulationSpace=BoneRelative
    alphaBoolBlend {blendOption=Linear}
    boundBone {boneName=dyn_Cloth_volume_r_C01}
    componentAppliedLinearAccClamp=<2500 2500 1100>
    componentLinearAccScale=<0.5 0.5 0.4>
    componentLinearVelScale=<0.55 0.55 0.45>
    gravityOverride {z=-980}
    physicsBodyDefinitions [{
      sphereCollisionRadius=0
      boundBone {boneName=dyn_Cloth_volume_r_C01}
      boxExtents=<25 15 25>
      constraintSetup {angularLimitsMax=<0 5 20>, angularLimitsMin=<0 -5 -10>, angularTarget=<0 0 0>}
      localJointOffset=<-5 0.2 -0.1>
     }]
    relativeSpaceBone {boneName=root}
   }
   n96 = Anim Dynamics - Bone: dyn_Cloth_volume_r_C02!(ComponentPose=@n95, Alpha=1, GravityOverride:Vector="0,0,-980", ExternalForce="0,0,0") -> (Pose) {
    angularDampingOverride=0.99
    angularSpringConstant=800
    bAngularSpring=true
    bGravityOverrideInSimSpace=true
    bOverrideAngularDamping=true
    bUseGravityOverride=true
    gravityScale=0.3
    simulationSpace=BoneRelative
    alphaBoolBlend {blendOption=Linear}
    boundBone {boneName=dyn_Cloth_volume_r_C02}
    componentAppliedLinearAccClamp=<2500 2500 1100>
    componentLinearAccScale=<0.5 0.5 0.4>
    componentLinearVelScale=<0.55 0.55 0.45>
    gravityOverride {z=-980}
    physicsBodyDefinitions [{
      sphereCollisionRadius=0
      boundBone {boneName=dyn_Cloth_volume_r_C02}
      boxExtents=<15 10 15>
      constraintSetup {angularLimitsMax=<0 5 20>, angularLimitsMin=<0 -5 -15>, angularTarget=<0 0 0>}
      localJointOffset=<-5 0.2 -0.1>
     }]
    relativeSpaceBone {boneName=root}
   }
   n97 = Anim Dynamics - Bone: dyn_Cloth_volume_bk!(ComponentPose=@n96, Alpha=1, GravityOverride:Vector="0,0,-980", ExternalForce="0,0,0") -> (Pose) {
    angularDampingOverride=0.99
    angularSpringConstant=800
    bAngularSpring=true
    bGravityOverrideInSimSpace=true
    bOverrideAngularDamping=true
    bUseGravityOverride=true
    gravityScale=0.3
    simulationSpace=BoneRelative
    alphaBoolBlend {blendOption=Linear}
    boundBone {boneName=dyn_Cloth_volume_bk}
    componentAppliedLinearAccClamp=<2500 2500 1100>
    componentLinearAccScale=<0.5 0.5 0.4>
    componentLinearVelScale=<0.55 0.55 0.45>
    gravityOverride {z=-980}
    physicsBodyDefinitions [{
      sphereCollisionRadius=0
      boundBone {boneName=dyn_Cloth_volume_bk}
      boxExtents=<45 25 45>
      constraintSetup {angularLimitsMax=<0 5 20>, angularLimitsMin=<0 -5 -10>, angularTarget=<0 0 0>}
      localJointOffset=<-5 0.2 -0.1>
     }]
    relativeSpaceBone {boneName=root}
   }
   n98 = Anim Dynamics - Bone: dyn_Cloth_volume_bk01!(ComponentPose=@n97, Alpha=1, GravityOverride:Vector="0,0,-980", ExternalForce="0,0,0") -> (Pose) {
    angularDampingOverride=0.99
    angularSpringConstant=800
    bAngularSpring=true
    bGravityOverrideInSimSpace=true
    bOverrideAngularDamping=true
    bUseGravityOverride=true
    gravityScale=0.3
    simulationSpace=BoneRelative
    alphaBoolBlend {blendOption=Linear}
    boundBone {boneName=dyn_Cloth_volume_bk01}
    componentAppliedLinearAccClamp=<2500 2500 1100>
    componentLinearAccScale=<0.5 0.5 0.4>
    componentLinearVelScale=<0.55 0.55 0.45>
    gravityOverride {z=-980}
    physicsBodyDefinitions [{
      sphereCollisionRadius=0
      boundBone {boneName=dyn_Cloth_volume_bk01}
      boxExtents=<25 15 25>
      constraintSetup {angularLimitsMax=<0 5 20>, angularLimitsMin=<0 -5 -10>, angularTarget=<0 0 0>}
      localJointOffset=<-5 0.2 -0.1>
     }]
    relativeSpaceBone {boneName=root}
   }
   n99 = Anim Dynamics - Bone: dyn_Cloth_volume_bk02!(ComponentPose=@n98, Alpha=1, GravityOverride:Vector="0,0,-980", ExternalForce="0,0,0") -> (Pose) {
    angularDampingOverride=0.99
    angularSpringConstant=800
    bAngularSpring=true
    bGravityOverrideInSimSpace=true
    bOverrideAngularDamping=true
    bUseGravityOverride=true
    gravityScale=0.3
    simulationSpace=BoneRelative
    alphaBoolBlend {blendOption=Linear}
    boundBone {boneName=dyn_Cloth_volume_bk02}
    componentAppliedLinearAccClamp=<2500 2500 1100>
    componentLinearAccScale=<0.5 0.5 0.4>
    componentLinearVelScale=<0.55 0.55 0.45>
    gravityOverride {z=-980}
    physicsBodyDefinitions [{
      sphereCollisionRadius=0
      boundBone {boneName=dyn_Cloth_volume_bk02}
      boxExtents=<15 10 15>
      constraintSetup {angularLimitsMax=<0 5 20>, angularLimitsMin=<0 -5 -15>, angularTarget=<0 0 0>}
      localJointOffset=<-5 0.2 -0.1>
     }]
    relativeSpaceBone {boneName=root}
   }
   n100 = Anim Dynamics - Bone: dyn_ribbon_l_01!(ComponentPose=@n99, Alpha=0.4, GravityOverride:Vector="0,0,-980", ExternalForce="0,0,0") -> (Pose) {
    angularDampingOverride=0.9999
    angularSpringConstant=350
    bAngularSpring=true
    bGravityOverrideInSimSpace=true
    bOverrideAngularDamping=true
    bUseGravityOverride=true
    gravityScale=0.3
    simulationSpace=BoneRelative
    alphaBoolBlend {blendOption=Linear}
    boundBone {boneName=dyn_ribbon_l_01}
    componentAppliedLinearAccClamp=<2500 2500 1500>
    componentLinearAccScale=<0.5 0.5 0.4>
    componentLinearVelScale=<0.55 0.55 0.45>
    gravityOverride {z=-980}
    physicsBodyDefinitions [{
      sphereCollisionRadius=0
      boundBone {boneName=dyn_ribbon_l_01}
      boxExtents=<20 10 20>
      constraintSetup {angularLimitsMax=<1 1 20>, angularLimitsMin=<-1 -1 -10>, angularTarget=<0 0 0>}
      localJointOffset=<0 5 -5>
     }]
    relativeSpaceBone {boneName=root}
   }
   n101 = Anim Dynamics - Bone: dyn_ribbon_l_02!(ComponentPose=@n100, Alpha=0.5, GravityOverride:Vector="0,0,-980", ExternalForce="0,0,0") -> (Pose) {
    angularDampingOverride=1
    bAngularSpring=true
    bGravityOverrideInSimSpace=true
    bOverrideAngularDamping=true
    bUseGravityOverride=true
    gravityScale=0.3
    simulationSpace=BoneRelative
    alphaBoolBlend {blendOption=Linear}
    boundBone {boneName=dyn_ribbon_l_02}
    componentAppliedLinearAccClamp=<2500 2500 1500>
    componentLinearAccScale=<0.5 0.5 0.4>
    componentLinearVelScale=<0.55 0.55 0.45>
    gravityOverride {z=-980}
    physicsBodyDefinitions [{
      sphereCollisionRadius=0
      boundBone {boneName=dyn_ribbon_l_02}
      boxExtents=<20 10 20>
      constraintSetup {angularLimitsMax=<1 5 45>, angularLimitsMin=<-1 -5 -45>, angularTarget=<0 0 0>}
      localJointOffset=<3 0 -3>
     }]
    relativeSpaceBone {boneName=root}
   }
   n102 = Anim Dynamics - Bone: dyn_ribbon_r_01!(ComponentPose=@n101, Alpha=0.6, GravityOverride:Vector="0,0,-980", ExternalForce="0,0,0") -> (Pose) {
    angularDampingOverride=0.999
    angularSpringConstant=500
    bAngularSpring=true
    bGravityOverrideInSimSpace=true
    bOverrideAngularDamping=true
    bUseGravityOverride=true
    gravityScale=0.3
    simulationSpace=BoneRelative
    alphaBoolBlend {blendOption=Linear}
    boundBone {boneName=dyn_ribbon_r_01}
    componentAppliedLinearAccClamp=<2500 2500 1500>
    componentLinearAccScale=<0.5 0.5 0.4>
    componentLinearVelScale=<0.55 0.55 0.45>
    gravityOverride {z=-980}
    physicsBodyDefinitions [{
      sphereCollisionRadius=0
      boundBone {boneName=dyn_ribbon_r_01}
      boxExtents=<20 10 20>
      constraintSetup {angularLimitsMax=<1 1 20>, angularLimitsMin=<-1 -1 -10>, angularTarget=<0 0 0>}
      localJointOffset=<0 -5 -5>
     }]
    relativeSpaceBone {boneName=root}
   }
   n103 = Anim Dynamics - Bone: dyn_ribbon_r_02!(ComponentPose=@n102, Alpha=0.35, GravityOverride:Vector="0,0,-980", ExternalForce="0,0,0") -> (Pose) {
    angularDampingOverride=0.9999
    bAngularSpring=true
    bGravityOverrideInSimSpace=true
    bOverrideAngularDamping=true
    bUseGravityOverride=true
    gravityScale=0.3
    simulationSpace=BoneRelative
    alphaBoolBlend {blendOption=Linear}
    boundBone {boneName=dyn_ribbon_r_02}
    componentAppliedLinearAccClamp=<2500 2500 1500>
    componentLinearAccScale=<0.5 0.5 0.4>
    componentLinearVelScale=<0.55 0.55 0.45>
    gravityOverride {z=-980}
    physicsBodyDefinitions [{
      sphereCollisionRadius=0
      boundBone {boneName=dyn_ribbon_r_02}
      boxExtents=<20 10 20>
      constraintSetup {angularLimitsMax=<1 5 45>, angularLimitsMin=<-1 -5 -45>, angularTarget=<0 0 0>}
      localJointOffset=<3 0 -3>
     }]
    relativeSpaceBone {boneName=root}
   }
   Save cached pose 'SkirtDynPose'(Pose=Component To Local(ComponentPose=@n103)) {cacheName=SkirtDynPose}
   Use cached pose 'SkirtCachePose'()
  ]}
 EventGraph {nodes [
   Event BlueprintUpdateAnimation()
   > n1 = Sequence!(execute) -> (then_0, then_1, then_2:Exec, then_3:Exec, then_4:Exec, then_5:Exec, then_6:Exec, then_7:Exec, then_8:Exec)
   n2 = Cast To PlayerPawn_Athena(execute=@n1.then_1, Object=TryGetPawnOwner())
   > Set head_mesh(head_mesh:Skeletal Mesh Component Object Reference=GetSkeletalMeshForPartType(self=@n2, PartType=Head)) -> (Output_Get:Skeletal Mesh Component Object Reference)
   n3 = Cast To FortPlayerAnimInstance(execute=@n1.then_2, Object=GetAnimInstance(self=GetOwnerSkeletalMeshComponent()))
   > Set isAimingPistol(isAimingPistol:double=0) -> (Output_Get:double)
   > Set isPickaxe(isPickaxe:double=0) -> (Output_Get:double)
   > Set isAimingRifle(isAimingRifle:double=0) -> (Output_Get:double)
   > Set isPickaxeBoo(isPickaxeBoo:bool=false) -> (Output_Get:bool)
   > n4 = Switch on EFortWeaponCoreAnimation!(execute, Selection:EFortWeaponCoreAnimation Enum=Get WeaponCoreAnim(self=@n3)) -> (
    Melee:Exec
    Pistol:Exec
    Shotgun:Exec
    PaperBlueprint:Exec
    Rifle:Exec
    MeleeOneHand:Exec
    MachinePistol:Exec
    RocketLauncher:Exec
    GrenadeLauncher:Exec
    GoingCommando:Exec
    AssaultRifle:Exec
    TacticalShotgun:Exec
    SniperRifle:Exec
    TrapPlacement:Exec
    ShoulderLauncher:Exec
    AbilityDecoTool:Exec
    Crossbow:Exec
    C4:Exec
    RemoteControl:Exec
    DualWield:Exec
    AR_BullPup:Exec
    AR_ForwardGrip:Exec
    MedPackPaddles:Exec
    SMG_P90:Exec
    AR_DrumGun:Exec
    Consumable_Small:Exec
    Consumable_Large:Exec
    Balloon:Exec
    MountedTurret:Exec
    CreativeTool:Exec
    ExplosiveBow:Exec
    AshtonIndigo:Exec
    AshtonChicago:Exec
    MeleeDualWield:Exec
    Unarmed:Exec
   )
   Branch(execute=[@n4.AssaultRifle, @n4.Crossbow, @n4.GrenadeLauncher, @n4.Rifle, @n4.RocketLauncher, @n4.Shotgun, @n4.SniperRifle, @n4.TacticalShotgun])
   > Set isAimingRifle(isAimingRifle=1)
   Branch(execute=[@n4.MachinePistol, @n4.Pistol])
   > Set isAimingPistol(isAimingPistol=1)
   Branch(execute=@n4.Melee)
   > Set isPickaxe(isPickaxe=1)
   > Set isPickaxeBoo(isPickaxeBoo=true)
   n5 = CheckForBackbling()
   Set Wearing Backbling(execute=@n1.then_3, Wearing Backbling:bool=OR Boolean(A=@n5, B=@n5.bIsBackblingACape, C=@n5.bIsBackblingLong)) -> (Output_Get:bool)
   n6 = Get bIsParachuteOpened()
   n7 = Get Wearing Backbling() -> (Wearing Backbling:bool)
   n8 = Branch(execute=@n1.then_4, Condition=OR Boolean(A=AND Boolean(A=@n7, B=@n6), B=Get bIsSkyDiving()))
   > Set BackpackParachuting(BackpackParachuting:bool=true) -> (Output_Get:bool)
   > Set isParachuting(isParachuting:double=0.2) -> (Output_Get:double)
   Set BackpackParachuting(execute=@n8.else, BackpackParachuting=false)
   > Set isParachuting(isParachuting=1)
   n9 = Get bIsParachuteOpened()
   n10 = Get bIsSkydiveDiveMode()
   n11 = Branch(execute=@n1.then_5, Condition=AND Boolean!(A=Get bIsSkyDiving(), B=NOT Boolean(A=@n9), C:bool=NOT Boolean(A=@n10)) -> (ReturnValue))
   > Set isSkyDiveNew(isSkyDiveNew:double=1) -> (Output_Get:double)
   Set isSkyDiveNew(execute=@n11.else, isSkyDiveNew=0)
   n12 = Branch(execute=@n1.then_6, Condition=Get bIsParachuteOpened())
   > Set ParachutingFromSkydive(ParachutingFromSkydive:double=1) -> (Output_Get:double)
   Set ParachutingFromSkydive(execute=@n12.else, ParachutingFromSkydive=0)
   n13 = Branch(execute=@n1.then_7, Condition=Get bIsSkyDiving())
   > n14 = Branch(Condition=Get bIsParachuteOpened())
   > Set goingUpFloat(goingUpFloat:double=0) -> (Output_Get:double)
   n15 = Branch(execute=@n14.else, Condition="float > float"(A=Get PawnVelocityZ(), B=50) {is=title K2Node_CallFunction})
   > Set goingUpFloat(goingUpFloat=1)
   n16 = Branch(execute=@n15.else, Condition=Get bIsSkydiveDiveMode())
   > Set goingUpFloat(goingUpFloat=0)
   Set goingUpFloat(execute=@n16.else, goingUpFloat=0)
   Set goingUpFloat(execute=@n13.else, goingUpFloat=0)
   n17 = Branch(execute=@n1.then_8, Condition=Get bIsInFrontEnd())
   > Set Frontend(Frontend:double=1) -> (Output_Get:double)
   Set Frontend(execute=@n17.else, Frontend=0)
   n18 = Gravity Override State Function DEPRECATED(
    execute=@n1.then_0
    CharacterPartAnimInstance=Self-Reference()
    GameTimeInSeconds=GetGameTimeInSeconds()
    GravityOverrideParams=Get Gravity Override Params()
   )
   > Set Gravity Override Cape(Gravity Override Cape:Vector=@n18) -> (Output_Get:Vector)
   n19 = Get bRigidBodyEnabled()
   n20 = Get CurrentLOD()
   > Set isRigidBodyDisabled(isRigidBodyDisabled:bool=OR Boolean!(A=NOT Boolean(A=@n19), B="integer > integer"(A=@n20, B=1) {is=title K2Node_CallFunction}) -> (ReturnValue)) -> (Output_Get:bool)
   Event BlueprintInitializeAnimation()
   > Set SkelMesh(SkelMesh=GetOwnerSkeletalMeshComponent())
   n21 = TryGetPawnOwner()
   > Is Valid(InputObject=@n21)
   > n22 = Cast To FortPlayerPawnAthena(Object=@n21)
   > Set PlayerPawn(PlayerPawn:Fort Player Pawn Athena Object Reference=@n22) -> (Output_Get:Fort Player Pawn Athena Object Reference)
   n23 = Event OnEndSkydiving()
   ResetDynamics(execute=[@n23.then, @n24.then], InTeleportType=ResetPhysics)
   n24 = Event OnParachuteOpened()
  ]}
]
variables [
 BackpackParachuting {type=boolean}
 Frontend {type=number}
 goingUp {type=boolean}
 goingUpFloat {type=number}
 Gravity Override Cape {type=Vector}
 Gravity Override Params {
  type=GravityOverrideParamsStruct
  default {
   riding_Moving_Targeting {gravityOverride=<0 -200 -800>}
   riding_Standing {gravityOverride=<0 -200 -980>}
   skydive_DivingDown {gravityOverride=<0 200 0>, windAmplitude=<0 0 0>, windFrequency=<0 0 0>}
   skydive_FloatingDown {jointName=pelvis, gravityOverride=<0 200 0>, windAmplitude=<0 0 0>, windFrequency=<0 0 0>, windOffset=<0 0 0>}
   skydive_GoingUp {gravityOverride=<0 -300 -200>}
   skydive_Parachuting {gravityOverride=<0 -600 -980>}
  }
 }
 head_mesh {type=SkeletalMeshComponent}
 isAimingPistol {type=number}
 isAimingRifle {type=number}
 isParachuting {type=number}
 isPickaxe {type=number}
 isPickaxeBoo {type=boolean}
 isRigidBodyDisabled {type=boolean}
 isSkyDiveNew {type=number}
 ParachutingFromSkydive {type=number}
 PlayerPawn {type=FortPlayerPawnAthena}
 SkelMesh {type=SkeletalMeshComponent}
 Wearing Backbling {type=boolean}
]
